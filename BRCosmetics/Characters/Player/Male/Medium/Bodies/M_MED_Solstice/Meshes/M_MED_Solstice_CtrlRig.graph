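# Control Rig for the M_MED_Solstice male character body that corrects both shoulder pads as the upper arms move.
# A spherical pose reader on each upperarm bone outputs a weight, with the right side using an opposite driver axis, flipped scaling and its own rotation offset.
# That weight blends fixed Euler offsets onto the current local rotation of the pad's MSTR and child bones.
# The fur, bandolier and other dyn bones are untouched.
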ControlRigBlueprint /BRCosmetics/Characters/Player/Male/Medium/Bodies/M_MED_Solstice/Meshes/M_MED_Solstice_CtrlRig
graphs [
 RigVMFunctionLibrary {links [], nodes []}
 RigVMModel {nodes [
   RigUnit_BeginExecution = &RigUnit_BeginExecution()
   n1 = &RigUnit_SphericalPoseReader(
    ActiveRegionScaleFactors=(PositiveWidth=1,NegativeWidth=1,PositiveHeight=1,NegativeHeight=1)
    ActiveRegionSize=0.1
    Debug=(bDrawDebug=false,DebugScale=25,DebugSegments=20,DebugThickness=0.25)
    DriverAxis=(X=1,Y=0,Z=0)
    DriverItem=(Type=Bone,Name="upperarm_l")
    FalloffRegionScaleFactors=(PositiveWidth=1,NegativeWidth=1,PositiveHeight=1,NegativeHeight=1)
    FalloffSize=0.5
    OptionalParentItem=(Type=Bone)
    RotationOffset=(X=70,Y=0,Z=-20)
    ExecutePin=@RigUnit_BeginExecution.ExecutePin
   )
   RigUnit_GetTransform = &RigUnit_GetTransform(Item=(Type=Bone,Name="dyn_shoulderpad_l_MSTR"), Space=LocalSpace, bInitial=false)
   n2 = &RigVMFunction_MathQuaternionToEuler(RotationOrder=ZYX, Value=@RigUnit_GetTransform.Transform.Rotation)
   > RigUnit_SetTransform = &RigUnit_SetTransform(
    Item=(Type=Bone,Name="dyn_shoulderpad_l_MSTR")
    Space=LocalSpace
    Value=(Rotation=(X=0.0132,Y=0.4099,Z=-0.0222,W=0.9118),Translation=(X=23.6045,Y=-6.37,Z=140.1115),Scale3D=(X=1,Y=1,Z=1))
    bInitial=false
    Value.Rotation=&RigVMFunction_MathQuaternionFromEuler(
     Euler=(X=0,Y=0,Z=0)
     Euler.X=&RigVMFunction_MathFloatAdd(B=60, A=@n2.Result.X) {template="Add::Execute(in A,in B,out Result)"}.Result
     Euler.Y=&RigVMFunction_MathFloatAdd(B=7, A=@n2.Result.Y) {template="Add::Execute(in A,in B,out Result)"}.Result
     Euler.Z=&RigVMFunction_MathFloatAdd(A=@n2.Result.Z) {template="Add::Execute(in A,in B,out Result)"}.Result
    ).Result
    Value.Scale3D=@RigUnit_GetTransform.Transform.Scale3D
    Value.Translation=@RigUnit_GetTransform.Transform.Translation
    Weight=@n1.OutputParam
   )
   n3 = &RigUnit_GetTransform(Item=(Type=Bone,Name="dyn_shoulderpad_l"), Space=LocalSpace)
   n4 = &RigVMFunction_MathQuaternionToEuler(RotationOrder=ZYX, Value=@n3.Transform.Rotation)
   > &RigUnit_SetTransform(
    Item=(Type=Bone,Name="dyn_shoulderpad_l")
    Space=LocalSpace
    Value=(Rotation=(X=-0.4099,Y=-0.0132,Z=0.9118,W=-0.0222),Translation=(X=-23.8034,Y=-6.3789,Z=139.887),Scale3D=(X=1,Y=1,Z=1))
    Value.Rotation=&RigVMFunction_MathQuaternionFromEuler(
     Euler=(X=0,Y=0,Z=0)
     Euler.Y=&RigVMFunction_MathFloatAdd(B=-7, A=@n4.Result.Y) {template="Add::Execute(in A,in B,out Result)"}.Result
     Euler.Z=&RigVMFunction_MathFloatAdd(B=0, A=@n4.Result.Z) {template="Add::Execute(in A,in B,out Result)"}.Result
     Euler.X=&RigVMFunction_MathFloatAdd(B=-15, A=@n4.Result.X) {template="Add::Execute(in A,in B,out Result)"}.Result
    ).Result
    Value.Scale3D=@n3.Transform.Scale3D
    Value.Translation=@n3.Transform.Translation
    Weight=@n1.OutputParam
   )
   > n5 = &RigUnit_SphericalPoseReader(
    ActiveRegionScaleFactors=(PositiveWidth=1,NegativeWidth=1,PositiveHeight=1,NegativeHeight=1)
    ActiveRegionSize=0.1
    Debug=(bDrawDebug=false,DebugScale=25,DebugSegments=20,DebugThickness=0.25)
    DriverAxis=(X=-1,Y=0,Z=0)
    DriverItem=(Type=Bone,Name="upperarm_r")
    FalloffRegionScaleFactors=(PositiveWidth=1,NegativeWidth=1,PositiveHeight=1,NegativeHeight=1)
    FalloffSize=0.5
    FlipHeightScaling=true
    FlipWidthScaling=true
    OptionalParentItem=(Type=Bone)
    RotationOffset=(X=-120,Y=0,Z=0)
   )
   n6 = &RigUnit_GetTransform(Item=(Type=Bone,Name="dyn_shoulderpad_r_MSTR"), Space=LocalSpace, bInitial=false)
   n7 = &RigVMFunction_MathQuaternionToEuler(RotationOrder=ZYX, Value=@n6.Transform.Rotation)
   > &RigUnit_SetTransform(
    Item=(Type=Bone,Name="dyn_shoulderpad_r_MSTR")
    Space=LocalSpace
    Value=(Rotation=(X=0.0132,Y=0.4099,Z=-0.0222,W=0.9118),Translation=(X=23.6045,Y=-6.37,Z=140.1115),Scale3D=(X=1,Y=1,Z=1))
    bInitial=false
    Value.Rotation=&RigVMFunction_MathQuaternionFromEuler(
     Euler=(X=0,Y=0,Z=0)
     Euler.Y=&RigVMFunction_MathFloatAdd(B=-7, A=@n7.Result.Y) {template="Add::Execute(in A,in B,out Result)"}.Result
     Euler.Z=&RigVMFunction_MathFloatAdd(A=@n7.Result.Z) {template="Add::Execute(in A,in B,out Result)"}.Result
     Euler.X=&RigVMFunction_MathFloatAdd(B=60, A=@n7.Result.X) {template="Add::Execute(in A,in B,out Result)"}.Result
    ).Result
    Value.Scale3D=@n6.Transform.Scale3D
    Value.Translation=@n6.Transform.Translation
    Weight=@n5.OutputParam
   )
   n8 = &RigUnit_GetTransform(Item=(Type=Bone,Name="dyn_shoulderpad_r"), Space=LocalSpace)
   n9 = &RigVMFunction_MathQuaternionToEuler(RotationOrder=ZYX, Value=@n8.Transform.Rotation)
   > &RigUnit_SetTransform(
    Item=(Type=Bone,Name="dyn_shoulderpad_r")
    Space=LocalSpace
    Value=(Rotation=(X=-0.4099,Y=-0.0132,Z=0.9118,W=-0.0222),Translation=(X=-23.8034,Y=-6.3789,Z=139.887),Scale3D=(X=1,Y=1,Z=1))
    Value.Rotation=&RigVMFunction_MathQuaternionFromEuler(
     Euler=(X=0,Y=0,Z=0)
     Euler.Y=&RigVMFunction_MathFloatAdd(B=0, A=@n9.Result.Y) {template="Add::Execute(in A,in B,out Result)"}.Result
     Euler.Z=&RigVMFunction_MathFloatAdd(A=@n9.Result.Z) {template="Add::Execute(in A,in B,out Result)"}.Result
     Euler.X=&RigVMFunction_MathFloatAdd(B=-20, A=@n9.Result.X) {template="Add::Execute(in A,in B,out Result)"}.Result
    ).Result
    Value.Scale3D=@n8.Transform.Scale3D
    Value.Translation=@n8.Transform.Translation
    Weight=@n5.OutputParam
   )
  ]}
]
hierarchy [
 root {loc=<0 0 0>, rot=<0 0 0 1>}
 pelvis {parent=root, loc=<0 -0.7071 93.1191>, rot=<0.0127 0.707 -0.0127 -0.707>}
 spine_01 {parent=pelvis, loc=<6.7494 -0.6388 0>, rot=<0 0 0.034 -0.9994>}
 spine_02 {parent=spine_01, loc=<7.2518 -0.035 0>, rot=<0 0 0.0238 -0.9997>}
 spine_03 {parent=spine_02, loc=<6.0104 0 0>, rot=<0 0 0.0149 -0.9999>}
 spine_04 {parent=spine_03, loc=<8.1024 0 0>, rot=<0 0 -0.0132 -0.9999>}
 spine_05 {parent=spine_04, loc=<10.0146 0 0>, rot=<0 0 -0.0015 -1>}
 clavicle_l {parent=spine_05, loc=<4.3575 -0.1466 -2.3223>, rot=<0.0465 -0.7289 -0.009 -0.6829>}
 upperarm_l {parent=clavicle_l, loc=<15.5111 0 0>, rot=<0.0226 -0.3234 -0.0088 -0.9459>}
 lowerarm_l {parent=upperarm_l, loc=<26.6514 0 0>, rot=<0.0391 -0.0099 -0.2453 -0.9686>}
 hand_l {parent=lowerarm_l, loc=<22.3491 0 0>, rot=<-0.7081 -0.0507 0.0975 -0.6975>}
 index_metacarpal_l {parent=hand_l, loc=<1.6535 0.0183 -2.96>, rot=<0.0158 -0.0124 -0.0173 -0.9996>}
 index_01_l {parent=index_metacarpal_l, loc=<8.3008 0.0357 -0.0524>, rot=<-0.0044 0.0485 0.0924 -0.9945>}
 index_02_l {parent=index_01_l, loc=<3.3587 -0.3004 0.0452>, rot=<-0.002 0.0027 0.3225 -0.9466>}
 index_03_l {parent=index_02_l, loc=<2.5578 -0.123 -0.0252>, rot=<-0.002 0.0067 0.2558 -0.9667>}
 middle_metacarpal_l {parent=hand_l, loc=<2.1856 -0.1725 -0.5984>, rot=<0.006 0.0168 -0.0329 -0.9993>}
 middle_01_l {parent=middle_metacarpal_l, loc=<7.3275 -0.2447 0.0129>, rot=<-0.0208 0.0042 0.0858 -0.9961>}
 middle_02_l {parent=middle_01_l, loc=<4.0123 0.1401 0.1175>, rot=<-0.0218 0.0258 0.3512 -0.9357>}
 middle_03_l {parent=middle_02_l, loc=<2.6077 0.0534 0.1373>, rot=<0.0082 0.0188 0.1708 -0.9851>}
 pinky_metacarpal_l {parent=hand_l, loc=<1.7616 -0.9753 3.3814>, rot=<-0.0397 0.0814 -0.069 -0.9935>}
 pinky_01_l {parent=pinky_metacarpal_l, loc=<7.1709 -0.4831 -0.1094>, rot=<-0.0654 -0.0035 0.2242 -0.9723>}
 pinky_02_l {parent=pinky_01_l, loc=<2.8274 0.0893 -0.1043>, rot=<0.0037 0.0011 0.2854 -0.9584>}
 pinky_03_l {parent=pinky_02_l, loc=<2.2046 0.0307 -0.0078>, rot=<-0.0002 0.0098 0.1445 -0.9895>}
 ring_metacarpal_l {parent=hand_l, loc=<2.2074 -0.3625 1.3638>, rot=<-0.0284 0.0722 -0.0458 -0.9959>}
 ring_01_l {parent=ring_metacarpal_l, loc=<7.0295 -0.6197 -0.1446>, rot=<-0.0489 -0.0252 0.1892 -0.9804>}
 ring_02_l {parent=ring_01_l, loc=<3.6174 0.4647 -0.1027>, rot=<0 0 0.2806 -0.9598>}
 ring_03_l {parent=ring_02_l, loc=<2.4152 0.114 0.0199>, rot=<0 0 0.1189 -0.9929>}
 thumb_01_l {parent=hand_l, loc=<1.3118 -0.3191 -2.459>, rot=<0.6546 -0.4042 0.0664 -0.6354>}
 thumb_02_l {parent=thumb_01_l, loc=<5.9336 -0.0448 0.0401>, rot=<-0.0037 0.0213 0.298 -0.9543>}
 thumb_03_l {parent=thumb_02_l, loc=<2.2223 0.3682 -0.0907>, rot=<0.0118 -0.013 0.0066 -0.9998>}
 dyn_fur_lowerarm_l_07_MSTR {parent=hand_l, loc=<4.3067 3.535 -0.8305>, rot=<0.6641 0.1984 0.0242 -0.7204>}
 dyn_fur_lowerarm_l_07 {parent=dyn_fur_lowerarm_l_07_MSTR, loc=<0.3 0 0>, rot=<0 0 0 1>}
 lowerarm_twist_01_l {parent=lowerarm_l, loc=<17.8549 0 0>, rot=<0 0 0 1>}
 dyn_fur_lowerarm_l_06_MSTR {parent=lowerarm_twist_01_l, loc=<3.7736 -5.0872 3.208>, rot=<-0.3545 -0.0014 0.2459 -0.9021>}
 dyn_fur_lowerarm_l_06 {parent=dyn_fur_lowerarm_l_06_MSTR, loc=<0.3 0 0>, rot=<0 0 0 1>}
 dyn_fur_lowerarm_l_08_MSTR {parent=lowerarm_twist_01_l, loc=<3.7728 3.649 3.4491>, rot=<0.2283 0.2682 -0.053 -0.9344>}
 dyn_fur_lowerarm_l_08 {parent=dyn_fur_lowerarm_l_08_MSTR, loc=<0.3 0 0>, rot=<0 0 0 1>}
 dyn_fur_lowerarm_l_05_MSTR {parent=lowerarm_twist_01_l, loc=<-1.1059 1.7736 5.4284>, rot=<0.0673 0.1022 0.0857 -0.9888>}
 dyn_fur_lowerarm_l_05 {parent=dyn_fur_lowerarm_l_05_MSTR, loc=<0.3 0 0>, rot=<0 0 0 1>}
 dyn_fur_lowerarm_l_09_MSTR {parent=lowerarm_twist_01_l, loc=<2.9663 -2.5929 -6.361>, rot=<0.9252 0.1182 -0.3413 0.1162>}
 dyn_fur_lowerarm_l_09 {parent=dyn_fur_lowerarm_l_09_MSTR, loc=<0.3 0 0>, rot=<0 0 0 1>}
 dyn_fur_lowerarm_l_10_MSTR {parent=lowerarm_twist_01_l, loc=<5.2502 5.3974 -3.3258>, rot=<0.6736 -0.0513 -0.4071 -0.6147>}
 dyn_fur_lowerarm_l_10 {parent=dyn_fur_lowerarm_l_10_MSTR, loc=<0.3 0 0>, rot=<0 0 0 1>}
 lowerarm_twist_02_l {parent=lowerarm_l, loc=<9.6199 0 0>, rot=<0 0 0 1>}
 dyn_fur_lowerarm_l_04_MSTR {parent=lowerarm_twist_02_l, loc=<0.9676 -6.3276 -0.3268>, rot=<-0.6477 0.1945 0.1519 -0.7209>}
 dyn_fur_lowerarm_l_04 {parent=dyn_fur_lowerarm_l_04_MSTR, loc=<0.3 0 0>, rot=<0 0 0 1>}
 deform_elbow_bk_l {parent=lowerarm_l, loc=<-1.2853 -2.9274 -0.5607>, rot=<-0.0285 0.0335 -0.0024 -0.999>}
 dyn_fur_lowerarm_l_03_MSTR {parent=lowerarm_l, loc=<0.3991 -5.9099 -3.3086>, rot=<-0.7365 0.187 0.0529 -0.648>}
 dyn_fur_lowerarm_l_03 {parent=dyn_fur_lowerarm_l_03_MSTR, loc=<0.3 0 0>, rot=<0 0 0 1>}
 dyn_fur_lowerarm_l_02_MSTR {parent=lowerarm_l, loc=<2.9102 1.0563 -8.1577>, rot=<0.9063 -0.118 -0.4053 0.0196>}
 dyn_fur_lowerarm_l_02 {parent=dyn_fur_lowerarm_l_02_MSTR, loc=<0.3 0 0>, rot=<0 0 0 1>}
 dyn_fur_lowerarm_l_01_MSTR {parent=lowerarm_l, loc=<-0.9223 -1.2787 7.1687>, rot=<-0.1319 0.1832 0.2555 -0.9401>}
 dyn_fur_lowerarm_l_01 {parent=dyn_fur_lowerarm_l_01_MSTR, loc=<0.3 0 0>, rot=<0 0 0 1>}
 upperarm_twist_01_l {parent=upperarm_l, loc=<0.0041 0 0>, rot=<0 0 0 1>}
 deform_deltoid_fr_l {parent=upperarm_twist_01_l, loc=<2.4511 0 0>, rot=<0 0 0 1>}
 dyn_shoulderpad_l_MSTR {parent=upperarm_twist_01_l, loc=<0 0 0.9678>, rot=<0 0 0 1>}
 dyn_shoulderpad_l {parent=dyn_shoulderpad_l_MSTR, loc=<0.9403 -1.623 7.0749>, rot=<-0.0266 -0.0594 -0.0022 -0.9979>}
 upperarm_twist_02_l {parent=upperarm_l, loc=<13.7941 0 0>, rot=<0 0 0 1>}
 dyn_fur_upperarm_l_01_MSTR {parent=upperarm_twist_02_l, loc=<-0.4853 2.8336 7.3847>, rot=<0.1444 0.1123 -0.2616 -0.9477>}
 dyn_fur_upperarm_l_01 {parent=dyn_fur_upperarm_l_01_MSTR, loc=<0.3 0 0>, rot=<0 0 0 1>}
 deform_elbow_fr_l {parent=upperarm_l, loc=<26.5048 0.1875 -0.5816>, rot=<0.0021 0.0156 -0.185 -0.9826>}
 dyn_fur_upperarm_l_02_MSTR {parent=upperarm_l, loc=<22.8166 7.471 5.2884>, rot=<0.2027 0.2699 -0.1674 -0.9263>}
 dyn_fur_upperarm_l_02 {parent=dyn_fur_upperarm_l_02_MSTR, loc=<0.3 0 0>, rot=<0 0 0 1>}
 deform_pectoral_fr_l {parent=clavicle_l, loc=<9.3905 0 0>, rot=<0 0 0 1>}
 clavicle_r {parent=spine_05, loc=<4.3575 -0.1466 2.3223>, rot=<0.7289 0.0465 -0.6829 0.009>}
 upperarm_r {parent=clavicle_r, loc=<-15.5111 0 0>, rot=<0.0226 -0.3234 -0.0088 -0.9459>}
 lowerarm_r {parent=upperarm_r, loc=<-26.6515 0 0>, rot=<0.0391 -0.0099 -0.2453 -0.9686>}
 hand_r {parent=lowerarm_r, loc=<-22.3492 0 0>, rot=<-0.7081 -0.0507 0.0975 -0.6975>}
 index_metacarpal_r {parent=hand_r, loc=<-1.6534 -0.0181 2.9601>, rot=<0.0158 -0.0124 -0.0173 -0.9996>}
 index_01_r {parent=index_metacarpal_r, loc=<-8.3008 -0.0357 0.0524>, rot=<-0.0044 0.0485 0.0924 -0.9945>}
 index_02_r {parent=index_01_r, loc=<-3.3589 0.3003 -0.0453>, rot=<-0.002 0.0027 0.3225 -0.9466>}
 index_03_r {parent=index_02_r, loc=<-2.5578 0.1231 0.0251>, rot=<-0.002 0.0067 0.2558 -0.9667>}
 middle_metacarpal_r {parent=hand_r, loc=<-2.1856 0.1726 0.5984>, rot=<0.006 0.0168 -0.0329 -0.9993>}
 middle_01_r {parent=middle_metacarpal_r, loc=<-7.3273 0.2447 -0.0129>, rot=<-0.0208 0.0042 0.0858 -0.9961>}
 middle_02_r {parent=middle_01_r, loc=<-4.0142 -0.1411 -0.07>, rot=<-0.0217 0.0259 0.3512 -0.9357>}
 middle_03_r {parent=middle_02_r, loc=<-2.6074 -0.0534 -0.1312>, rot=<0.0083 0.0188 0.1708 -0.9851>}
 pinky_metacarpal_r {parent=hand_r, loc=<-1.7617 0.9753 -3.3814>, rot=<-0.0397 0.0814 -0.069 -0.9935>}
 pinky_01_r {parent=pinky_metacarpal_r, loc=<-7.1707 0.4832 0.1094>, rot=<-0.0654 -0.0035 0.2242 -0.9723>}
 pinky_02_r {parent=pinky_01_r, loc=<-2.8274 -0.0892 0.1043>, rot=<0.0037 0.0011 0.2854 -0.9584>}
 pinky_03_r {parent=pinky_02_r, loc=<-2.2045 -0.031 0.0078>, rot=<-0.0002 0.0098 0.1445 -0.9895>}
 ring_metacarpal_r {parent=hand_r, loc=<-2.2073 0.3627 -1.3638>, rot=<-0.0284 0.0722 -0.0458 -0.9959>}
 ring_01_r {parent=ring_metacarpal_r, loc=<-7.0297 0.6197 0.1446>, rot=<-0.0489 -0.0252 0.1892 -0.9804>}
 ring_02_r {parent=ring_01_r, loc=<-3.6171 -0.4648 0.1027>, rot=<0 0 0.2806 -0.9598>}
 ring_03_r {parent=ring_02_r, loc=<-2.4153 -0.1141 -0.0199>, rot=<0 0 0.1189 -0.9929>}
 thumb_01_r {parent=hand_r, loc=<-1.3119 0.3192 2.459>, rot=<0.6546 -0.4042 0.0664 -0.6354>}
 thumb_02_r {parent=thumb_01_r, loc=<-5.9335 0.0447 -0.0401>, rot=<-0.0037 0.0213 0.298 -0.9543>}
 thumb_03_r {parent=thumb_02_r, loc=<-2.2223 -0.3682 0.0907>, rot=<0.0118 -0.013 0.0066 -0.9998>}
 dyn_fur_lowerarm_r_07_MSTR {parent=hand_r, loc=<-4.8613 -3.9403 0.1247>, rot=<-0.0242 -0.7204 0.6641 -0.1984>}
 dyn_fur_lowerarm_r_07 {parent=dyn_fur_lowerarm_r_07_MSTR, loc=<0.2997 0 0.0005>, rot=<0 0 0 1>}
 lowerarm_twist_01_r {parent=lowerarm_r, loc=<-17.855 0 0>, rot=<0 0 0 1>}
 dyn_fur_lowerarm_r_05_MSTR {parent=lowerarm_twist_01_r, loc=<1.249 -1.9219 -5.358>, rot=<-0.0857 -0.9888 0.0673 -0.1022>}
 dyn_fur_lowerarm_r_05 {parent=dyn_fur_lowerarm_r_05_MSTR, loc=<0.2998 0 0.0003>, rot=<0 0 0 -1>}
 dyn_fur_lowerarm_r_08_MSTR {parent=lowerarm_twist_01_r, loc=<-3.8118 -2.9736 -3.5846>, rot=<0.1349 -0.9261 0.251 -0.2471>}
 dyn_fur_lowerarm_r_08 {parent=dyn_fur_lowerarm_r_08_MSTR, loc=<0.3 0 -0.0001>, rot=<0 0 0 -1>}
 dyn_fur_lowerarm_r_06_MSTR {parent=lowerarm_twist_01_r, loc=<-4.2882 4.9722 -3.0326>, rot=<-0.3251 -0.8767 -0.353 0.0329>}
 dyn_fur_lowerarm_r_06 {parent=dyn_fur_lowerarm_r_06_MSTR, loc=<0.2999 -0.0001 0.0001>, rot=<0 0 0 1>}
 dyn_fur_lowerarm_r_09_MSTR {parent=lowerarm_twist_01_r, loc=<-3.5727 2.6031 6.7984>, rot=<0.3672 0.1194 0.9152 -0.1149>}
 dyn_fur_lowerarm_r_09 {parent=dyn_fur_lowerarm_r_09_MSTR, loc=<0.3 0 0>, rot=<0 0 0 1>}
 lowerarm_twist_02_r {parent=lowerarm_r, loc=<-9.6199 0 0>, rot=<0 0 0 1>}
 dyn_fur_lowerarm_r_04_MSTR {parent=lowerarm_twist_02_r, loc=<-0.3669 1.6528 -6.9601>, rot=<-0.146 -0.9804 -0.0611 -0.1174>}
 dyn_fur_lowerarm_r_04 {parent=dyn_fur_lowerarm_r_04_MSTR, loc=<0.3001 0 0>, rot=<0 0 0 -1>}
 deform_elbow_bk_r {parent=lowerarm_r, loc=<1.2852 2.9274 0.561>, rot=<-0.0285 0.0335 -0.0024 -0.999>}
 dyn_fur_lowerarm_r_01_MSTR {parent=lowerarm_r, loc=<0.8259 1.3415 -7.2698>, rot=<-0.2555 -0.9401 -0.1319 -0.1832>}
 dyn_fur_lowerarm_r_01 {parent=dyn_fur_lowerarm_r_01_MSTR, loc=<0.2997 -0.0001 0.0006>, rot=<0 0 0 1>}
 dyn_fur_lowerarm_r_03_MSTR {parent=lowerarm_r, loc=<-0.5415 5.9255 2.9391>, rot=<0.0997 0.6424 0.7481 0.1331>}
 dyn_fur_lowerarm_r_03 {parent=dyn_fur_lowerarm_r_03_MSTR, loc=<0.3001 0.0001 0>, rot=<0 0 0 1>}
 upperarm_twist_01_r {parent=upperarm_r, loc=<-0.0041 0 0>, rot=<0 0 0 1>}
 deform_deltoid_fr_r {parent=upperarm_twist_01_r, loc=<-2.4512 0 0>, rot=<0 0 0 1>}
 dyn_shoulderpad_r_MSTR {parent=upperarm_twist_01_r, loc=<-0.0003 0 -0.9675>, rot=<0 -1 0 0>}
 dyn_shoulderpad_r {parent=dyn_shoulderpad_r_MSTR, loc=<0.9401 1.6231 7.075>, rot=<0.0266 -0.0594 0.0022 -0.9979>}
 upperarm_twist_02_r {parent=upperarm_r, loc=<-13.7942 0 0>, rot=<0 0 0 1>}
 dyn_fur_upperarm_r_01_MSTR {parent=upperarm_twist_02_r, loc=<0.485 -2.8336 -7.3843>, rot=<0.2616 -0.9477 0.1444 -0.1123>}
 dyn_fur_upperarm_r_01 {parent=dyn_fur_upperarm_r_01_MSTR, loc=<0.2997 0 0.0006>, rot=<0 0 0 -1>}
 deform_elbow_fr_r {parent=upperarm_r, loc=<-26.5046 -0.1875 0.5813>, rot=<0.0021 0.0156 -0.185 -0.9826>}
 dyn_fur_upperarm_r_02_MSTR {parent=upperarm_r, loc=<-21.7214 -6.8516 -4.7475>, rot=<0.1482 -0.9293 0.2995 -0.1572>}
 dyn_fur_upperarm_r_02 {parent=dyn_fur_upperarm_r_02_MSTR, loc=<0.2998 0.0001 0>, rot=<0 0 0 1>}
 deform_pectoral_fr_r {parent=clavicle_r, loc=<-9.39 0 0>, rot=<0 0 0 1>}
 neck_01 {parent=spine_05, loc=<11.3138 0.1529 0>, rot=<0 0 -0.2003 -0.9797>}
 neck_02 {parent=neck_01, loc=<4.5627 0 0>, rot=<0 0 0.0504 -0.9987>}
 dyn_bandolier_fr_top_l_01_MSTR {parent=spine_05, loc=<-3.5656 14.4851 -9.6509>, rot=<0.1118 -0.8219 -0.5338 -0.1644>}
 dyn_bandolier_fr_top_l_01 {parent=dyn_bandolier_fr_top_l_01_MSTR, loc=<0.3 0 0>, rot=<0 0 0 1>}
 dyn_bandolier_fr_top_l_02_MSTR {parent=spine_05, loc=<-0.8644 11.0493 -15.321>, rot=<0.0892 -0.8438 -0.4924 -0.1939>}
 dyn_bandolier_fr_top_l_02 {parent=dyn_bandolier_fr_top_l_02_MSTR, loc=<0.3 0 0>, rot=<0 0 0 1>}
 dyn_bandolier_fr_top_l_03_MSTR {parent=spine_05, loc=<1.6055 7.6196 -20.6348>, rot=<0.0751 -0.8262 -0.4994 -0.2496>}
 dyn_bandolier_fr_top_l_03 {parent=dyn_bandolier_fr_top_l_03_MSTR, loc=<0.3 0 0>, rot=<0 0 0 1>}
 dyn_bandolier_bk_top_l_01_MSTR {parent=spine_05, loc=<-2.4656 -14.8715 -9.885>, rot=<-0.1562 -0.7459 0.6295 -0.1516>}
 dyn_bandolier_bk_top_l_01 {parent=dyn_bandolier_bk_top_l_01_MSTR, loc=<0.3 0 0>, rot=<0 0 0 1>}
 dyn_bandolier_bk_top_l_02_MSTR {parent=spine_05, loc=<0.3651 -13.5542 -15.3759>, rot=<-0.1548 -0.7631 0.5921 -0.2078>}
 dyn_bandolier_bk_top_l_02 {parent=dyn_bandolier_bk_top_l_02_MSTR, loc=<0.3 0 0>, rot=<0 0 0 1>}
 dyn_bandolier_bk_top_l_03_MSTR {parent=spine_05, loc=<3.5071 -11.6204 -20.6766>, rot=<-0.1199 -0.8029 0.5297 -0.2458>}
 dyn_bandolier_bk_top_l_03 {parent=dyn_bandolier_bk_top_l_03_MSTR, loc=<0.3 0 0>, rot=<0 0 0 1>}
 dyn_bandolier_fr_top_r_01_MSTR {parent=spine_05, loc=<-3.5652 14.4852 9.6509>, rot=<0.1644 0.5338 0.8219 -0.1118>}
 dyn_bandolier_fr_top_r_01 {parent=dyn_bandolier_fr_top_r_01_MSTR, loc=<0.3004 -0.0001 -0.0001>, rot=<0 0 0 -1>}
 dyn_bandolier_fr_top_r_02_MSTR {parent=spine_05, loc=<-0.8642 11.0494 15.321>, rot=<0.1939 0.4924 0.8438 -0.0892>}
 dyn_bandolier_fr_top_r_02 {parent=dyn_bandolier_fr_top_r_02_MSTR, loc=<0.3002 0 0>, rot=<0 0 0 1>}
 dyn_bandolier_fr_top_r_03_MSTR {parent=spine_05, loc=<1.6056 7.6196 20.6348>, rot=<0.2496 0.4994 0.8262 -0.0751>}
 dyn_bandolier_fr_top_r_03 {parent=dyn_bandolier_fr_top_r_03_MSTR, loc=<0.2996 0.0001 0.0002>, rot=<0 0 0 -1>}
 dyn_bandolier_bk_top_r_01_MSTR {parent=spine_05, loc=<-2.4652 -14.8715 9.885>, rot=<0.1516 -0.6295 0.7459 0.1562>}
 dyn_bandolier_bk_top_r_01 {parent=dyn_bandolier_bk_top_r_01_MSTR, loc=<0.3003 0.0001 0>, rot=<0 0 0 1>}
 dyn_bandolier_bk_top_r_02_MSTR {parent=spine_05, loc=<0.3653 -13.5541 15.3759>, rot=<0.2078 -0.5921 0.7631 0.1548>}
 dyn_bandolier_bk_top_r_02 {parent=dyn_bandolier_bk_top_r_02_MSTR, loc=<0.2999 0 0.0001>, rot=<0 0 0 1>}
 dyn_bandolier_bk_top_r_03_MSTR {parent=spine_05, loc=<3.5075 -11.6204 20.6766>, rot=<0.2458 -0.5297 0.8029 0.1199>}
 dyn_bandolier_bk_top_r_03 {parent=dyn_bandolier_bk_top_r_03_MSTR, loc=<0.3 0 -0.0001>, rot=<0 0 0 1>}
 dyn_fur_torso_bk_01_MSTR {parent=spine_05, loc=<5.2253 -13.7698 5.8531>, rot=<0.2181 -0.5864 0.7601 -0.1754>}
 dyn_fur_torso_bk_01 {parent=dyn_fur_torso_bk_01_MSTR, loc=<0.3 0 0>, rot=<0 0 0 1>}
 dyn_fur_torso_r_02_MSTR {parent=spine_04, loc=<-5.1026 0.0969 21.0049>, rot=<0.047 0.1042 0.9934 0.009>}
 dyn_fur_torso_r_02 {parent=dyn_fur_torso_r_02_MSTR, loc=<0.3 0 0>, rot=<0 0 0 1>}
 dyn_fur_torso_l_02_MSTR {parent=spine_04, loc=<1.0167 12.1927 -16.3766>, rot=<-0.1027 -0.8291 -0.5433 -0.083>}
 dyn_fur_torso_l_02 {parent=dyn_fur_torso_l_02_MSTR, loc=<0.3 0 0>, rot=<0 0 0 1>}
 dyn_fur_torso_l_01_MSTR {parent=spine_04, loc=<-5.203 -0.0078 -21.0111>, rot=<-0.009 -0.9934 -0.1042 -0.047>}
 dyn_fur_torso_l_01 {parent=dyn_fur_torso_l_01_MSTR, loc=<0.3 0 0>, rot=<0 0 0 1>}
 dyn_bandolier_fr_mid_MSTR {parent=spine_04, loc=<1.6031 15.6031 0>, rot=<-0.0294 0.7065 0.7065 -0.0294>}
 dyn_bandolier_fr_mid {parent=dyn_bandolier_fr_mid_MSTR, loc=<0.3 0 0>, rot=<0 0 0 1>}
 dyn_bandolier_bk_mid_MSTR {parent=spine_04, loc=<3.0065 -15.1786 0>, rot=<-0.0302 -0.7065 0.7065 0.0302>}
 dyn_bandolier_bk_mid {parent=dyn_bandolier_bk_mid_MSTR, loc=<0.3 0 0>, rot=<0 0 0 1>}
 dyn_fur_torso_r_01_MSTR {parent=spine_03, loc=<-0.9953 -13.7161 -7.7928>, rot=<0.2046 -0.7071 0.6722 -0.0791>}
 dyn_fur_torso_r_01 {parent=dyn_fur_torso_r_01_MSTR, loc=<0.3 0 0>, rot=<0 0 0 1>}
 dyn_fur_torso_fr_02_MSTR {parent=spine_03, loc=<-1.3199 14.0662 0.6279>, rot=<0.0991 0.6978 0.7091 0.0207>}
 dyn_fur_torso_fr_02 {parent=dyn_fur_torso_fr_02_MSTR, loc=<0.3 0 0>, rot=<0 0 0 1>}
 dyn_fur_torso_fr_04_MSTR {parent=spine_03, loc=<-3.6171 13.3342 -6.8133>, rot=<-0.022 -0.81 -0.5797 -0.0862>}
 dyn_fur_torso_fr_04 {parent=dyn_fur_torso_fr_04_MSTR, loc=<0.3 0 0>, rot=<0 0 0 1>}
 dyn_fur_torso_bk_02_MSTR {parent=spine_02, loc=<1.181 -12.6268 -9.7315>, rot=<0.1341 -0.8216 0.5479 -0.0822>}
 dyn_fur_torso_bk_02 {parent=dyn_fur_torso_bk_02_MSTR, loc=<0.3 0 0>, rot=<0 0 0 1>}
 dyn_fur_torso_fr_01_MSTR {parent=spine_02, loc=<0.0235 10.5638 10.4066>, rot=<0.1823 0.5014 0.8325 0.1496>}
 dyn_fur_torso_fr_01 {parent=dyn_fur_torso_fr_01_MSTR, loc=<0.3 0 0>, rot=<0 0 0 1>}
 dyn_fur_torso_fr_03_MSTR {parent=spine_02, loc=<-3.0672 12.733 3.3356>, rot=<-0.0957 -0.7075 -0.6879 -0.1303>}
 dyn_fur_torso_fr_03 {parent=dyn_fur_torso_fr_03_MSTR, loc=<0.3 0 0>, rot=<0 0 0 1>}
 thigh_l {parent=pelvis, loc=<-2.5693 -0.0109 -7.7501>, rot=<0.0012 0.0572 0.0155 -0.9982>}
 calf_l {parent=thigh_l, loc=<-39.0484 0 0>, rot=<0.0006 -0.0161 -0.0391 -0.9991>}
 calf_twist_01_l {parent=calf_l, loc=<-35.813 -0.0967 0.0397>, rot=<0.0237 0.0095 -0.0035 -0.9997>}
 dyn_fur_lowerleg_l_05_MSTR {parent=calf_twist_01_l, loc=<6.2423 -5.0576 -7.9458>, rot=<0.1191 -0.8586 0.4925 -0.0773>}
 dyn_fur_lowerleg_l_05 {parent=dyn_fur_lowerleg_l_05_MSTR, loc=<0.3 0 0>, rot=<0 0 0 1>}
 calf_twist_02_l {parent=calf_l, loc=<-18.8064 -0.0967 0.0397>, rot=<0.0237 0.0095 -0.0035 -0.9997>}
 dyn_fur_lowerleg_l_03_MSTR {parent=calf_twist_02_l, loc=<-2.8645 -2.8481 7.9065>, rot=<0.1887 0.1107 0.9609 0.1695>}
 dyn_fur_lowerleg_l_03 {parent=dyn_fur_lowerleg_l_03_MSTR, loc=<0.3 0 0>, rot=<0 0 0 1>}
 dyn_fur_lowerleg_l_02_MSTR {parent=calf_twist_02_l, loc=<3.8174 -8.7261 5.7679>, rot=<0.2602 -0.1276 0.9539 0.0783>}
 dyn_fur_lowerleg_l_02 {parent=dyn_fur_lowerleg_l_02_MSTR, loc=<0.3 0 0>, rot=<0 0 0 1>}
 dyn_fur_lowerleg_l_04_MSTR {parent=calf_twist_02_l, loc=<3.3322 3.0926 -8.7049>, rot=<-0.1634 -0.8439 -0.5044 -0.0822>}
 dyn_fur_lowerleg_l_04 {parent=dyn_fur_lowerleg_l_04_MSTR, loc=<0.3 0 0>, rot=<0 0 0 1>}
 foot_l {parent=calf_l, loc=<-42.561 0.1077 0.4292>, rot=<0.0418 -0.0455 0.0343 -0.9975>}
 ball_l {parent=foot_l, loc=<-6.1599 13.8264 0.6806>, rot=<0.0001 -0.0001 -0.7063 -0.7079>}
 dyn_fur_lowerleg_l_09_MSTR {parent=foot_l, loc=<-1.0143 2.3289 6.9835>, rot=<0.3116 -0.0244 0.9465 0.0805>}
 dyn_fur_lowerleg_l_09 {parent=dyn_fur_lowerleg_l_09_MSTR, loc=<0.3 0 0>, rot=<0 0 0 1>}
 dyn_fur_lowerleg_l_07_MSTR {parent=foot_l, loc=<-0.8447 -8.3393 1.9267>, rot=<0.1939 -0.5458 0.8019 -0.1463>}
 dyn_fur_lowerleg_l_07 {parent=dyn_fur_lowerleg_l_07_MSTR, loc=<0.3 0 0>, rot=<0 0 0 1>}
 dyn_fur_lowerleg_l_06_MSTR {parent=foot_l, loc=<-0.5067 -4.3816 -7.1619>, rot=<0.0704 -0.9125 0.2615 -0.3065>}
 dyn_fur_lowerleg_l_06 {parent=dyn_fur_lowerleg_l_06_MSTR, loc=<0.3 0 0>, rot=<0 0 0 1>}
 dyn_fur_lowerleg_l_08_MSTR {parent=foot_l, loc=<-0.6029 -8.094 -2.2223>, rot=<0.1721 -0.7459 0.5958 -0.2429>}
 dyn_fur_lowerleg_l_08 {parent=dyn_fur_lowerleg_l_08_MSTR, loc=<0.3 0 0>, rot=<0 0 0 1>}
 deform_knee_bk_01_l {parent=calf_l, loc=<0.6022 -1.0607 0>, rot=<0 0 -0.0408 -0.9992>}
 deform_knee_fr_01_l {parent=calf_l, loc=<-0.2342 0 0>, rot=<0 0 0 1>}
 dyn_fur_lowerleg_l_01_MSTR {parent=calf_l, loc=<-5.6253 0.1199 7.4131>, rot=<0.1965 -0.0894 0.949 -0.2298>}
 dyn_fur_lowerleg_l_01 {parent=dyn_fur_lowerleg_l_01_MSTR, loc=<0.3 0 0>, rot=<0 0 0 1>}
 thigh_twist_01_l {parent=thigh_l, loc=<-21.6303 0 0>, rot=<0.0013 0 0 -1>}
 deform_thigh_bk_l {parent=thigh_twist_01_l, loc=<16.9514 0.2058 0.1808>, rot=<-0.0013 -0.0217 0 -0.9998>}
 deform_thigh_fr_l {parent=thigh_twist_01_l, loc=<21.5084 0.2281 -0.0598>, rot=<-0.0013 -0.0217 0 -0.9998>}
 dyn_fur_upperleg_l_07_MSTR {parent=thigh_twist_01_l, loc=<10.0819 -4.8955 -10.0554>, rot=<0.0983 -0.9275 0.2466 -0.2632>}
 dyn_fur_upperleg_l_07 {parent=dyn_fur_upperleg_l_07_MSTR, loc=<0.3 0 0>, rot=<0 0 0 1>}
 dyn_fur_upperleg_l_01_MSTR {parent=thigh_twist_01_l, loc=<6.8942 5.3198 -11.4079>, rot=<-0.0314 -0.9498 -0.2591 -0.1727>}
 dyn_fur_upperleg_l_01 {parent=dyn_fur_upperleg_l_01_MSTR, loc=<0.3 0 0>, rot=<0 0 0 1>}
 dyn_fur_upperleg_l_02_MSTR {parent=thigh_l, loc=<-23.3499 -4.6555 7.8405>, rot=<0.1282 -0.2284 0.9651 -0.0045>}
 dyn_fur_upperleg_l_02 {parent=dyn_fur_upperleg_l_02_MSTR, loc=<0.3 0 0>, rot=<0 0 0 1>}
 dyn_fur_upperleg_l_05_MSTR {parent=thigh_l, loc=<-31.1263 0.39 9.0517>, rot=<0.154 -0.0605 0.9798 -0.1128>}
 dyn_fur_upperleg_l_05 {parent=dyn_fur_upperleg_l_05_MSTR, loc=<0.3 0 0>, rot=<0 0 0 1>}
 dyn_fur_upperleg_l_03_MSTR {parent=thigh_l, loc=<-36.4904 -1.9939 -9.5568>, rot=<-0.0147 -0.988 0.0576 -0.1428>}
 dyn_fur_upperleg_l_03 {parent=dyn_fur_upperleg_l_03_MSTR, loc=<0.3 0 0>, rot=<0 0 0 1>}
 dyn_fur_upperleg_l_06_MSTR {parent=thigh_l, loc=<-34.6099 -9.4319 0.978>, rot=<0.1466 -0.6843 0.7052 -0.1136>}
 dyn_fur_upperleg_l_06 {parent=dyn_fur_upperleg_l_06_MSTR, loc=<0.3 0 0>, rot=<0 0 0 1>}
 thigh_r {parent=pelvis, loc=<-2.5691 -0.0109 7.7501>, rot=<0.0572 -0.0012 0.9982 0.0155>}
 calf_r {parent=thigh_r, loc=<39.0487 0 0>, rot=<0.0006 -0.0161 -0.0391 -0.9991>}
 calf_twist_01_r {parent=calf_r, loc=<35.8132 0.0967 -0.0397>, rot=<0.0237 0.0095 -0.0035 -0.9997>}
 dyn_fur_lowerleg_r_05_MSTR {parent=calf_twist_01_r, loc=<-6.1636 5.024 8.1044>, rot=<0.4925 0.0773 -0.1191 -0.8586>}
 dyn_fur_lowerleg_r_05 {parent=dyn_fur_lowerleg_r_05_MSTR, loc=<0.3 0 0>, rot=<0 0 0 1>}
 dyn_fur_lowerleg_r_03_MSTR {parent=calf_twist_01_r, loc=<-2.6889 -1.5075 -5.9507>, rot=<0.9453 -0.1525 -0.2395 0.1607>}
 dyn_fur_lowerleg_r_03 {parent=dyn_fur_lowerleg_r_03_MSTR, loc=<0.3 0 0>, rot=<0 0 0 1>}
 calf_twist_02_r {parent=calf_r, loc=<18.8064 0.0967 -0.0397>, rot=<0.0237 0.0095 -0.0035 -0.9997>}
 dyn_fur_lowerleg_r_02_MSTR {parent=calf_twist_02_r, loc=<-4.4065 8.9295 -5.8173>, rot=<0.9524 -0.0236 -0.2537 -0.1676>}
 dyn_fur_lowerleg_r_02 {parent=dyn_fur_lowerleg_r_02_MSTR, loc=<0.3 -0.0001 0>, rot=<0 0 0 1>}
 dyn_fur_lowerleg_r_04_MSTR {parent=calf_twist_02_r, loc=<-0.8523 0.8838 10.2818>, rot=<-0.0246 0.1702 0.0708 -0.9826>}
 dyn_fur_lowerleg_r_04 {parent=dyn_fur_lowerleg_r_04_MSTR, loc=<0.3 0 -0.0001>, rot=<0 0 0 1>}
 foot_r {parent=calf_r, loc=<42.5612 -0.1077 -0.4292>, rot=<0.0418 -0.0455 0.0343 -0.9975>}
 ball_r {parent=foot_r, loc=<6.1598 -13.8264 -0.6806>, rot=<0.0001 -0.0001 -0.7063 -0.7079>}
 dyn_fur_lowerleg_r_08_MSTR {parent=foot_r, loc=<0.3346 8.3654 2.5766>, rot=<0.5958 0.2429 -0.1721 -0.7459>}
 dyn_fur_lowerleg_r_08 {parent=dyn_fur_lowerleg_r_08_MSTR, loc=<0.3 0 0>, rot=<0 0 0 -1>}
 dyn_fur_lowerleg_r_07_MSTR {parent=foot_r, loc=<0.896 8.2601 -1.8974>, rot=<0.8019 0.1463 -0.1939 -0.5458>}
 dyn_fur_lowerleg_r_07 {parent=dyn_fur_lowerleg_r_07_MSTR, loc=<0.3 0 0>, rot=<0 0 0 1>}
 dyn_fur_lowerleg_r_06_MSTR {parent=foot_r, loc=<0.817 3.5887 6.656>, rot=<0.2615 0.3065 -0.0704 -0.9125>}
 dyn_fur_lowerleg_r_06 {parent=dyn_fur_lowerleg_r_06_MSTR, loc=<0.3 0 0>, rot=<0 0 0 1>}
 dyn_fur_lowerleg_r_09_MSTR {parent=foot_r, loc=<1.826 1.8341 -6.7709>, rot=<0.9468 0.0762 -0.3034 -0.0753>}
 dyn_fur_lowerleg_r_09 {parent=dyn_fur_lowerleg_r_09_MSTR, loc=<0.3 0 0>, rot=<0 0 0 -1>}
 deform_knee_bk_01_r {parent=calf_r, loc=<-0.6022 1.0607 0>, rot=<0 0 -0.0408 -0.9992>}
 deform_knee_fr_01_r {parent=calf_r, loc=<0.2341 0 0>, rot=<0 0 0 1>}
 dyn_fur_lowerleg_r_01_MSTR {parent=calf_r, loc=<5.6875 4.1863 8.8085>, rot=<0.6773 0.4194 0.3232 -0.5108>}
 dyn_fur_lowerleg_r_01 {parent=dyn_fur_lowerleg_r_01_MSTR, loc=<0.3 0 0>, rot=<0 0 0 -1>}
 thigh_twist_01_r {parent=thigh_r, loc=<21.6303 0 0>, rot=<0.0013 0 0 -1>}
 deform_thigh_bk_r {parent=thigh_twist_01_r, loc=<-16.9512 -0.2058 -0.1808>, rot=<-0.0013 -0.0217 0 -0.9998>}
 deform_thigh_fr_r {parent=thigh_twist_01_r, loc=<-21.5083 -0.2281 0.0598>, rot=<-0.0013 -0.0217 0 -0.9998>}
 dyn_fur_upperleg_r_01_MSTR {parent=thigh_twist_01_r, loc=<-7.0155 -5.0085 11.2037>, rot=<-0.1252 0.1728 0.0405 -0.9761>}
 dyn_fur_upperleg_r_01 {parent=dyn_fur_upperleg_r_01_MSTR, loc=<0.3001 0 0>, rot=<0 0 0 1>}
 dyn_fur_upperleg_r_07_MSTR {parent=thigh_twist_01_r, loc=<-9.4063 5.096 10.271>, rot=<0.2466 0.2632 -0.0983 -0.9275>}
 dyn_fur_upperleg_r_07 {parent=dyn_fur_upperleg_r_07_MSTR, loc=<0.3 0 0>, rot=<0 0 0 1>}
 dyn_fur_upperleg_r_05_MSTR {parent=thigh_r, loc=<34.3349 -0.5527 -8.7817>, rot=<0.983 0.0467 -0.1125 0.1375>}
 dyn_fur_upperleg_r_05 {parent=dyn_fur_upperleg_r_05_MSTR, loc=<0.3 0 0>, rot=<0 0 0 1>}
 dyn_fur_upperleg_r_02_MSTR {parent=thigh_r, loc=<23.3501 4.6555 -7.8406>, rot=<0.9651 0.0045 -0.1282 -0.2284>}
 dyn_fur_upperleg_r_02 {parent=dyn_fur_upperleg_r_02_MSTR, loc=<0.3 0 0>, rot=<0 0 0 -1>}
 dyn_fur_upperleg_r_04_MSTR {parent=thigh_r, loc=<32.0852 -8.8676 4.4581>, rot=<-0.5537 0.1048 0.1459 -0.8131>}
 dyn_fur_upperleg_r_04 {parent=dyn_fur_upperleg_r_04_MSTR, loc=<0.3 0 0>, rot=<0 0 0 1>}
 dyn_fur_upperleg_r_03_MSTR {parent=thigh_r, loc=<36.7709 2.004 9.9183>, rot=<0.0002 0.1434 0.0064 -0.9896>}
 dyn_fur_upperleg_r_03 {parent=dyn_fur_upperleg_r_03_MSTR, loc=<0.2999 0 0>, rot=<0 0 0 1>}
 dyn_scalp_fr_r_01_MSTR {parent=pelvis, loc=<-0.5729 3.5553 17.2439>, rot=<0.4076 0.0506 0.9116 0.0156>}
 dyn_scalp_fr_r_01 {parent=dyn_scalp_fr_r_01_MSTR, loc=<0.3 0 0>, rot=<0 0 0 1>}
 dyn_scalp_fr_r_02 {parent=dyn_scalp_fr_r_01, loc=<3.1109 0 0>, rot=<0 -0.0849 0 -0.9964>}
 dyn_scalp_fr_r_03 {parent=dyn_scalp_fr_r_02, loc=<2.5988 0.0021 -0.4692>, rot=<0.0097 -0.1717 -0.0244 -0.9848>}
 dyn_scalp_fr_r_04 {parent=dyn_scalp_fr_r_03, loc=<10.5458 0 0>, rot=<-0.0002 -0.1175 -0.0017 -0.9931>}
 dyn_scalp_fr_r_05 {parent=dyn_scalp_fr_r_04, loc=<10.8285 0 0>, rot=<0 0 0 1>}
 dyn_scalp_bk_r_01_MSTR {parent=pelvis, loc=<-0.6898 2.006 17.4109>, rot=<0.4489 0.016 0.8911 -0.065>}
 dyn_scalp_bk_r_01 {parent=dyn_scalp_bk_r_01_MSTR, loc=<0.3 0 0>, rot=<0 0 0 1>}
 dyn_scalp_bk_r_02 {parent=dyn_scalp_bk_r_01, loc=<3.4431 0 0>, rot=<0 -0.205 0 -0.9788>}
 dyn_scalp_bk_r_03 {parent=dyn_scalp_bk_r_02, loc=<2.6981 0 0>, rot=<0.0008 -0.1258 0.0068 -0.992>}
 dyn_scalp_bk_r_04 {parent=dyn_scalp_bk_r_03, loc=<10.5458 0 0>, rot=<0.0038 -0.069 0.0329 -0.9971>}
 dyn_scalp_bk_r_05 {parent=dyn_scalp_bk_r_04, loc=<10.8285 0 0>, rot=<0 -0.0291 0 -0.9996>}
 dyn_pouch_bk_r_MSTR {parent=pelvis, loc=<2.4775 -11.9115 13.8409>, rot=<-0.0271 -0.4068 0.9128 0.0259>}
 dyn_pouch_bk_r {parent=dyn_pouch_bk_r_MSTR, loc=<0.3 0 0>, rot=<0 0 0 1>}
 dyn_fur_pelvis_l_01_MSTR {parent=pelvis, loc=<-7.0923 -3.8881 -18.9129>, rot=<-0.0744 -0.9586 0.0058 -0.2749>}
 dyn_fur_pelvis_l_01 {parent=dyn_fur_pelvis_l_01_MSTR, loc=<0.3 0 0>, rot=<0 0 0 1>}
 dyn_fur_pelvis_l_02_MSTR {parent=pelvis, loc=<-1.4892 0.5073 -19.7044>, rot=<-0.0098 -0.9746 -0.016 -0.2232>}
 dyn_fur_pelvis_l_02 {parent=dyn_fur_pelvis_l_02_MSTR, loc=<0.3 0 0>, rot=<0 0 0 1>}
 dyn_fur_pelvis_l_03_MSTR {parent=pelvis, loc=<-0.043 6.2836 -14.7631>, rot=<-0.1019 -0.9161 -0.3282 -0.2065>}
 dyn_fur_pelvis_l_03 {parent=dyn_fur_pelvis_l_03_MSTR, loc=<0.3 0 0>, rot=<0 0 0 1>}
 dyn_fur_pelvis_l_04_MSTR {parent=pelvis, loc=<-2.7093 -14.6383 -6.2306>, rot=<0.054 -0.7828 0.5812 -0.2156>}
 dyn_fur_pelvis_l_04 {parent=dyn_fur_pelvis_l_04_MSTR, loc=<0.3 0 0>, rot=<0 0 0 1>}
 dyn_fur_pelvis_l_05_MSTR {parent=pelvis, loc=<-10.989 11.139 -5.4852>, rot=<-0.0038 -0.7924 -0.6097 -0.0179>}
 dyn_fur_pelvis_l_05 {parent=dyn_fur_pelvis_l_05_MSTR, loc=<0.3 0 0>, rot=<0 0 0 1>}
 dyn_belt_buckle_MSTR {parent=pelvis, loc=<8.9163 10.7095 0>, rot=<0.0127 0.707 0.707 0.0127>}
 dyn_belt_buckle {parent=dyn_belt_buckle_MSTR, loc=<0.3 0 0>, rot=<0 0 0 1>}
 dyn_belt_buckle_target_MSTR {parent=pelvis, loc=<-7.0193 11.2807 0>, rot=<0.0127 0.707 0.707 0.0127>}
 dyn_belt_buckle_target {parent=dyn_belt_buckle_target_MSTR, loc=<0.3 0 0>, rot=<0 0 0 1>}
]
