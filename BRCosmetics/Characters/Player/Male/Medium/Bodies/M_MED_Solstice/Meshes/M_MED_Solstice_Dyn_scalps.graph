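# Physics asset for the Solstice male medium body, simulating two dangling "scalp" chains, front and back, on the character's right side.
# Each chain runs five capsules from a small master anchor constrained to the pelvis, with moderate swing limits and tight twist between links.
# Pelvis, right thigh and right calf capsules, joined by free angular constraints, supply body shapes along that leg.
# Shape names still carry braid labels.
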PhysicsAsset /BRCosmetics/Characters/Player/Male/Medium/Bodies/M_MED_Solstice/Meshes/M_MED_Solstice_Dyn_scalps
bodies {
 calf_r [{shapeType=Capsule, length=32.3578, radius=8.9879, shapeName=dyn_braid_fr_l_MSTR_capsule, center=<20.7065 -0.0001 0>, rotation=<rot 90 0 0>}]
 dyn_scalp_bk_r_01 [{shapeType=Capsule, length=5.7864, radius=3.1631, shapeName=dyn_braid_fr_l_01_capsule, center=<3.4053 -0.0002 3.5>, rotation=<rot 90 0 0>}]
 dyn_scalp_bk_r_01_MSTR [{shapeType=Capsule, length=0.505, radius=0.505, shapeName=dyn_braid_fr_l_MSTR_capsule}]
 dyn_scalp_bk_r_02 [{shapeType=Capsule, length=0.75, radius=3.4131, shapeName=dyn_braid_fr_l_01_capsule, center=<2.165 0 3.4>, rotation=<rot 90 0 0>}]
 dyn_scalp_bk_r_03 [{
   shapeType=Capsule
   length=6.5703
   radius=4.1631
   shapeName=dyn_braid_fr_l_01_capsule
   center=<5.0555 -0.0001 3>
   rotation=<rot 86.4618 0.0001 0.0002>
  }]
 dyn_scalp_bk_r_04 [{shapeType=Capsule, length=7.3481, radius=4.1631, shapeName=dyn_braid_fr_l_01_capsule, center=<5.0497 -0.0002 2>, rotation=<rot 90 0 0>}]
 dyn_scalp_bk_r_05 [{shapeType=Capsule, length=5.144, radius=4.1631, shapeName=dyn_braid_fr_l_01_capsule, center=<3.4397 -0.0002 2.4>, rotation=<rot 90 0 0>}]
 dyn_scalp_fr_r_01 [{shapeType=Capsule, length=5.7864, radius=3.1631, shapeName=dyn_braid_fr_l_01_capsule, center=<3.4053 -0.0002 3.5>, rotation=<rot 90 0 0>}]
 dyn_scalp_fr_r_01_MSTR [{shapeType=Capsule, length=0.505, radius=0.505, shapeName=dyn_braid_fr_l_MSTR_capsule}]
 dyn_scalp_fr_r_02 [{shapeType=Capsule, length=0.75, radius=3.4131, shapeName=dyn_braid_fr_l_01_capsule, center=<2.165 0 3.4>, rotation=<rot 90 0 0>}]
 dyn_scalp_fr_r_03 [{
   shapeType=Capsule
   length=6.5703
   radius=4.1631
   shapeName=dyn_braid_fr_l_01_capsule
   center=<5.0554 -0.0001 3>
   rotation=<rot 86.4479 0.0002 0.0003>
  }]
 dyn_scalp_fr_r_04 [{shapeType=Capsule, length=7.3481, radius=4.1631, shapeName=dyn_braid_fr_l_01_capsule, center=<5.0497 -0.0002 2>, rotation=<rot 90 0 0>}]
 dyn_scalp_fr_r_05 [{shapeType=Capsule, length=5.144, radius=4.1631, shapeName=dyn_braid_fr_l_01_capsule, center=<3.4397 -0.0002 2.4>, rotation=<rot 90 0 0>}]
 pelvis [{
   shapeType=Capsule
   length=3.5038
   radius=7.5592
   shapeName=dyn_braid_fr_l_MSTR_capsule
   center=<-1.9948 -16.6022 17.0491>
   rotation=<rot 90 0 0>
  }]
 thigh_r [{shapeType=Capsule, length=31.0661, radius=13.0548, shapeName=dyn_braid_fr_l_MSTR_capsule, center=<23.5875 -0.0001 0>, rotation=<rot 90 0 0>}]
}
constraints [
 {bone1Name=dyn_scalp_fr_r_01, bone2Name=dyn_scalp_fr_r_01_MSTR, swing1LimitDegrees=30, swing2LimitDegrees=30, twistLimitDegrees=5}
 {bone1Name=dyn_scalp_fr_r_02, bone2Name=dyn_scalp_fr_r_01, swing1LimitDegrees=30, swing2LimitDegrees=30, twistLimitDegrees=5}
 {bone1Name=dyn_scalp_fr_r_03, bone2Name=dyn_scalp_fr_r_02, swing1LimitDegrees=30, swing2LimitDegrees=30, twistLimitDegrees=5}
 {bone1Name=dyn_scalp_fr_r_04, bone2Name=dyn_scalp_fr_r_03, swing1LimitDegrees=30, swing2LimitDegrees=30, twistLimitDegrees=5}
 {bone1Name=dyn_scalp_fr_r_05, bone2Name=dyn_scalp_fr_r_04, swing1LimitDegrees=30, swing2LimitDegrees=30, twistLimitDegrees=5}
 {bone1Name=dyn_scalp_bk_r_01, bone2Name=dyn_scalp_bk_r_01_MSTR, swing1LimitDegrees=30, swing2LimitDegrees=30, twistLimitDegrees=5}
 {bone1Name=dyn_scalp_bk_r_02, bone2Name=dyn_scalp_bk_r_01, swing1LimitDegrees=30, swing2LimitDegrees=30, twistLimitDegrees=5}
 {bone1Name=dyn_scalp_bk_r_03, bone2Name=dyn_scalp_bk_r_02, swing1LimitDegrees=30, swing2LimitDegrees=30, twistLimitDegrees=5}
 {bone1Name=dyn_scalp_bk_r_04, bone2Name=dyn_scalp_bk_r_03, swing1LimitDegrees=30, swing2LimitDegrees=30, twistLimitDegrees=5}
 {bone1Name=dyn_scalp_bk_r_05, bone2Name=dyn_scalp_bk_r_04, swing1LimitDegrees=30, swing2LimitDegrees=30, twistLimitDegrees=5}
 {
  bone1Name=calf_r
  bone2Name=thigh_r
  swing1LimitDegrees=45
  swing1Motion=Free
  swing2LimitDegrees=45
  swing2Motion=Free
  twistLimitDegrees=45
  twistMotion=Free
 }
 {
  bone1Name=thigh_r
  bone2Name=pelvis
  swing1LimitDegrees=45
  swing1Motion=Free
  swing2LimitDegrees=45
  swing2Motion=Free
  twistLimitDegrees=45
  twistMotion=Free
 }
 {
  bone1Name=dyn_scalp_fr_r_01_MSTR
  bone2Name=pelvis
  swing1LimitDegrees=45
  swing1Motion=Free
  swing2LimitDegrees=45
  swing2Motion=Free
  twistLimitDegrees=45
  twistMotion=Free
 }
 {
  bone1Name=dyn_scalp_bk_r_01_MSTR
  bone2Name=pelvis
  swing1LimitDegrees=45
  swing1Motion=Free
  swing2LimitDegrees=45
  swing2Motion=Free
  twistLimitDegrees=45
  twistMotion=Free
 }
]
