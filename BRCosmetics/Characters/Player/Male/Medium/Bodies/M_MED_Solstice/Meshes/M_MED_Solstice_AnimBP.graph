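# Post-process style Animation Blueprint for the Solstice male medium body that adds deformation and secondary motion to a pose copied from the attached parent mesh.
# Three Control Rigs run in sequence first: deform twists, sculpters with knee corrective inputs, then the character's own rig.
# A long chain of weighted multi-bone Constraint nodes then drives master bones for bandolier, belt buckle and fur from nearby body bones.
# A Weighted Look At aims the belt buckle at its target bone.
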AnimBlueprint /BRCosmetics/Characters/Player/Male/Medium/Bodies/M_MED_Solstice/Meshes/M_MED_Solstice_AnimBP
target_skeleton=&/M_MED_Solstice_Skeleton
graphs [
 AnimGraph {nodes [
   n1 = Control Rig(Source=Copy Pose From Mesh() {bUseAttachedParent=true}) {
    alphaBoolBlend {blendOption=Linear}
    controlRigAssetReference {blueprintRigClass=&/M_MED_Solstice_Deform_CtrlRig_twists.M_MED_Solstice_Deform_CtrlRig_twists_C}
   }
   n2 = Control Rig!(
    Source=@n1
    Alpha
    KneeRot_fr_l_114deg:Vector
    KneeTrans_fr_l_114deg:Vector="0.258,-3.591,0.009"
    KneeScale_fr_l_114deg:Vector="1,1,1"
    KneeScale_fr_l_130deg:Vector="1,1,1"
    KneeRot_fr_l_130deg:Vector="0,0,-51.2"
    KneeTrans_fr_l_130deg:Vector="0.258,-3.591,0.009"
   ) -> (Pose) {
    alphaBoolBlend {blendOption=Linear}
    controlRigAssetReference {blueprintRigClass=&/M_MED_Solstice_Deform_CtrlRig_sculpters.M_MED_Solstice_Deform_CtrlRig_sculpters_C}
   }
   n3 = Control Rig!(Source=@n2, Alpha) -> (Pose) {
    alphaBoolBlend {blendOption=Linear}
    controlRigAssetReference {blueprintRigClass=&/M_MED_Solstice_CtrlRig.M_MED_Solstice_CtrlRig_C}
   }
   n4 = "Constraint - dyn_bandolier_fr_top_l_01_MSTR (PPPPPP)"!(
    ComponentPose=Local To Component(LocalPose=@n3)
    Alpha=1
    ConstraintWeights_0=0.4119
    ConstraintWeights_1:float=0.3298
    ConstraintWeights_2:float=0.1376
    ConstraintWeights_3:float=0.0717
    ConstraintWeights_4:float=0.0289
    ConstraintWeights_5:float=0.0202
   ) -> (Pose) {
    alphaBoolBlend {blendOption=Linear}
    boneToModify {boneName=dyn_bandolier_fr_top_l_01_MSTR}
    constraintSetup [
     {targetBone {boneName=spine_04}}
     {targetBone {boneName=spine_05}}
     {targetBone {boneName=deform_pectoral_fr_l}}
     {targetBone {boneName=spine_03}}
     {targetBone {boneName=spine_02}}
     {targetBone {boneName=clavicle_l}}
    ]
    constraintWeights [1,1,1,1,1,1]
   }
   n5 = "Constraint - dyn_bandolier_fr_top_l_02_MSTR (PPPPPP)"!(
    ComponentPose=@n4
    Alpha=1
    ConstraintWeights_0=0.3446
    ConstraintWeights_1:float=0.2725
    ConstraintWeights_2:float=0.2196
    ConstraintWeights_3:float=0.0896
    ConstraintWeights_4:float=0.0573
    ConstraintWeights_5:float=0.0165
   ) -> (Pose) {
    alphaBoolBlend {blendOption=Linear}
    boneToModify {boneName=dyn_bandolier_fr_top_l_02_MSTR}
    constraintSetup [
     {targetBone {boneName=deform_pectoral_fr_l}}
     {targetBone {boneName=spine_04}}
     {targetBone {boneName=spine_05}}
     {targetBone {boneName=deform_deltoid_fr_l}}
     {targetBone {boneName=clavicle_l}}
     {targetBone {boneName=spine_03}}
    ]
    constraintWeights [1,1,1,1,1,1]
   }
   n6 = "Constraint - dyn_bandolier_fr_top_l_03_MSTR (PPPPPP)"!(
    ComponentPose=@n5
    Alpha=1
    ConstraintWeights_0=0.3655
    ConstraintWeights_1:float=0.3653
    ConstraintWeights_2:float=0.0854
    ConstraintWeights_3:float=0.0797
    ConstraintWeights_4:float=0.0542
    ConstraintWeights_5:float=0.0499
   ) -> (Pose) {
    alphaBoolBlend {blendOption=Linear}
    boneToModify {boneName=dyn_bandolier_fr_top_l_03_MSTR}
    constraintSetup [
     {targetBone {boneName=deform_pectoral_fr_l}}
     {targetBone {boneName=deform_deltoid_fr_l}}
     {targetBone {boneName=upperarm_twist_01_l}}
     {targetBone {boneName=spine_05}}
     {targetBone {boneName=spine_04}}
     {targetBone {boneName=clavicle_l}}
    ]
    constraintWeights [1,1,1,1,1,1]
   }
   n7 = "Constraint - dyn_bandolier_fr_top_r_01_MSTR (PPPPPP)"!(
    ComponentPose=@n6
    Alpha=1
    ConstraintWeights_0=0.4119
    ConstraintWeights_1:float=0.3298
    ConstraintWeights_2:float=0.1376
    ConstraintWeights_3:float=0.0717
    ConstraintWeights_4:float=0.0289
    ConstraintWeights_5:float=0.0202
   ) -> (Pose) {
    alphaBoolBlend {blendOption=Linear}
    boneToModify {boneName=dyn_bandolier_fr_top_r_01_MSTR}
    constraintSetup [
     {targetBone {boneName=spine_04}}
     {targetBone {boneName=spine_05}}
     {targetBone {boneName=deform_pectoral_fr_r}}
     {targetBone {boneName=spine_03}}
     {targetBone {boneName=spine_02}}
     {targetBone {boneName=clavicle_r}}
    ]
    constraintWeights [1,1,1,1,1,1]
   }
   n8 = "Constraint - dyn_bandolier_fr_top_r_02_MSTR (PPPPPP)"!(
    ComponentPose=@n7
    Alpha=1
    ConstraintWeights_0=0.3486
    ConstraintWeights_1:float=0.2646
    ConstraintWeights_2:float=0.2158
    ConstraintWeights_3:float=0.0964
    ConstraintWeights_4:float=0.0593
    ConstraintWeights_5:float=0.0154
   ) -> (Pose) {
    alphaBoolBlend {blendOption=Linear}
    boneToModify {boneName=dyn_bandolier_fr_top_r_02_MSTR}
    constraintSetup [
     {targetBone {boneName=deform_pectoral_fr_r}}
     {targetBone {boneName=spine_04}}
     {targetBone {boneName=spine_05}}
     {targetBone {boneName=deform_deltoid_fr_r}}
     {targetBone {boneName=clavicle_r}}
     {targetBone {boneName=spine_03}}
    ]
    constraintWeights [1,1,1,1,1,1]
   }
   n9 = "Constraint - dyn_bandolier_fr_top_r_03_MSTR (PPPPPP)"!(
    ComponentPose=@n8
    Alpha=1
    ConstraintWeights_0=0.3658
    ConstraintWeights_1:float=0.3563
    ConstraintWeights_2:float=0.0839
    ConstraintWeights_3:float=0.083
    ConstraintWeights_4:float=0.0596
    ConstraintWeights_5:float=0.0515
   ) -> (Pose) {
    alphaBoolBlend {blendOption=Linear}
    boneToModify {boneName=dyn_bandolier_fr_top_r_03_MSTR}
    constraintSetup [
     {targetBone {boneName=deform_pectoral_fr_r}}
     {targetBone {boneName=deform_deltoid_fr_r}}
     {targetBone {boneName=spine_05}}
     {targetBone {boneName=upperarm_twist_01_r}}
     {targetBone {boneName=spine_04}}
     {targetBone {boneName=clavicle_r}}
    ]
    constraintWeights [1,1,1,1,1,1]
   }
   n10 = "Constraint - dyn_bandolier_bk_top_r_01_MSTR (PPPPPP)"!(
    ComponentPose=@n9
    Alpha=1
    ConstraintWeights_0=0.3472
    ConstraintWeights_1:float=0.2974
    ConstraintWeights_2:float=0.2163
    ConstraintWeights_3:float=0.0857
    ConstraintWeights_4:float=0.0404
    ConstraintWeights_5:float=0.013
   ) -> (Pose) {
    alphaBoolBlend {blendOption=Linear}
    boneToModify {boneName=dyn_bandolier_bk_top_r_01_MSTR}
    constraintSetup [
     {targetBone {boneName=spine_05}}
     {targetBone {boneName=spine_04}}
     {targetBone {boneName=clavicle_r}}
     {targetBone {boneName=spine_03}}
     {targetBone {boneName=spine_02}}
     {targetBone {boneName=deform_pectoral_fr_r}}
    ]
    constraintWeights [1,1,1,1,1,1]
   }
   n11 = "Constraint - dyn_bandolier_bk_top_r_02_MSTR (PPPPPPP)"!(
    ComponentPose=@n10
    Alpha=1
    ConstraintWeights_0=0.3043
    ConstraintWeights_1:float=0.2632
    ConstraintWeights_2:float=0.2118
    ConstraintWeights_3:float=0.0837
    ConstraintWeights_4:float=0.0745
    ConstraintWeights_5:float=0.0476
    ConstraintWeights_6:float=0.0148
   ) -> (Pose) {
    alphaBoolBlend {blendOption=Linear}
    boneToModify {boneName=dyn_bandolier_bk_top_r_02_MSTR}
    constraintSetup [
     {targetBone {boneName=spine_05}}
     {targetBone {boneName=clavicle_r}}
     {targetBone {boneName=spine_04}}
     {targetBone {boneName=deform_deltoid_fr_r}}
     {targetBone {boneName=deform_pectoral_fr_r}}
     {targetBone {boneName=upperarm_twist_01_r}}
     {targetBone {boneName=spine_03}}
    ]
    constraintWeights [1,1,1,1,1,1,1]
   }
   n12 = "Constraint - dyn_bandolier_bk_top_r_03_MSTR (PPPPPP)"!(
    ComponentPose=@n11
    Alpha=1
    ConstraintWeights_0=0.2255
    ConstraintWeights_1:float=0.2134
    ConstraintWeights_2:float=0.1749
    ConstraintWeights_3:float=0.1467
    ConstraintWeights_4:float=0.121
    ConstraintWeights_5:float=0.1185
   ) -> (Pose) {
    alphaBoolBlend {blendOption=Linear}
    boneToModify {boneName=dyn_bandolier_bk_top_r_03_MSTR}
    constraintSetup [
     {targetBone {boneName=clavicle_r}}
     {targetBone {boneName=spine_05}}
     {targetBone {boneName=deform_deltoid_fr_r}}
     {targetBone {boneName=spine_04}}
     {targetBone {boneName=upperarm_twist_01_r}}
     {targetBone {boneName=deform_pectoral_fr_r}}
    ]
    constraintWeights [1,1,1,1,1,1]
   }
   n13 = "Constraint - dyn_bandolier_bk_top_l_01_MSTR (PPPPPP)"!(
    ComponentPose=@n12
    Alpha=1
    ConstraintWeights_0=0.4185
    ConstraintWeights_1:float=0.2881
    ConstraintWeights_2:float=0.188
    ConstraintWeights_3:float=0.0654
    ConstraintWeights_4:float=0.0299
    ConstraintWeights_5:float=0.0102
   ) -> (Pose) {
    alphaBoolBlend {blendOption=Linear}
    boneToModify {boneName=dyn_bandolier_bk_top_l_01_MSTR}
    constraintSetup [
     {targetBone {boneName=spine_05}}
     {targetBone {boneName=spine_04}}
     {targetBone {boneName=clavicle_l}}
     {targetBone {boneName=spine_03}}
     {targetBone {boneName=spine_02}}
     {targetBone {boneName=deform_pectoral_fr_l}}
    ]
    constraintWeights [1,1,1,1,1,1]
   }
   n14 = "Constraint - dyn_bandolier_bk_top_l_02_MSTR (PPPPPPPP)"!(
    ComponentPose=@n13
    Alpha=1
    ConstraintWeights_0=0.3572
    ConstraintWeights_1:float=0.2575
    ConstraintWeights_2:float=0.2572
    ConstraintWeights_3:float=0.0423
    ConstraintWeights_4:float=0.0322
    ConstraintWeights_5:float=0.0293
    ConstraintWeights_6:float=0.0142
    ConstraintWeights_7:float=0.0101
   ) -> (Pose) {
    alphaBoolBlend {blendOption=Linear}
    boneToModify {boneName=dyn_bandolier_bk_top_l_02_MSTR}
    constraintSetup [
     {targetBone {boneName=spine_05}}
     {targetBone {boneName=spine_04}}
     {targetBone {boneName=clavicle_l}}
     {targetBone {boneName=deform_pectoral_fr_l}}
     {targetBone {boneName=upperarm_twist_01_l}}
     {targetBone {boneName=spine_03}}
     {targetBone {boneName=deform_deltoid_fr_l}}
     {targetBone {boneName=spine_02}}
    ]
    constraintWeights [1,1,1,1,1,1,1,1]
   }
   n15 = "Constraint - dyn_bandolier_bk_top_l_03_MSTR (PPPPPP)"!(
    ComponentPose=@n14
    Alpha=1
    ConstraintWeights_0=0.3016
    ConstraintWeights_1:float=0.1815
    ConstraintWeights_2:float=0.1605
    ConstraintWeights_3:float=0.1285
    ConstraintWeights_4:float=0.1247
    ConstraintWeights_5:float=0.1032
   ) -> (Pose) {
    alphaBoolBlend {blendOption=Linear}
    boneToModify {boneName=dyn_bandolier_bk_top_l_03_MSTR}
    constraintSetup [
     {targetBone {boneName=clavicle_l}}
     {targetBone {boneName=spine_05}}
     {targetBone {boneName=deform_deltoid_fr_l}}
     {targetBone {boneName=upperarm_twist_01_l}}
     {targetBone {boneName=spine_04}}
     {targetBone {boneName=deform_pectoral_fr_l}}
    ]
    constraintWeights [1,1,1,1,1,1]
   }
   n16 = "Constraint - dyn_bandolier_fr_mid_MSTR (PPPP)"!(
    ComponentPose=@n15
    Alpha=1
    ConstraintWeights_0=0.4195
    ConstraintWeights_1:float=0.2453
    ConstraintWeights_2:float=0.1974
    ConstraintWeights_3:float=0.1377
   ) -> (Pose) {
    alphaBoolBlend {blendOption=Linear}
    boneToModify {boneName=dyn_bandolier_fr_mid_MSTR}
    constraintSetup [{targetBone {boneName=spine_04}}, {targetBone {boneName=spine_05}}, {targetBone {boneName=spine_03}}, {targetBone {boneName=spine_02}}]
    constraintWeights [1,1,1,1]
   }
   n17 = "Constraint - dyn_bandolier_bk_mid_MSTR (PPPPP)"!(
    ComponentPose=@n16
    Alpha=1
    ConstraintWeights_0=0.365
    ConstraintWeights_1:float=0.3216
    ConstraintWeights_2:float=0.1729
    ConstraintWeights_3:float=0.1109
    ConstraintWeights_4:float=0.0296
   ) -> (Pose) {
    alphaBoolBlend {blendOption=Linear}
    boneToModify {boneName=dyn_bandolier_bk_mid_MSTR}
    constraintSetup [
     {targetBone {boneName=spine_05}}
     {targetBone {boneName=spine_04}}
     {targetBone {boneName=spine_03}}
     {targetBone {boneName=spine_02}}
     {targetBone {boneName=spine_01}}
    ]
    constraintWeights [1,1,1,1,1]
   }
   n18 = "Constraint - dyn_belt_buckle_MSTR (PPPP)"!(
    ComponentPose=@n17
    Alpha=1
    ConstraintWeights_0=0.7924
    ConstraintWeights_1:float=0.1644
    ConstraintWeights_2:float=0.0281
    ConstraintWeights_3:float=0.0151
   ) -> (Pose) {
    alphaBoolBlend {blendOption=Linear}
    boneToModify {boneName=dyn_belt_buckle_MSTR}
    constraintSetup [{targetBone {boneName=pelvis}}, {targetBone {boneName=spine_01}}, {targetBone {boneName=spine_02}}, {targetBone {boneName=spine_03}}]
    constraintWeights [1,1,1,1]
   }
   n19 = "Constraint - dyn_belt_buckle_target_MSTR (PPPPP)"!(
    ComponentPose=@n18
    Alpha=1
    ConstraintWeights_0=0.5442
    ConstraintWeights_1:float=0.1794
    ConstraintWeights_2:float=0.1642
    ConstraintWeights_3:float=0.0618
    ConstraintWeights_4:float=0.0504
   ) -> (Pose) {
    alphaBoolBlend {blendOption=Linear}
    boneToModify {boneName=dyn_belt_buckle_target_MSTR}
    constraintSetup [
     {targetBone {boneName=pelvis}}
     {targetBone {boneName=thigh_twist_01_r}}
     {targetBone {boneName=thigh_twist_01_l}}
     {targetBone {boneName=deform_thigh_fr_r}}
     {targetBone {boneName=deform_thigh_fr_l}}
    ]
    constraintWeights [1,1,1,1,1]
   }
   n20 = Weighted Look At - Bone: dyn_belt_buckle_MSTR!(ComponentPose:Component Space Pose Link Structure=@n19, Alpha:float=1) -> (Pose:Component Space Pose Link Structure) {
    is=title FortAnimGraphNode_WeightedLookAt
    alphaBoolBlend {blendOption=Linear}
    boneToModify {boneName=dyn_belt_buckle_MSTR}
    lookAtLocation {x=0}
    lookAtTarget {boneReference {boneName=dyn_belt_buckle_target}}
    lookAt_Axis {axis=<1 0>}
    lookUp_Axis {axis {x=0, z=1}}
   }
   n21 = "Constraint - dyn_fur_lowerarm_r_09_MSTR (PPP)"!(ComponentPose=@n20, Alpha=1, ConstraintWeights_0=0.8708, ConstraintWeights_1:float=0.0695, ConstraintWeights_2:float=0.0598) -> (Pose) {
    alphaBoolBlend {blendOption=Linear}
    boneToModify {boneName=dyn_fur_lowerarm_r_09_MSTR}
    constraintSetup [{targetBone {boneName=lowerarm_twist_01_r}}, {targetBone {boneName=pinky_metacarpal_r}}, {targetBone {boneName=lowerarm_twist_02_r}}]
    constraintWeights [1,1,1]
   }
   n22 = "Constraint - dyn_fur_lowerarm_r_08_MSTR (PP)"!(ComponentPose=@n21, Alpha=1, ConstraintWeights_0=0.7476, ConstraintWeights_1:float=0.2524) -> (Pose) {
    alphaBoolBlend {blendOption=Linear}
    boneToModify {boneName=dyn_fur_lowerarm_r_08_MSTR}
    constraintSetup [{targetBone {boneName=lowerarm_twist_01_r}}, {targetBone {boneName=lowerarm_twist_02_r}}]
    constraintWeights [1,1]
   }
   n23 = "Constraint - dyn_fur_lowerarm_r_07_MSTR (PPPPP)"!(
    ComponentPose=@n22
    Alpha=1
    ConstraintWeights_0=0.4886
    ConstraintWeights_1:float=0.2718
    ConstraintWeights_2:float=0.1755
    ConstraintWeights_3:float=0.0326
    ConstraintWeights_4:float=0.0316
   ) -> (Pose) {
    alphaBoolBlend {blendOption=Linear}
    boneToModify {boneName=dyn_fur_lowerarm_r_07_MSTR}
    constraintSetup [
     {targetBone {boneName=lowerarm_twist_01_r}}
     {targetBone {boneName=dyn_fur_lowerarm_r_07}}
     {targetBone {boneName=index_metacarpal_r}}
     {targetBone {boneName=middle_metacarpal_r}}
     {targetBone {boneName=ring_metacarpal_r}}
    ]
    constraintWeights [1,1,1,1,1]
   }
   n24 = "Constraint - dyn_fur_lowerarm_r_06_MSTR (PPPP)"!(
    ComponentPose=@n23
    Alpha=1
    ConstraintWeights_0=0.539
    ConstraintWeights_1:float=0.2435
    ConstraintWeights_2:float=0.163
    ConstraintWeights_3:float=0.0545
   ) -> (Pose) {
    alphaBoolBlend {blendOption=Linear}
    boneToModify {boneName=dyn_fur_lowerarm_r_06_MSTR}
    constraintSetup [
     {targetBone {boneName=lowerarm_twist_01_r}}
     {targetBone {boneName=dyn_fur_lowerarm_r_06}}
     {targetBone {boneName=lowerarm_twist_02_r}}
     {targetBone {boneName=pinky_metacarpal_r}}
    ]
    constraintWeights [1,1,1,1]
   }
   n25 = "Constraint - dyn_fur_lowerarm_r_05_MSTR (PPP)"!(ComponentPose=@n24, Alpha=1, ConstraintWeights_0=0.6761, ConstraintWeights_1:float=0.2898, ConstraintWeights_2:float=0.0341) -> (Pose) {
    alphaBoolBlend {blendOption=Linear}
    boneToModify {boneName=dyn_fur_lowerarm_r_05_MSTR}
    constraintSetup [{targetBone {boneName=lowerarm_twist_01_r}}, {targetBone {boneName=lowerarm_twist_02_r}}, {targetBone {boneName=dyn_fur_lowerarm_r_05}}]
    constraintWeights [1,1,1]
   }
   n26 = "Constraint - dyn_fur_lowerarm_r_04_MSTR (PPPPPP)"!(
    ComponentPose=@n25
    Alpha=1
    ConstraintWeights_0=0.3693
    ConstraintWeights_1:float=0.2986
    ConstraintWeights_2:float=0.2611
    ConstraintWeights_3:float=0.0418
    ConstraintWeights_4:float=0.0163
    ConstraintWeights_5:float=0.0129
   ) -> (Pose) {
    alphaBoolBlend {blendOption=Linear}
    boneToModify {boneName=dyn_fur_lowerarm_r_04_MSTR}
    constraintSetup [
     {targetBone {boneName=lowerarm_twist_02_r}}
     {targetBone {boneName=dyn_fur_lowerarm_r_04}}
     {targetBone {boneName=lowerarm_twist_01_r}}
     {targetBone {boneName=lowerarm_r}}
     {targetBone {boneName=deform_elbow_bk_r}}
     {targetBone {boneName=deform_elbow_fr_r}}
    ]
    constraintWeights [1,1,1,1,1,1]
   }
   n27 = "Constraint - dyn_fur_lowerarm_r_03_MSTR (PPPPP)"!(
    ComponentPose=@n26
    Alpha=1
    ConstraintWeights_0=0.4941
    ConstraintWeights_1:float=0.1666
    ConstraintWeights_2:float=0.1495
    ConstraintWeights_3:float=0.1249
    ConstraintWeights_4:float=0.0649
   ) -> (Pose) {
    alphaBoolBlend {blendOption=Linear}
    boneToModify {boneName=dyn_fur_lowerarm_r_03_MSTR}
    constraintSetup [
     {targetBone {boneName=deform_elbow_bk_r}}
     {targetBone {boneName=upperarm_r}}
     {targetBone {boneName=dyn_fur_lowerarm_r_03}}
     {targetBone {boneName=lowerarm_twist_02_r}}
     {targetBone {boneName=lowerarm_r}}
    ]
    constraintWeights [1,1,1,1,1]
   }
   n28 = "Constraint - dyn_fur_lowerarm_r_01_MSTR (PPPPPP)"!(
    ComponentPose=@n27
    Alpha=1
    ConstraintWeights_0=0.3095
    ConstraintWeights_1:float=0.2026
    ConstraintWeights_2:float=0.191
    ConstraintWeights_3:float=0.1446
    ConstraintWeights_4:float=0.0863
    ConstraintWeights_5:float=0.066
   ) -> (Pose) {
    alphaBoolBlend {blendOption=Linear}
    boneToModify {boneName=dyn_fur_lowerarm_r_01_MSTR}
    constraintSetup [
     {targetBone {boneName=upperarm_r}}
     {targetBone {boneName=deform_elbow_fr_r}}
     {targetBone {boneName=deform_elbow_bk_r}}
     {targetBone {boneName=lowerarm_twist_02_r}}
     {targetBone {boneName=upperarm_twist_02_r}}
     {targetBone {boneName=lowerarm_r}}
    ]
    constraintWeights [1,1,1,1,1,1]
   }
   n29 = "Constraint - dyn_fur_upperarm_r_02_MSTR (PPPPPPP)"!(
    ComponentPose=@n28
    Alpha=1
    ConstraintWeights_0=0.2927
    ConstraintWeights_1:float=0.2458
    ConstraintWeights_2:float=0.2322
    ConstraintWeights_3:float=0.1405
    ConstraintWeights_4:float=0.0541
    ConstraintWeights_5:float=0.0239
    ConstraintWeights_6:float=0.0108
   ) -> (Pose) {
    alphaBoolBlend {blendOption=Linear}
    boneToModify {boneName=dyn_fur_upperarm_r_02_MSTR}
    constraintSetup [
     {targetBone {boneName=upperarm_r}}
     {targetBone {boneName=deform_elbow_fr_r}}
     {targetBone {boneName=upperarm_twist_02_r}}
     {targetBone {boneName=upperarm_twist_01_r}}
     {targetBone {boneName=lowerarm_twist_02_r}}
     {targetBone {boneName=lowerarm_r}}
     {targetBone {boneName=dyn_fur_upperarm_r_02}}
    ]
    constraintWeights [1,1,1,1,1,1,1]
   }
   n30 = "Constraint - dyn_fur_upperarm_r_01_MSTR (PPPPPP)"!(
    ComponentPose=@n29
    Alpha=1
    ConstraintWeights_0=0.4159
    ConstraintWeights_1:float=0.2989
    ConstraintWeights_2:float=0.1083
    ConstraintWeights_3:float=0.0701
    ConstraintWeights_4:float=0.0552
    ConstraintWeights_5:float=0.0517
   ) -> (Pose) {
    alphaBoolBlend {blendOption=Linear}
    boneToModify {boneName=dyn_fur_upperarm_r_01_MSTR}
    constraintSetup [
     {targetBone {boneName=deform_deltoid_fr_r}}
     {targetBone {boneName=upperarm_twist_01_r}}
     {targetBone {boneName=dyn_fur_upperarm_r_01}}
     {targetBone {boneName=upperarm_twist_02_r}}
     {targetBone {boneName=upperarm_r}}
     {targetBone {boneName=deform_pectoral_fr_r}}
    ]
    constraintWeights [1,1,1,1,1,1]
   }
   n31 = "Constraint - dyn_fur_lowerarm_l_10_MSTR (PPP)"!(ComponentPose=@n30, Alpha=1, ConstraintWeights_0=0.6435, ConstraintWeights_1:float=0.1954, ConstraintWeights_2:float=0.1611) -> (Pose) {
    alphaBoolBlend {blendOption=Linear}
    boneToModify {boneName=dyn_fur_lowerarm_l_10_MSTR}
    constraintSetup [{targetBone {boneName=thumb_01_l}}, {targetBone {boneName=lowerarm_twist_01_l}}, {targetBone {boneName=hand_l}}]
    constraintWeights [1,1,1]
   }
   n32 = "Constraint - dyn_fur_lowerarm_l_09_MSTR (PPP)"!(ComponentPose=@n31, Alpha=1, ConstraintWeights_0=0.7382, ConstraintWeights_1:float=0.1321, ConstraintWeights_2:float=0.1296) -> (Pose) {
    alphaBoolBlend {blendOption=Linear}
    boneToModify {boneName=dyn_fur_lowerarm_l_09_MSTR}
    constraintSetup [{targetBone {boneName=lowerarm_twist_01_l}}, {targetBone {boneName=lowerarm_twist_02_l}}, {targetBone {boneName=pinky_metacarpal_l}}]
    constraintWeights [1,1,1]
   }
   n33 = "Constraint - dyn_fur_lowerarm_l_08_MSTR (PPP)"!(ComponentPose=@n32, Alpha=1, ConstraintWeights_0=0.61, ConstraintWeights_1:float=0.298, ConstraintWeights_2:float=0.092) -> (Pose) {
    alphaBoolBlend {blendOption=Linear}
    boneToModify {boneName=dyn_fur_lowerarm_l_08_MSTR}
    constraintSetup [{targetBone {boneName=lowerarm_twist_01_l}}, {targetBone {boneName=lowerarm_twist_02_l}}, {targetBone {boneName=thumb_01_l}}]
    constraintWeights [1,1,1]
   }
   n34 = "Constraint - dyn_fur_lowerarm_l_07_MSTR (PPPPP)"!(
    ComponentPose=@n33
    Alpha=1
    ConstraintWeights_0=0.5075
    ConstraintWeights_1:float=0.2436
    ConstraintWeights_2:float=0.1823
    ConstraintWeights_3:float=0.0338
    ConstraintWeights_4:float=0.0328
   ) -> (Pose) {
    alphaBoolBlend {blendOption=Linear}
    boneToModify {boneName=dyn_fur_lowerarm_l_07_MSTR}
    constraintSetup [
     {targetBone {boneName=lowerarm_twist_01_l}}
     {targetBone {boneName=dyn_fur_lowerarm_l_07}}
     {targetBone {boneName=index_metacarpal_l}}
     {targetBone {boneName=middle_metacarpal_l}}
     {targetBone {boneName=ring_metacarpal_l}}
    ]
    constraintWeights [1,1,1,1,1]
   }
   n35 = "Constraint - dyn_fur_lowerarm_l_06_MSTR (PPP)"!(ComponentPose=@n34, Alpha=1, ConstraintWeights_0=0.5524, ConstraintWeights_1:float=0.2606, ConstraintWeights_2:float=0.187) -> (Pose) {
    alphaBoolBlend {blendOption=Linear}
    boneToModify {boneName=dyn_fur_lowerarm_l_06_MSTR}
    constraintSetup [{targetBone {boneName=lowerarm_twist_01_l}}, {targetBone {boneName=dyn_fur_lowerarm_l_06}}, {targetBone {boneName=lowerarm_twist_02_l}}]
    constraintWeights [1,1,1]
   }
   n36 = "Constraint - dyn_fur_lowerarm_l_05_MSTR (PPP)"!(ComponentPose=@n35, Alpha=1, ConstraintWeights_0=0.6547, ConstraintWeights_1:float=0.2807, ConstraintWeights_2:float=0.0645) -> (Pose) {
    alphaBoolBlend {blendOption=Linear}
    boneToModify {boneName=dyn_fur_lowerarm_l_05_MSTR}
    constraintSetup [{targetBone {boneName=lowerarm_twist_01_l}}, {targetBone {boneName=lowerarm_twist_02_l}}, {targetBone {boneName=dyn_fur_lowerarm_l_05}}]
    constraintWeights [1,1,1]
   }
   n37 = "Constraint - dyn_fur_lowerarm_l_04_MSTR (PPPPP)"!(
    ComponentPose=@n36
    Alpha=1
    ConstraintWeights_0=0.4088
    ConstraintWeights_1:float=0.3585
    ConstraintWeights_2:float=0.1753
    ConstraintWeights_3:float=0.0346
    ConstraintWeights_4:float=0.0228
   ) -> (Pose) {
    alphaBoolBlend {blendOption=Linear}
    boneToModify {boneName=dyn_fur_lowerarm_l_04_MSTR}
    constraintSetup [
     {targetBone {boneName=lowerarm_twist_02_l}}
     {targetBone {boneName=lowerarm_twist_01_l}}
     {targetBone {boneName=dyn_fur_lowerarm_l_04}}
     {targetBone {boneName=lowerarm_l}}
     {targetBone {boneName=deform_elbow_bk_l}}
    ]
    constraintWeights [1,1,1,1,1]
   }
   n38 = "Constraint - dyn_fur_lowerarm_l_03_MSTR (PPPPP)"!(
    ComponentPose=@n37
    Alpha=1
    ConstraintWeights_0=0.4744
    ConstraintWeights_1:float=0.2873
    ConstraintWeights_2:float=0.1458
    ConstraintWeights_3:float=0.0624
    ConstraintWeights_4:float=0.0301
   ) -> (Pose) {
    alphaBoolBlend {blendOption=Linear}
    boneToModify {boneName=dyn_fur_lowerarm_l_03_MSTR}
    constraintSetup [
     {targetBone {boneName=deform_elbow_bk_l}}
     {targetBone {boneName=dyn_fur_lowerarm_l_03}}
     {targetBone {boneName=upperarm_l}}
     {targetBone {boneName=lowerarm_twist_02_l}}
     {targetBone {boneName=lowerarm_l}}
    ]
    constraintWeights [1,1,1,1,1]
   }
   n39 = "Constraint - dyn_fur_lowerarm_l_02_MSTR (PPPPPPP)"!(
    ComponentPose=@n38
    Alpha=1
    ConstraintWeights_0=0.2641
    ConstraintWeights_1:float=0.2131
    ConstraintWeights_2:float=0.1354
    ConstraintWeights_3:float=0.1226
    ConstraintWeights_4:float=0.1212
    ConstraintWeights_5:float=0.0937
    ConstraintWeights_6:float=0.0498
   ) -> (Pose) {
    alphaBoolBlend {blendOption=Linear}
    boneToModify {boneName=dyn_fur_lowerarm_l_02_MSTR}
    constraintSetup [
     {targetBone {boneName=dyn_fur_lowerarm_l_02}}
     {targetBone {boneName=upperarm_l}}
     {targetBone {boneName=deform_elbow_bk_l}}
     {targetBone {boneName=deform_elbow_fr_l}}
     {targetBone {boneName=lowerarm_twist_02_l}}
     {targetBone {boneName=lowerarm_l}}
     {targetBone {boneName=upperarm_twist_02_l}}
    ]
    constraintWeights [1,1,1,1,1,1,1]
   }
   n40 = "Constraint - dyn_fur_lowerarm_l_01_MSTR (PPPPPP)"!(
    ComponentPose=@n39
    Alpha=1
    ConstraintWeights_0=0.3095
    ConstraintWeights_1:float=0.2026
    ConstraintWeights_2:float=0.191
    ConstraintWeights_3:float=0.1446
    ConstraintWeights_4:float=0.0863
    ConstraintWeights_5:float=0.066
   ) -> (Pose) {
    alphaBoolBlend {blendOption=Linear}
    boneToModify {boneName=dyn_fur_lowerarm_l_01_MSTR}
    constraintSetup [
     {targetBone {boneName=upperarm_l}}
     {targetBone {boneName=deform_elbow_fr_l}}
     {targetBone {boneName=deform_elbow_bk_l}}
     {targetBone {boneName=lowerarm_twist_02_l}}
     {targetBone {boneName=upperarm_twist_02_l}}
     {targetBone {boneName=lowerarm_l}}
    ]
    constraintWeights [1,1,1,1,1,1]
   }
   n41 = "Constraint - dyn_fur_upperarm_l_02_MSTR (PPPPPPP)"!(
    ComponentPose=@n40
    Alpha=1
    ConstraintWeights_0=0.252
    ConstraintWeights_1:float=0.2409
    ConstraintWeights_2:float=0.1772
    ConstraintWeights_3:float=0.1178
    ConstraintWeights_4:float=0.102
    ConstraintWeights_5:float=0.063
    ConstraintWeights_6:float=0.0471
   ) -> (Pose) {
    alphaBoolBlend {blendOption=Linear}
    boneToModify {boneName=dyn_fur_upperarm_l_02_MSTR}
    constraintSetup [
     {targetBone {boneName=upperarm_l}}
     {targetBone {boneName=deform_elbow_fr_l}}
     {targetBone {boneName=dyn_fur_upperarm_l_02}}
     {targetBone {boneName=upperarm_twist_02_l}}
     {targetBone {boneName=lowerarm_twist_02_l}}
     {targetBone {boneName=upperarm_twist_01_l}}
     {targetBone {boneName=lowerarm_l}}
    ]
    constraintWeights [1,1,1,1,1,1,1]
   }
   n42 = "Constraint - dyn_fur_upperarm_l_01_MSTR (PPPP)"!(
    ComponentPose=@n41
    Alpha=1
    ConstraintWeights_0=0.4535
    ConstraintWeights_1:float=0.4285
    ConstraintWeights_2:float=0.0647
    ConstraintWeights_3:float=0.0534
   ) -> (Pose) {
    alphaBoolBlend {blendOption=Linear}
    boneToModify {boneName=dyn_fur_upperarm_l_01_MSTR}
    constraintSetup [
     {targetBone {boneName=upperarm_twist_01_l}}
     {targetBone {boneName=deform_deltoid_fr_l}}
     {targetBone {boneName=upperarm_twist_02_l}}
     {targetBone {boneName=upperarm_l}}
    ]
    constraintWeights [1,1,1,1]
   }
   n43 = "Constraint - dyn_fur_torso_r_02_MSTR (PPPPPP)"!(
    ComponentPose=@n42
    Alpha=1
    ConstraintWeights_0=0.2968
    ConstraintWeights_1:float=0.2853
    ConstraintWeights_2:float=0.163
    ConstraintWeights_3:float=0.1605
    ConstraintWeights_4:float=0.0707
    ConstraintWeights_5:float=0.0238
   ) -> (Pose) {
    alphaBoolBlend {blendOption=Linear}
    boneToModify {boneName=dyn_fur_torso_r_02_MSTR}
    constraintSetup [
     {targetBone {boneName=spine_04}}
     {targetBone {boneName=dyn_fur_torso_r_02}}
     {targetBone {boneName=spine_03}}
     {targetBone {boneName=spine_02}}
     {targetBone {boneName=spine_01}}
     {targetBone {boneName=spine_05}}
    ]
    constraintWeights [1,1,1,1,1,1]
   }
   n44 = "Constraint - dyn_fur_torso_r_01_MSTR (PPPPPPP)"!(
    ComponentPose=@n43
    Alpha=1
    ConstraintWeights_0=0.4726
    ConstraintWeights_1:float=0.1829
    ConstraintWeights_2:float=0.0949
    ConstraintWeights_3:float=0.0853
    ConstraintWeights_4:float=0.071
    ConstraintWeights_5:float=0.0684
    ConstraintWeights_6:float=0.0249
   ) -> (Pose) {
    alphaBoolBlend {blendOption=Linear}
    boneToModify {boneName=dyn_fur_torso_r_01_MSTR}
    constraintSetup [
     {targetBone {boneName=dyn_fur_torso_r_01}}
     {targetBone {boneName=spine_02}}
     {targetBone {boneName=spine_03}}
     {targetBone {boneName=pelvis}}
     {targetBone {boneName=spine_01}}
     {targetBone {boneName=spine_04}}
     {targetBone {boneName=spine_05}}
    ]
    constraintWeights [1,1,1,1,1,1,1]
   }
   n45 = "Constraint - dyn_fur_torso_l_02_MSTR (PPPPPP)"!(
    ComponentPose=@n44
    Alpha=1
    ConstraintWeights_0=0.5124
    ConstraintWeights_1:float=0.1383
    ConstraintWeights_2:float=0.1353
    ConstraintWeights_3:float=0.1119
    ConstraintWeights_4:float=0.0875
    ConstraintWeights_5:float=0.0147
   ) -> (Pose) {
    alphaBoolBlend {blendOption=Linear}
    boneToModify {boneName=dyn_fur_torso_l_02_MSTR}
    constraintSetup [
     {targetBone {boneName=spine_04}}
     {targetBone {boneName=spine_03}}
     {targetBone {boneName=spine_05}}
     {targetBone {boneName=spine_02}}
     {targetBone {boneName=deform_pectoral_fr_l}}
     {targetBone {boneName=upperarm_twist_02_l}}
    ]
    constraintWeights [1,1,1,1,1,1]
   }
   n46 = "Constraint - dyn_fur_torso_l_01_MSTR (PPPPPP)"!(
    ComponentPose=@n45
    Alpha=1
    ConstraintWeights_0=0.3265
    ConstraintWeights_1:float=0.2121
    ConstraintWeights_2:float=0.1948
    ConstraintWeights_3:float=0.1732
    ConstraintWeights_4:float=0.0682
    ConstraintWeights_5:float=0.0251
   ) -> (Pose) {
    alphaBoolBlend {blendOption=Linear}
    boneToModify {boneName=dyn_fur_torso_l_01_MSTR}
    constraintSetup [
     {targetBone {boneName=spine_04}}
     {targetBone {boneName=spine_02}}
     {targetBone {boneName=dyn_fur_torso_l_01}}
     {targetBone {boneName=spine_03}}
     {targetBone {boneName=spine_01}}
     {targetBone {boneName=spine_05}}
    ]
    constraintWeights [1,1,1,1,1,1]
   }
   n47 = "Constraint - dyn_fur_torso_fr_04_MSTR (PPPPPPP)"!(
    ComponentPose=@n46
    Alpha=1
    ConstraintWeights_0=0.237
    ConstraintWeights_1:float=0.1988
    ConstraintWeights_2:float=0.1897
    ConstraintWeights_3:float=0.1762
    ConstraintWeights_4:float=0.1164
    ConstraintWeights_5:float=0.0703
    ConstraintWeights_6:float=0.0117
   ) -> (Pose) {
    alphaBoolBlend {blendOption=Linear}
    boneToModify {boneName=dyn_fur_torso_fr_04_MSTR}
    constraintSetup [
     {targetBone {boneName=spine_02}}
     {targetBone {boneName=dyn_fur_torso_fr_04}}
     {targetBone {boneName=spine_03}}
     {targetBone {boneName=spine_04}}
     {targetBone {boneName=spine_01}}
     {targetBone {boneName=pelvis}}
     {targetBone {boneName=spine_05}}
    ]
    constraintWeights [1,1,1,1,1,1,1]
   }
   n48 = "Constraint - dyn_fur_torso_fr_03_MSTR (PPPPPP)"!(
    ComponentPose=@n47
    Alpha=1
    ConstraintWeights_0=0.638
    ConstraintWeights_1:float=0.109
    ConstraintWeights_2:float=0.0928
    ConstraintWeights_3:float=0.0893
    ConstraintWeights_4:float=0.0551
    ConstraintWeights_5:float=0.0159
   ) -> (Pose) {
    alphaBoolBlend {blendOption=Linear}
    boneToModify {boneName=dyn_fur_torso_fr_03_MSTR}
    constraintSetup [
     {targetBone {boneName=dyn_fur_torso_fr_03}}
     {targetBone {boneName=pelvis}}
     {targetBone {boneName=spine_02}}
     {targetBone {boneName=spine_01}}
     {targetBone {boneName=spine_03}}
     {targetBone {boneName=spine_04}}
    ]
    constraintWeights [1,1,1,1,1,1]
   }
   n49 = "Constraint - dyn_fur_torso_fr_02_MSTR (PPPPPP)"!(
    ComponentPose=@n48
    Alpha=1
    ConstraintWeights_0=0.2487
    ConstraintWeights_1:float=0.2477
    ConstraintWeights_2:float=0.2465
    ConstraintWeights_3:float=0.111
    ConstraintWeights_4:float=0.0755
    ConstraintWeights_5:float=0.0707
   ) -> (Pose) {
    alphaBoolBlend {blendOption=Linear}
    boneToModify {boneName=dyn_fur_torso_fr_02_MSTR}
    constraintSetup [
     {targetBone {boneName=spine_02}}
     {targetBone {boneName=spine_03}}
     {targetBone {boneName=spine_04}}
     {targetBone {boneName=spine_01}}
     {targetBone {boneName=pelvis}}
     {targetBone {boneName=spine_05}}
    ]
    constraintWeights [1,1,1,1,1,1]
   }
   n50 = "Constraint - dyn_fur_torso_fr_01_MSTR (PPPPPP)"!(
    ComponentPose=@n49
    Alpha=1
    ConstraintWeights_0=0.1994
    ConstraintWeights_1:float=0.1918
    ConstraintWeights_2:float=0.1912
    ConstraintWeights_3:float=0.1729
    ConstraintWeights_4:float=0.1403
    ConstraintWeights_5:float=0.1046
   ) -> (Pose) {
    alphaBoolBlend {blendOption=Linear}
    boneToModify {boneName=dyn_fur_torso_fr_01_MSTR}
    constraintSetup [
     {targetBone {boneName=dyn_fur_torso_fr_01}}
     {targetBone {boneName=spine_01}}
     {targetBone {boneName=pelvis}}
     {targetBone {boneName=spine_02}}
     {targetBone {boneName=spine_03}}
     {targetBone {boneName=spine_04}}
    ]
    constraintWeights [1,1,1,1,1,1]
   }
   n51 = "Constraint - dyn_fur_torso_bk_02_MSTR (PPPPPP)"!(
    ComponentPose=@n50
    Alpha=1
    ConstraintWeights_0=0.3941
    ConstraintWeights_1:float=0.199
    ConstraintWeights_2:float=0.1413
    ConstraintWeights_3:float=0.1034
    ConstraintWeights_4:float=0.0905
    ConstraintWeights_5:float=0.0718
   ) -> (Pose) {
    alphaBoolBlend {blendOption=Linear}
    boneToModify {boneName=dyn_fur_torso_bk_02_MSTR}
    constraintSetup [
     {targetBone {boneName=dyn_fur_torso_bk_02}}
     {targetBone {boneName=pelvis}}
     {targetBone {boneName=spine_01}}
     {targetBone {boneName=spine_02}}
     {targetBone {boneName=spine_03}}
     {targetBone {boneName=spine_04}}
    ]
    constraintWeights [1,1,1,1,1,1]
   }
   n52 = "Constraint - dyn_fur_torso_bk_01_MSTR (PPPPPP)"!(
    ComponentPose=@n51
    Alpha=1
    ConstraintWeights_0=0.5005
    ConstraintWeights_1:float=0.2213
    ConstraintWeights_2:float=0.1812
    ConstraintWeights_3:float=0.0541
    ConstraintWeights_4:float=0.0326
    ConstraintWeights_5:float=0.0104
   ) -> (Pose) {
    alphaBoolBlend {blendOption=Linear}
    boneToModify {boneName=dyn_fur_torso_bk_01_MSTR}
    constraintSetup [
     {targetBone {boneName=spine_05}}
     {targetBone {boneName=dyn_fur_torso_bk_01}}
     {targetBone {boneName=spine_04}}
     {targetBone {boneName=clavicle_r}}
     {targetBone {boneName=spine_03}}
     {targetBone {boneName=spine_02}}
    ]
    constraintWeights [1,1,1,1,1,1]
   }
   n53 = "Constraint - dyn_fur_lowerleg_l_01_MSTR (PPPPPP)"!(
    ComponentPose=@n52
    Alpha=1
    ConstraintWeights_0=0.277
    ConstraintWeights_1:float=0.2438
    ConstraintWeights_2:float=0.1521
    ConstraintWeights_3:float=0.1435
    ConstraintWeights_4:float=0.1001
    ConstraintWeights_5:float=0.0835
   ) -> (Pose) {
    alphaBoolBlend {blendOption=Linear}
    boneToModify {boneName=dyn_fur_lowerleg_l_01_MSTR}
    constraintSetup [
     {targetBone {boneName=calf_l}}
     {targetBone {boneName=deform_knee_fr_01_l}}
     {targetBone {boneName=deform_knee_bk_01_l}}
     {targetBone {boneName=dyn_fur_lowerleg_l_01}}
     {targetBone {boneName=thigh_l}}
     {targetBone {boneName=calf_twist_02_l}}
    ]
    constraintWeights [1,1,1,1,1,1]
   }
   n54 = "Constraint - dyn_fur_lowerleg_l_02_MSTR (PPP)"!(ComponentPose=@n53, Alpha=1, ConstraintWeights_0=0.4415, ConstraintWeights_1:float=0.2858, ConstraintWeights_2:float=0.2726) -> (Pose) {
    alphaBoolBlend {blendOption=Linear}
    boneToModify {boneName=dyn_fur_lowerleg_l_02_MSTR}
    constraintSetup [{targetBone {boneName=calf_twist_02_l}}, {targetBone {boneName=dyn_fur_lowerleg_l_02}}, {targetBone {boneName=calf_l}}]
    constraintWeights [1,1,1]
   }
   n55 = "Constraint - dyn_fur_lowerleg_l_03_MSTR (PPPP)"!(
    ComponentPose=@n54
    Alpha=1
    ConstraintWeights_0=0.4644
    ConstraintWeights_1:float=0.2205
    ConstraintWeights_2:float=0.1741
    ConstraintWeights_3:float=0.141
   ) -> (Pose) {
    alphaBoolBlend {blendOption=Linear}
    boneToModify {boneName=dyn_fur_lowerleg_l_03_MSTR}
    constraintSetup [
     {targetBone {boneName=calf_twist_02_l}}
     {targetBone {boneName=calf_l}}
     {targetBone {boneName=calf_twist_01_l}}
     {targetBone {boneName=dyn_fur_lowerleg_l_03}}
    ]
    constraintWeights [1,1,1,1]
   }
   n56 = "Constraint - dyn_fur_lowerleg_l_04_MSTR (PPPP)"!(
    ComponentPose=@n55
    Alpha=1
    ConstraintWeights_0=0.4741
    ConstraintWeights_1:float=0.473
    ConstraintWeights_2:float=0.0357
    ConstraintWeights_3:float=0.0172
   ) -> (Pose) {
    alphaBoolBlend {blendOption=Linear}
    boneToModify {boneName=dyn_fur_lowerleg_l_04_MSTR}
    constraintSetup [
     {targetBone {boneName=calf_l}}
     {targetBone {boneName=calf_twist_02_l}}
     {targetBone {boneName=deform_knee_fr_01_l}}
     {targetBone {boneName=calf_twist_01_l}}
    ]
    constraintWeights [1,1,1,1]
   }
   n57 = "Constraint - dyn_fur_lowerleg_l_05_MSTR (PPPP)"!(
    ComponentPose=@n56
    Alpha=1
    ConstraintWeights_0=0.4558
    ConstraintWeights_1:float=0.2773
    ConstraintWeights_2:float=0.1993
    ConstraintWeights_3:float=0.0677
   ) -> (Pose) {
    alphaBoolBlend {blendOption=Linear}
    boneToModify {boneName=dyn_fur_lowerleg_l_05_MSTR}
    constraintSetup [
     {targetBone {boneName=calf_twist_02_l}}
     {targetBone {boneName=calf_twist_01_l}}
     {targetBone {boneName=dyn_fur_lowerleg_l_05}}
     {targetBone {boneName=calf_l}}
    ]
    constraintWeights [1,1,1,1]
   }
   n58 = "Constraint - dyn_fur_lowerleg_l_06_MSTR (PPP)"!(ComponentPose=@n57, Alpha=1, ConstraintWeights_0=0.4884, ConstraintWeights_1:float=0.3382, ConstraintWeights_2:float=0.1734) -> (Pose) {
    alphaBoolBlend {blendOption=Linear}
    boneToModify {boneName=dyn_fur_lowerleg_l_06_MSTR}
    constraintSetup [{targetBone {boneName=foot_l}}, {targetBone {boneName=dyn_fur_lowerleg_l_06}}, {targetBone {boneName=calf_twist_01_l}}]
    constraintWeights [1,1,1]
   }
   n59 = "Constraint - dyn_fur_lowerleg_l_07_MSTR (PPP)"!(ComponentPose=@n58, Alpha=1, ConstraintWeights_0=0.423, ConstraintWeights_1:float=0.4157, ConstraintWeights_2:float=0.1614) -> (Pose) {
    alphaBoolBlend {blendOption=Linear}
    boneToModify {boneName=dyn_fur_lowerleg_l_07_MSTR}
    constraintSetup [{targetBone {boneName=calf_twist_01_l}}, {targetBone {boneName=foot_l}}, {targetBone {boneName=dyn_fur_lowerleg_l_07}}]
    constraintWeights [1,1,1]
   }
   n60 = "Constraint - dyn_fur_lowerleg_l_08_MSTR (PPP)"!(ComponentPose=@n59, Alpha=1, ConstraintWeights_0=0.5336, ConstraintWeights_1:float=0.377, ConstraintWeights_2:float=0.0894) -> (Pose) {
    alphaBoolBlend {blendOption=Linear}
    boneToModify {boneName=dyn_fur_lowerleg_l_08_MSTR}
    constraintSetup [{targetBone {boneName=calf_twist_01_l}}, {targetBone {boneName=foot_l}}, {targetBone {boneName=calf_twist_02_l}}]
    constraintWeights [1,1,1]
   }
   n61 = "Constraint - dyn_fur_lowerleg_l_09_MSTR (PPP)"!(ComponentPose=@n60, Alpha=1, ConstraintWeights_0=0.4853, ConstraintWeights_1:float=0.3159, ConstraintWeights_2:float=0.1988) -> (Pose) {
    alphaBoolBlend {blendOption=Linear}
    boneToModify {boneName=dyn_fur_lowerleg_l_09_MSTR}
    constraintSetup [{targetBone {boneName=foot_l}}, {targetBone {boneName=dyn_fur_lowerleg_l_09}}, {targetBone {boneName=calf_twist_01_l}}]
    constraintWeights [1,1,1]
   }
   n62 = "Constraint - dyn_fur_lowerleg_r_01_MSTR (PPPPPP)"!(
    ComponentPose=@n61
    Alpha=1
    ConstraintWeights_0=0.2618
    ConstraintWeights_1:float=0.24
    ConstraintWeights_2:float=0.1891
    ConstraintWeights_3:float=0.1492
    ConstraintWeights_4:float=0.1164
    ConstraintWeights_5:float=0.0435
   ) -> (Pose) {
    alphaBoolBlend {blendOption=Linear}
    boneToModify {boneName=dyn_fur_lowerleg_r_01_MSTR}
    constraintSetup [
     {targetBone {boneName=dyn_fur_lowerleg_r_01}}
     {targetBone {boneName=calf_r}}
     {targetBone {boneName=deform_knee_fr_01_r}}
     {targetBone {boneName=deform_knee_bk_01_r}}
     {targetBone {boneName=calf_twist_02_r}}
     {targetBone {boneName=thigh_r}}
    ]
    constraintWeights [1,1,1,1,1,1]
   }
   n63 = "Constraint - dyn_fur_lowerleg_r_02_MSTR (PPPP)"!(
    ComponentPose=@n62
    Alpha=1
    ConstraintWeights_0=0.5127
    ConstraintWeights_1:float=0.3772
    ConstraintWeights_2:float=0.0853
    ConstraintWeights_3:float=0.0249
   ) -> (Pose) {
    alphaBoolBlend {blendOption=Linear}
    boneToModify {boneName=dyn_fur_lowerleg_r_02_MSTR}
    constraintSetup [
     {targetBone {boneName=calf_twist_02_r}}
     {targetBone {boneName=calf_r}}
     {targetBone {boneName=dyn_fur_lowerleg_r_02}}
     {targetBone {boneName=deform_knee_bk_01_r}}
    ]
    constraintWeights [1,1,1,1]
   }
   n64 = "Constraint - dyn_fur_lowerleg_r_03_MSTR (PPP)"!(ComponentPose=@n63, Alpha=1, ConstraintWeights_0=0.5527, ConstraintWeights_1:float=0.3029, ConstraintWeights_2:float=0.1444) -> (Pose) {
    alphaBoolBlend {blendOption=Linear}
    boneToModify {boneName=dyn_fur_lowerleg_r_03_MSTR}
    constraintSetup [{targetBone {boneName=calf_twist_01_r}}, {targetBone {boneName=calf_twist_02_r}}, {targetBone {boneName=dyn_fur_lowerleg_r_03}}]
    constraintWeights [1,1,1]
   }
   n65 = "Constraint - dyn_fur_lowerleg_r_04_MSTR (PPPP)"!(
    ComponentPose=@n64
    Alpha=1
    ConstraintWeights_0=0.4733
    ConstraintWeights_1:float=0.3902
    ConstraintWeights_2:float=0.0963
    ConstraintWeights_3:float=0.0402
   ) -> (Pose) {
    alphaBoolBlend {blendOption=Linear}
    boneToModify {boneName=dyn_fur_lowerleg_r_04_MSTR}
    constraintSetup [
     {targetBone {boneName=calf_twist_02_r}}
     {targetBone {boneName=calf_r}}
     {targetBone {boneName=dyn_fur_lowerleg_r_04}}
     {targetBone {boneName=calf_twist_01_r}}
    ]
    constraintWeights [1,1,1,1]
   }
   n66 = "Constraint - dyn_fur_lowerleg_r_05_MSTR (PPP)"!(ComponentPose=@n65, Alpha=1, ConstraintWeights_0=0.5524, ConstraintWeights_1:float=0.3973, ConstraintWeights_2:float=0.0503) -> (Pose) {
    alphaBoolBlend {blendOption=Linear}
    boneToModify {boneName=dyn_fur_lowerleg_r_05_MSTR}
    constraintSetup [{targetBone {boneName=calf_twist_02_r}}, {targetBone {boneName=calf_twist_01_r}}, {targetBone {boneName=calf_r}}]
    constraintWeights [1,1,1]
   }
   n67 = "Constraint - dyn_fur_lowerleg_r_06_MSTR (PPP)"!(ComponentPose=@n66, Alpha=1, ConstraintWeights_0=0.6815, ConstraintWeights_1:float=0.1602, ConstraintWeights_2:float=0.1583) -> (Pose) {
    alphaBoolBlend {blendOption=Linear}
    boneToModify {boneName=dyn_fur_lowerleg_r_06_MSTR}
    constraintSetup [{targetBone {boneName=foot_r}}, {targetBone {boneName=calf_twist_01_r}}, {targetBone {boneName=dyn_fur_lowerleg_r_06}}]
    constraintWeights [1,1,1]
   }
   n68 = "Constraint - dyn_fur_lowerleg_r_07_MSTR (PPP)"!(ComponentPose=@n67, Alpha=1, ConstraintWeights_0=0.5737, ConstraintWeights_1:float=0.2801, ConstraintWeights_2:float=0.1462) -> (Pose) {
    alphaBoolBlend {blendOption=Linear}
    boneToModify {boneName=dyn_fur_lowerleg_r_07_MSTR}
    constraintSetup [{targetBone {boneName=foot_r}}, {targetBone {boneName=calf_twist_01_r}}, {targetBone {boneName=dyn_fur_lowerleg_r_07}}]
    constraintWeights [1,1,1]
   }
   n69 = "Constraint - dyn_fur_lowerleg_r_08_MSTR (PPP)"!(ComponentPose=@n68, Alpha=1, ConstraintWeights_0=0.4171, ConstraintWeights_1:float=0.3827, ConstraintWeights_2:float=0.2003) -> (Pose) {
    alphaBoolBlend {blendOption=Linear}
    boneToModify {boneName=dyn_fur_lowerleg_r_08_MSTR}
    constraintSetup [{targetBone {boneName=dyn_fur_lowerleg_r_08}}, {targetBone {boneName=foot_r}}, {targetBone {boneName=calf_twist_01_r}}]
    constraintWeights [1,1,1]
   }
   n70 = "Constraint - dyn_fur_lowerleg_r_09_MSTR (PPP)"!(ComponentPose=@n69, Alpha=1, ConstraintWeights_0=0.512, ConstraintWeights_1:float=0.2905, ConstraintWeights_2:float=0.1976) -> (Pose) {
    alphaBoolBlend {blendOption=Linear}
    boneToModify {boneName=dyn_fur_lowerleg_r_09_MSTR}
    constraintSetup [{targetBone {boneName=foot_r}}, {targetBone {boneName=dyn_fur_lowerleg_r_09}}, {targetBone {boneName=calf_twist_01_r}}]
    constraintWeights [1,1,1]
   }
   n71 = "Constraint - dyn_fur_upperleg_r_03_MSTR (PPPPPP)"!(
    ComponentPose=@n70
    Alpha=1
    ConstraintWeights_0=0.3368
    ConstraintWeights_1:float=0.2472
    ConstraintWeights_2:float=0.165
    ConstraintWeights_3:float=0.112
    ConstraintWeights_4:float=0.0814
    ConstraintWeights_5:float=0.0576
   ) -> (Pose) {
    alphaBoolBlend {blendOption=Linear}
    boneToModify {boneName=dyn_fur_upperleg_r_03_MSTR}
    constraintSetup [
     {targetBone {boneName=dyn_fur_upperleg_r_03}}
     {targetBone {boneName=thigh_r}}
     {targetBone {boneName=deform_knee_bk_01_r}}
     {targetBone {boneName=thigh_twist_01_r}}
     {targetBone {boneName=deform_knee_fr_01_r}}
     {targetBone {boneName=calf_r}}
    ]
    constraintWeights [1,1,1,1,1,1]
   }
   n72 = "Constraint - dyn_fur_upperleg_r_04_MSTR (PPPP)"!(
    ComponentPose=@n71
    Alpha=1
    ConstraintWeights_0=0.4626
    ConstraintWeights_1:float=0.2599
    ConstraintWeights_2:float=0.243
    ConstraintWeights_3:float=0.0345
   ) -> (Pose) {
    alphaBoolBlend {blendOption=Linear}
    boneToModify {boneName=dyn_fur_upperleg_r_04_MSTR}
    constraintSetup [
     {targetBone {boneName=dyn_fur_upperleg_r_04}}
     {targetBone {boneName=thigh_twist_01_r}}
     {targetBone {boneName=thigh_r}}
     {targetBone {boneName=deform_knee_fr_01_r}}
    ]
    constraintWeights [1,1,1,1]
   }
   n73 = "Constraint - dyn_fur_upperleg_r_02_MSTR (PPP)"!(ComponentPose=@n72, Alpha=1, ConstraintWeights_0=0.4763, ConstraintWeights_1:float=0.2687, ConstraintWeights_2:float=0.255) -> (Pose) {
    alphaBoolBlend {blendOption=Linear}
    boneToModify {boneName=dyn_fur_upperleg_r_02_MSTR}
    constraintSetup [{targetBone {boneName=thigh_twist_01_r}}, {targetBone {boneName=thigh_r}}, {targetBone {boneName=dyn_fur_upperleg_r_02}}]
    constraintWeights [1,1,1]
   }
   n74 = "Constraint - dyn_fur_upperleg_r_05_MSTR (PPPPP)"!(
    ComponentPose=@n73
    Alpha=1
    ConstraintWeights_0=0.4459
    ConstraintWeights_1:float=0.2197
    ConstraintWeights_2:float=0.2005
    ConstraintWeights_3:float=0.0674
    ConstraintWeights_4:float=0.0665
   ) -> (Pose) {
    alphaBoolBlend {blendOption=Linear}
    boneToModify {boneName=dyn_fur_upperleg_r_05_MSTR}
    constraintSetup [
     {targetBone {boneName=thigh_r}}
     {targetBone {boneName=thigh_twist_01_r}}
     {targetBone {boneName=deform_knee_bk_01_r}}
     {targetBone {boneName=deform_knee_fr_01_r}}
     {targetBone {boneName=calf_r}}
    ]
    constraintWeights [1,1,1,1,1]
   }
   n75 = "Constraint - dyn_fur_upperleg_r_07_MSTR (PPPPPP)"!(
    ComponentPose=@n74
    Alpha=1
    ConstraintWeights_0=0.3746
    ConstraintWeights_1:float=0.2996
    ConstraintWeights_2:float=0.1605
    ConstraintWeights_3:float=0.0691
    ConstraintWeights_4:float=0.0503
    ConstraintWeights_5:float=0.0459
   ) -> (Pose) {
    alphaBoolBlend {blendOption=Linear}
    boneToModify {boneName=dyn_fur_upperleg_r_07_MSTR}
    constraintSetup [
     {targetBone {boneName=thigh_twist_01_r}}
     {targetBone {boneName=dyn_fur_upperleg_r_07}}
     {targetBone {boneName=pelvis}}
     {targetBone {boneName=thigh_r}}
     {targetBone {boneName=deform_thigh_bk_r}}
     {targetBone {boneName=deform_thigh_fr_r}}
    ]
    constraintWeights [1,1,1,1,1,1]
   }
   n76 = "Constraint - dyn_fur_upperleg_r_01_MSTR (PPPPPP)"!(
    ComponentPose=@n75
    Alpha=1
    ConstraintWeights_0=0.4921
    ConstraintWeights_1:float=0.3382
    ConstraintWeights_2:float=0.0707
    ConstraintWeights_3:float=0.0446
    ConstraintWeights_4:float=0.0374
    ConstraintWeights_5:float=0.017
   ) -> (Pose) {
    alphaBoolBlend {blendOption=Linear}
    boneToModify {boneName=dyn_fur_upperleg_r_01_MSTR}
    constraintSetup [
     {targetBone {boneName=thigh_twist_01_r}}
     {targetBone {boneName=dyn_fur_upperleg_r_01}}
     {targetBone {boneName=thigh_r}}
     {targetBone {boneName=pelvis}}
     {targetBone {boneName=deform_thigh_fr_r}}
     {targetBone {boneName=deform_thigh_bk_r}}
    ]
    constraintWeights [1,1,1,1,1,1]
   }
   n77 = "Constraint - dyn_fur_upperleg_l_06_MSTR (PPPPP)"!(
    ComponentPose=@n76
    Alpha=1
    ConstraintWeights_0=0.427
    ConstraintWeights_1:float=0.2211
    ConstraintWeights_2:float=0.2037
    ConstraintWeights_3:float=0.1077
    ConstraintWeights_4:float=0.0405
   ) -> (Pose) {
    alphaBoolBlend {blendOption=Linear}
    boneToModify {boneName=dyn_fur_upperleg_l_06_MSTR}
    constraintSetup [
     {targetBone {boneName=dyn_fur_upperleg_l_06}}
     {targetBone {boneName=thigh_l}}
     {targetBone {boneName=deform_knee_bk_01_l}}
     {targetBone {boneName=thigh_twist_01_l}}
     {targetBone {boneName=calf_l}}
    ]
    constraintWeights [1,1,1,1,1]
   }
   n78 = "Constraint - dyn_fur_upperleg_l_03_MSTR (PPPPPP)"!(
    ComponentPose=@n77
    Alpha=1
    ConstraintWeights_0=0.3945
    ConstraintWeights_1:float=0.2448
    ConstraintWeights_2:float=0.1501
    ConstraintWeights_3:float=0.0832
    ConstraintWeights_4:float=0.0737
    ConstraintWeights_5:float=0.0537
   ) -> (Pose) {
    alphaBoolBlend {blendOption=Linear}
    boneToModify {boneName=dyn_fur_upperleg_l_03_MSTR}
    constraintSetup [
     {targetBone {boneName=dyn_fur_upperleg_l_03}}
     {targetBone {boneName=thigh_l}}
     {targetBone {boneName=deform_knee_bk_01_l}}
     {targetBone {boneName=thigh_twist_01_l}}
     {targetBone {boneName=deform_knee_fr_01_l}}
     {targetBone {boneName=calf_l}}
    ]
    constraintWeights [1,1,1,1,1,1]
   }
   n79 = "Constraint - dyn_fur_upperleg_l_05_MSTR (PPPPP)"!(
    ComponentPose=@n78
    Alpha=1
    ConstraintWeights_0=0.3735
    ConstraintWeights_1:float=0.3051
    ConstraintWeights_2:float=0.2423
    ConstraintWeights_3:float=0.0634
    ConstraintWeights_4:float=0.0158
   ) -> (Pose) {
    alphaBoolBlend {blendOption=Linear}
    boneToModify {boneName=dyn_fur_upperleg_l_05_MSTR}
    constraintSetup [
     {targetBone {boneName=thigh_l}}
     {targetBone {boneName=thigh_twist_01_l}}
     {targetBone {boneName=dyn_fur_upperleg_l_05}}
     {targetBone {boneName=deform_knee_bk_01_l}}
     {targetBone {boneName=deform_knee_fr_01_l}}
    ]
    constraintWeights [1,1,1,1,1]
   }
   n80 = "Constraint - dyn_fur_upperleg_l_02_MSTR (PPP)"!(ComponentPose=@n79, Alpha=1, ConstraintWeights_0=0.5192, ConstraintWeights_1:float=0.2625, ConstraintWeights_2:float=0.2183) -> (Pose) {
    alphaBoolBlend {blendOption=Linear}
    boneToModify {boneName=dyn_fur_upperleg_l_02_MSTR}
    constraintSetup [{targetBone {boneName=thigh_twist_01_l}}, {targetBone {boneName=dyn_fur_upperleg_l_02}}, {targetBone {boneName=thigh_l}}]
    constraintWeights [1,1,1]
   }
   n81 = "Constraint - dyn_fur_upperleg_l_01_MSTR (PPPPPP)"!(
    ComponentPose=@n80
    Alpha=1
    ConstraintWeights_0=0.5178
    ConstraintWeights_1:float=0.2363
    ConstraintWeights_2:float=0.0999
    ConstraintWeights_3:float=0.0798
    ConstraintWeights_4:float=0.0462
    ConstraintWeights_5:float=0.02
   ) -> (Pose) {
    alphaBoolBlend {blendOption=Linear}
    boneToModify {boneName=dyn_fur_upperleg_l_01_MSTR}
    constraintSetup [
     {targetBone {boneName=thigh_twist_01_l}}
     {targetBone {boneName=dyn_fur_upperleg_l_01}}
     {targetBone {boneName=pelvis}}
     {targetBone {boneName=deform_thigh_fr_l}}
     {targetBone {boneName=thigh_l}}
     {targetBone {boneName=deform_thigh_bk_l}}
    ]
    constraintWeights [1,1,1,1,1,1]
   }
   n82 = "Constraint - dyn_fur_upperleg_l_07_MSTR (PPPPPP)"!(
    ComponentPose=@n81
    Alpha=1
    ConstraintWeights_0=0.5157
    ConstraintWeights_1:float=0.2441
    ConstraintWeights_2:float=0.0712
    ConstraintWeights_3:float=0.0684
    ConstraintWeights_4:float=0.0582
    ConstraintWeights_5:float=0.0425
   ) -> (Pose) {
    alphaBoolBlend {blendOption=Linear}
    boneToModify {boneName=dyn_fur_upperleg_l_07_MSTR}
    constraintSetup [
     {targetBone {boneName=thigh_twist_01_l}}
     {targetBone {boneName=dyn_fur_upperleg_l_07}}
     {targetBone {boneName=deform_thigh_fr_l}}
     {targetBone {boneName=thigh_l}}
     {targetBone {boneName=pelvis}}
     {targetBone {boneName=deform_thigh_bk_l}}
    ]
    constraintWeights [1,1,1,1,1,1]
   }
   n83 = "Constraint - dyn_fur_pelvis_l_05_MSTR (PPPPP)"!(
    ComponentPose=@n82
    Alpha=1
    ConstraintWeights_0=0.3488
    ConstraintWeights_1:float=0.3483
    ConstraintWeights_2:float=0.1602
    ConstraintWeights_3:float=0.1129
    ConstraintWeights_4:float=0.0297
   ) -> (Pose) {
    alphaBoolBlend {blendOption=Linear}
    boneToModify {boneName=dyn_fur_pelvis_l_05_MSTR}
    constraintSetup [
     {targetBone {boneName=thigh_twist_01_l}}
     {targetBone {boneName=pelvis}}
     {targetBone {boneName=deform_thigh_fr_l}}
     {targetBone {boneName=dyn_fur_pelvis_l_05}}
     {targetBone {boneName=thigh_twist_01_r}}
    ]
    constraintWeights [1,1,1,1,1]
   }
   n84 = "Constraint - dyn_fur_pelvis_l_04_MSTR (PPPPP)"!(
    ComponentPose=@n83
    Alpha=1
    ConstraintWeights_0=0.4988
    ConstraintWeights_1:float=0.2772
    ConstraintWeights_2:float=0.1238
    ConstraintWeights_3:float=0.0717
    ConstraintWeights_4:float=0.0286
   ) -> (Pose) {
    alphaBoolBlend {blendOption=Linear}
    boneToModify {boneName=dyn_fur_pelvis_l_04_MSTR}
    constraintSetup [
     {targetBone {boneName=pelvis}}
     {targetBone {boneName=dyn_fur_pelvis_l_04}}
     {targetBone {boneName=deform_thigh_bk_l}}
     {targetBone {boneName=thigh_twist_01_l}}
     {targetBone {boneName=deform_thigh_bk_r}}
    ]
    constraintWeights [1,1,1,1,1]
   }
   n85 = "Constraint - dyn_fur_pelvis_l_03_MSTR (PPPPP)"!(
    ComponentPose=@n84
    Alpha=1
    ConstraintWeights_0=0.7091
    ConstraintWeights_1:float=0.1111
    ConstraintWeights_2:float=0.0849
    ConstraintWeights_3:float=0.0627
    ConstraintWeights_4:float=0.0322
   ) -> (Pose) {
    alphaBoolBlend {blendOption=Linear}
    boneToModify {boneName=dyn_fur_pelvis_l_03_MSTR}
    constraintSetup [
     {targetBone {boneName=pelvis}}
     {targetBone {boneName=deform_thigh_fr_l}}
     {targetBone {boneName=dyn_fur_pelvis_l_03}}
     {targetBone {boneName=thigh_twist_01_l}}
     {targetBone {boneName=deform_thigh_bk_l}}
    ]
    constraintWeights [1,1,1,1,1]
   }
   n86 = "Constraint - dyn_fur_pelvis_l_02_MSTR (PPPPP)"!(
    ComponentPose=@n85
    Alpha=1
    ConstraintWeights_0=0.3195
    ConstraintWeights_1:float=0.2239
    ConstraintWeights_2:float=0.208
    ConstraintWeights_3:float=0.1804
    ConstraintWeights_4:float=0.0682
   ) -> (Pose) {
    alphaBoolBlend {blendOption=Linear}
    boneToModify {boneName=dyn_fur_pelvis_l_02_MSTR}
    constraintSetup [
     {targetBone {boneName=pelvis}}
     {targetBone {boneName=dyn_fur_pelvis_l_02}}
     {targetBone {boneName=thigh_twist_01_l}}
     {targetBone {boneName=deform_thigh_fr_l}}
     {targetBone {boneName=deform_thigh_bk_l}}
    ]
    constraintWeights [1,1,1,1,1]
   }
   n87 = "Constraint - dyn_fur_pelvis_l_01_MSTR (PPPPP)"!(
    ComponentPose=@n86
    Alpha=1
    ConstraintWeights_0=0.2892
    ConstraintWeights_1:float=0.2882
    ConstraintWeights_2:float=0.1468
    ConstraintWeights_3:float=0.1396
    ConstraintWeights_4:float=0.1362
   ) -> (Pose) {
    alphaBoolBlend {blendOption=Linear}
    boneToModify {boneName=dyn_fur_pelvis_l_01_MSTR}
    constraintSetup [
     {targetBone {boneName=thigh_twist_01_l}}
     {targetBone {boneName=pelvis}}
     {targetBone {boneName=deform_thigh_bk_l}}
     {targetBone {boneName=deform_thigh_fr_l}}
     {targetBone {boneName=dyn_fur_pelvis_l_01}}
    ]
    constraintWeights [1,1,1,1,1]
   }
   n88 = Anim Dynamics - Bone: dyn_fur_pelvis_l_01(ComponentPose=@n87, Alpha=1, ExternalForce="0,0,0") {
    angularDampingOverride=1
    angularSpringConstant=20000
    bAngularSpring=true
    bOverrideAngularDamping=true
    bUseGravityOverride=true
    gravityScale=0.35
    simulationSpace=RootRelative
    alphaBoolBlend {blendOption=Linear}
    boundBone {boneName=dyn_fur_pelvis_l_01}
    componentLinearAccScale=<1 1 1>
    componentLinearVelScale=<1 1 1>
    gravityOverride {z=-980}
    physicsBodyDefinitions [{
      sphereCollisionRadius=0
      boundBone {boneName=dyn_fur_pelvis_l_01}
      boxExtents=<25 36 29>
      constraintSetup {angularLimitsMax=<15 15 15>, angularLimitsMin=<-15 -15 -15>, angularTarget=<1 0 0>}
      localJointOffset=<4.5 0 0>
     }]
   }
   n89 = Anim Dynamics - Bone: dyn_fur_pelvis_l_03(ComponentPose=@n88, Alpha=1, ExternalForce="0,0,0") {
    angularDampingOverride=1
    angularSpringConstant=20000
    bAngularSpring=true
    bOverrideAngularDamping=true
    bUseGravityOverride=true
    gravityScale=0.35
    simulationSpace=RootRelative
    alphaBoolBlend {blendOption=Linear}
    boundBone {boneName=dyn_fur_pelvis_l_03}
    componentLinearAccScale=<1 1 1>
    componentLinearVelScale=<1 1 1>
    gravityOverride {z=-980}
    physicsBodyDefinitions [{
      sphereCollisionRadius=0
      boundBone {boneName=dyn_fur_pelvis_l_03}
      boxExtents=<20 30 30>
      constraintSetup {angularLimitsMax=<15 15 15>, angularLimitsMin=<-15 -15 -15>, angularTarget=<1 0 0>}
      localJointOffset=<4 0 0>
     }]
   }
   n90 = Anim Dynamics - Bone: dyn_fur_pelvis_l_02(ComponentPose=@n89, Alpha=1, ExternalForce="0,0,0") {
    angularDampingOverride=1
    angularSpringConstant=20000
    bAngularSpring=true
    bOverrideAngularDamping=true
    bUseGravityOverride=true
    gravityScale=0.35
    simulationSpace=RootRelative
    alphaBoolBlend {blendOption=Linear}
    boundBone {boneName=dyn_fur_pelvis_l_02}
    componentLinearAccScale=<1 1 1>
    componentLinearVelScale=<1 1 1>
    gravityOverride {z=-980}
    physicsBodyDefinitions [{
      sphereCollisionRadius=0
      boundBone {boneName=dyn_fur_pelvis_l_02}
      boxExtents=<25 36 29>
      constraintSetup {angularLimitsMax=<15 15 15>, angularLimitsMin=<-15 -15 -15>, angularTarget=<1 0 0>}
      localJointOffset=<4.5 0 0>
     }]
   }
   n91 = Anim Dynamics - Bone: dyn_fur_pelvis_l_04(ComponentPose=@n90, Alpha=1, ExternalForce="0,0,0") {
    angularDampingOverride=1
    angularSpringConstant=20000
    bAngularSpring=true
    bOverrideAngularDamping=true
    bUseGravityOverride=true
    gravityScale=0.35
    simulationSpace=RootRelative
    alphaBoolBlend {blendOption=Linear}
    boundBone {boneName=dyn_fur_pelvis_l_04}
    componentLinearAccScale=<1 1 1>
    componentLinearVelScale=<1 1 1>
    gravityOverride {z=-980}
    physicsBodyDefinitions [{
      sphereCollisionRadius=0
      boundBone {boneName=dyn_fur_pelvis_l_04}
      boxExtents=<18 28 37>
      constraintSetup {angularLimitsMax=<15 15 15>, angularLimitsMin=<-15 -15 -15>, angularTarget=<1 0 0>}
      localJointOffset=<3.9 0 0>
     }]
   }
   n92 = Anim Dynamics - Bone: dyn_fur_upperleg_l_07(ComponentPose=@n91, Alpha=1, ExternalForce="0,0,0") {
    angularDampingOverride=1
    angularSpringConstant=20000
    bAngularSpring=true
    bOverrideAngularDamping=true
    bUseGravityOverride=true
    gravityScale=0.35
    simulationSpace=RootRelative
    alphaBoolBlend {blendOption=Linear}
    boundBone {boneName=dyn_fur_upperleg_l_07}
    componentLinearAccScale=<1 1 1>
    componentLinearVelScale=<1 1 1>
    gravityOverride {z=-980}
    physicsBodyDefinitions [{
      sphereCollisionRadius=0
      boundBone {boneName=dyn_fur_upperleg_l_07}
      boxExtents=<20 30 30>
      constraintSetup {angularLimitsMax=<15 15 15>, angularLimitsMin=<-15 -15 -15>, angularTarget=<1 0 0>}
      localJointOffset=<4 0 0>
     }]
   }
   n93 = Anim Dynamics - Bone: dyn_fur_pelvis_l_05(ComponentPose=@n92, Alpha=1, ExternalForce="0,0,0") {
    angularDampingOverride=1
    angularSpringConstant=20000
    bAngularSpring=true
    bOverrideAngularDamping=true
    bUseGravityOverride=true
    gravityScale=0.35
    simulationSpace=RootRelative
    alphaBoolBlend {blendOption=Linear}
    boundBone {boneName=dyn_fur_pelvis_l_05}
    componentLinearAccScale=<1 1 1>
    componentLinearVelScale=<1 1 1>
    gravityOverride {z=-980}
    physicsBodyDefinitions [{
      sphereCollisionRadius=0
      boundBone {boneName=dyn_fur_pelvis_l_05}
      boxExtents=<18 28 37>
      constraintSetup {angularLimitsMax=<15 15 15>, angularLimitsMin=<-15 -15 -15>, angularTarget=<1 0 0>}
      localJointOffset=<3.9 0 0>
     }]
   }
   n94 = Anim Dynamics - Bone: dyn_fur_upperleg_l_01(ComponentPose=@n93, Alpha=1, ExternalForce="0,0,0") {
    angularDampingOverride=1
    angularSpringConstant=20000
    bAngularSpring=true
    bOverrideAngularDamping=true
    bUseGravityOverride=true
    gravityScale=0.35
    simulationSpace=RootRelative
    alphaBoolBlend {blendOption=Linear}
    boundBone {boneName=dyn_fur_upperleg_l_01}
    componentLinearAccScale=<1 1 1>
    componentLinearVelScale=<1 1 1>
    gravityOverride {z=-980}
    physicsBodyDefinitions [{
      sphereCollisionRadius=0
      boundBone {boneName=dyn_fur_upperleg_l_01}
      boxExtents=<20 30 30>
      constraintSetup {angularLimitsMax=<15 15 15>, angularLimitsMin=<-15 -15 -15>, angularTarget=<1 0 0>}
      localJointOffset=<4 0 0>
     }]
   }
   n95 = Anim Dynamics - Bone: dyn_fur_upperleg_l_05(ComponentPose=@n94, Alpha=1, ExternalForce="0,0,0") {
    angularDampingOverride=1
    angularSpringConstant=20000
    bAngularSpring=true
    bOverrideAngularDamping=true
    bUseGravityOverride=true
    gravityScale=0.35
    simulationSpace=RootRelative
    alphaBoolBlend {blendOption=Linear}
    boundBone {boneName=dyn_fur_upperleg_l_05}
    componentLinearAccScale=<1 1 1>
    componentLinearVelScale=<1 1 1>
    gravityOverride {z=-980}
    physicsBodyDefinitions [{
      sphereCollisionRadius=0
      boundBone {boneName=dyn_fur_upperleg_l_05}
      boxExtents=<18 28 37>
      constraintSetup {angularLimitsMax=<15 15 15>, angularLimitsMin=<-15 -15 -15>, angularTarget=<1 0 0>}
      localJointOffset=<3.9 0 0>
     }]
   }
   n96 = Anim Dynamics - Bone: dyn_fur_upperleg_l_02(ComponentPose=@n95, Alpha=1, ExternalForce="0,0,0") {
    angularDampingOverride=1
    angularSpringConstant=20000
    bAngularSpring=true
    bOverrideAngularDamping=true
    bUseGravityOverride=true
    gravityScale=0.35
    simulationSpace=RootRelative
    alphaBoolBlend {blendOption=Linear}
    boundBone {boneName=dyn_fur_upperleg_l_02}
    componentLinearAccScale=<1 1 1>
    componentLinearVelScale=<1 1 1>
    gravityOverride {z=-980}
    physicsBodyDefinitions [{
      sphereCollisionRadius=0
      boundBone {boneName=dyn_fur_upperleg_l_02}
      boxExtents=<25 36 29>
      constraintSetup {angularLimitsMax=<15 15 15>, angularLimitsMin=<-15 -15 -15>, angularTarget=<1 0 0>}
      localJointOffset=<4.5 0 0>
     }]
   }
   n97 = Anim Dynamics - Bone: dyn_fur_upperleg_l_03(ComponentPose=@n96, Alpha=1, ExternalForce="0,0,0") {
    angularDampingOverride=1
    angularSpringConstant=20000
    bAngularSpring=true
    bOverrideAngularDamping=true
    bUseGravityOverride=true
    gravityScale=0.35
    simulationSpace=RootRelative
    alphaBoolBlend {blendOption=Linear}
    boundBone {boneName=dyn_fur_upperleg_l_03}
    componentLinearAccScale=<1 1 1>
    componentLinearVelScale=<1 1 1>
    gravityOverride {z=-980}
    physicsBodyDefinitions [{
      sphereCollisionRadius=0
      boundBone {boneName=dyn_fur_upperleg_l_03}
      boxExtents=<20 30 30>
      constraintSetup {angularLimitsMax=<15 15 15>, angularLimitsMin=<-15 -15 -15>, angularTarget=<1 0 0>}
      localJointOffset=<4 0 0>
     }]
   }
   n98 = Anim Dynamics - Bone: dyn_fur_upperleg_r_01(ComponentPose=@n97, Alpha=1, ExternalForce="0,0,0") {
    angularDampingOverride=1
    angularSpringConstant=20000
    bAngularSpring=true
    bOverrideAngularDamping=true
    bUseGravityOverride=true
    gravityScale=0.35
    simulationSpace=RootRelative
    alphaBoolBlend {blendOption=Linear}
    boundBone {boneName=dyn_fur_upperleg_r_01}
    componentLinearAccScale=<1 1 1>
    componentLinearVelScale=<1 1 1>
    gravityOverride {z=-980}
    physicsBodyDefinitions [{
      sphereCollisionRadius=0
      boundBone {boneName=dyn_fur_upperleg_r_01}
      boxExtents=<25 36 29>
      constraintSetup {angularLimitsMax=<15 15 15>, angularLimitsMin=<-15 -15 -15>, angularTarget=<1 0 0>}
      localJointOffset=<4.5 0 0>
     }]
   }
   n99 = Anim Dynamics - Bone: dyn_fur_upperleg_l_06(ComponentPose=@n98, Alpha=1, ExternalForce="0,0,0") {
    angularDampingOverride=1
    angularSpringConstant=20000
    bAngularSpring=true
    bOverrideAngularDamping=true
    bUseGravityOverride=true
    gravityScale=0.35
    simulationSpace=RootRelative
    alphaBoolBlend {blendOption=Linear}
    boundBone {boneName=dyn_fur_upperleg_l_06}
    componentLinearAccScale=<1 1 1>
    componentLinearVelScale=<1 1 1>
    gravityOverride {z=-980}
    physicsBodyDefinitions [{
      sphereCollisionRadius=0
      boundBone {boneName=dyn_fur_upperleg_l_06}
      boxExtents=<20 30 30>
      constraintSetup {angularLimitsMax=<15 15 15>, angularLimitsMin=<-15 -15 -15>, angularTarget=<1 0 0>}
      localJointOffset=<4 0 0>
     }]
   }
   n100 = Anim Dynamics - Bone: dyn_fur_upperleg_r_07(ComponentPose=@n99, Alpha=1, ExternalForce="0,0,0") {
    angularDampingOverride=1
    angularSpringConstant=20000
    bAngularSpring=true
    bOverrideAngularDamping=true
    bUseGravityOverride=true
    gravityScale=0.35
    simulationSpace=RootRelative
    alphaBoolBlend {blendOption=Linear}
    boundBone {boneName=dyn_fur_upperleg_r_07}
    componentLinearAccScale=<1 1 1>
    componentLinearVelScale=<1 1 1>
    gravityOverride {z=-980}
    physicsBodyDefinitions [{
      sphereCollisionRadius=0
      boundBone {boneName=dyn_fur_upperleg_r_07}
      boxExtents=<20 30 30>
      constraintSetup {angularLimitsMax=<15 15 15>, angularLimitsMin=<-15 -15 -15>, angularTarget=<1 0 0>}
      localJointOffset=<4 0 0>
     }]
   }
   n101 = Anim Dynamics - Bone: dyn_fur_upperleg_r_02(ComponentPose=@n100, Alpha=1, ExternalForce="0,0,0") {
    angularDampingOverride=1
    angularSpringConstant=20000
    bAngularSpring=true
    bOverrideAngularDamping=true
    bUseGravityOverride=true
    gravityScale=0.35
    simulationSpace=RootRelative
    alphaBoolBlend {blendOption=Linear}
    boundBone {boneName=dyn_fur_upperleg_r_02}
    componentLinearAccScale=<1 1 1>
    componentLinearVelScale=<1 1 1>
    gravityOverride {z=-980}
    physicsBodyDefinitions [{
      sphereCollisionRadius=0
      boundBone {boneName=dyn_fur_upperleg_r_02}
      boxExtents=<18 28 37>
      constraintSetup {angularLimitsMax=<15 15 15>, angularLimitsMin=<-15 -15 -15>, angularTarget=<1 0 0>}
      localJointOffset=<3.9 0 0>
     }]
   }
   n102 = Anim Dynamics - Bone: dyn_fur_upperleg_r_05(ComponentPose=@n101, Alpha=1, ExternalForce="0,0,0") {
    angularDampingOverride=1
    angularSpringConstant=20000
    bAngularSpring=true
    bOverrideAngularDamping=true
    bUseGravityOverride=true
    gravityScale=0.35
    simulationSpace=RootRelative
    alphaBoolBlend {blendOption=Linear}
    boundBone {boneName=dyn_fur_upperleg_r_05}
    componentLinearAccScale=<1 1 1>
    componentLinearVelScale=<1 1 1>
    gravityOverride {z=-980}
    physicsBodyDefinitions [{
      sphereCollisionRadius=0
      boundBone {boneName=dyn_fur_upperleg_r_05}
      boxExtents=<20 30 30>
      constraintSetup {angularLimitsMax=<15 15 15>, angularLimitsMin=<-15 -15 -15>, angularTarget=<1 0 0>}
      localJointOffset=<4 0 0>
     }]
   }
   n103 = Anim Dynamics - Bone: dyn_fur_upperleg_r_04(ComponentPose=@n102, Alpha=1, ExternalForce="0,0,0") {
    angularDampingOverride=1
    angularSpringConstant=20000
    bAngularSpring=true
    bOverrideAngularDamping=true
    bUseGravityOverride=true
    gravityScale=0.35
    simulationSpace=RootRelative
    alphaBoolBlend {blendOption=Linear}
    boundBone {boneName=dyn_fur_upperleg_r_04}
    componentLinearAccScale=<1 1 1>
    componentLinearVelScale=<1 1 1>
    gravityOverride {z=-980}
    physicsBodyDefinitions [{
      sphereCollisionRadius=0
      boundBone {boneName=dyn_fur_upperleg_r_04}
      boxExtents=<25 36 29>
      constraintSetup {angularLimitsMax=<15 15 15>, angularLimitsMin=<-15 -15 -15>, angularTarget=<1 0 0>}
      localJointOffset=<4.5 0 0>
     }]
   }
   n104 = Anim Dynamics - Bone: dyn_fur_lowerarm_r_09(ComponentPose=@n103, Alpha=1, ExternalForce="0,0,0") {
    angularDampingOverride=1
    angularSpringConstant=20000
    bAngularSpring=true
    bOverrideAngularDamping=true
    bUseGravityOverride=true
    gravityScale=0.35
    simulationSpace=RootRelative
    alphaBoolBlend {blendOption=Linear}
    boundBone {boneName=dyn_fur_lowerarm_r_09}
    componentLinearAccScale=<1 1 1>
    componentLinearVelScale=<1 1 1>
    gravityOverride {z=-980}
    physicsBodyDefinitions [{
      sphereCollisionRadius=0
      boundBone {boneName=dyn_fur_lowerarm_r_09}
      boxExtents=<20 30 30>
      constraintSetup {angularLimitsMax=<15 15 15>, angularLimitsMin=<-15 -15 -15>, angularTarget=<1 0 0>}
      localJointOffset=<4 0 0>
     }]
   }
   n105 = Anim Dynamics - Bone: dyn_fur_upperleg_r_03(ComponentPose=@n104, Alpha=1, ExternalForce="0,0,0") {
    angularDampingOverride=1
    angularSpringConstant=20000
    bAngularSpring=true
    bOverrideAngularDamping=true
    bUseGravityOverride=true
    gravityScale=0.35
    simulationSpace=RootRelative
    alphaBoolBlend {blendOption=Linear}
    boundBone {boneName=dyn_fur_upperleg_r_03}
    componentLinearAccScale=<1 1 1>
    componentLinearVelScale=<1 1 1>
    gravityOverride {z=-980}
    physicsBodyDefinitions [{
      sphereCollisionRadius=0
      boundBone {boneName=dyn_fur_upperleg_r_03}
      boxExtents=<25 36 29>
      constraintSetup {angularLimitsMax=<15 15 15>, angularLimitsMin=<-15 -15 -15>, angularTarget=<1 0 0>}
      localJointOffset=<4.5 0 0>
     }]
   }
   n106 = Anim Dynamics - Bone: dyn_fur_lowerarm_r_08(ComponentPose=@n105, Alpha=1, ExternalForce="0,0,0") {
    angularDampingOverride=1
    angularSpringConstant=20000
    bAngularSpring=true
    bOverrideAngularDamping=true
    bUseGravityOverride=true
    gravityScale=0.35
    simulationSpace=RootRelative
    alphaBoolBlend {blendOption=Linear}
    boundBone {boneName=dyn_fur_lowerarm_r_08}
    componentLinearAccScale=<1 1 1>
    componentLinearVelScale=<1 1 1>
    gravityOverride {z=-980}
    physicsBodyDefinitions [{
      sphereCollisionRadius=0
      boundBone {boneName=dyn_fur_lowerarm_r_08}
      boxExtents=<18 28 37>
      constraintSetup {angularLimitsMax=<15 15 15>, angularLimitsMin=<-15 -15 -15>, angularTarget=<1 0 0>}
      localJointOffset=<3.9 0 0>
     }]
   }
   n107 = Anim Dynamics - Bone: dyn_fur_lowerarm_r_06(ComponentPose=@n106, Alpha=1, ExternalForce="0,0,0") {
    angularDampingOverride=1
    angularSpringConstant=20000
    bAngularSpring=true
    bOverrideAngularDamping=true
    bUseGravityOverride=true
    gravityScale=0.35
    simulationSpace=RootRelative
    alphaBoolBlend {blendOption=Linear}
    boundBone {boneName=dyn_fur_lowerarm_r_06}
    componentLinearAccScale=<1 1 1>
    componentLinearVelScale=<1 1 1>
    gravityOverride {z=-980}
    physicsBodyDefinitions [{
      sphereCollisionRadius=0
      boundBone {boneName=dyn_fur_lowerarm_r_06}
      boxExtents=<20 30 30>
      constraintSetup {angularLimitsMax=<15 15 15>, angularLimitsMin=<-15 -15 -15>, angularTarget=<1 0 0>}
      localJointOffset=<4 0 0>
     }]
   }
   n108 = Anim Dynamics - Bone: dyn_fur_lowerarm_r_07(ComponentPose=@n107, Alpha=1, ExternalForce="0,0,0") {
    angularDampingOverride=1
    angularSpringConstant=20000
    bAngularSpring=true
    bOverrideAngularDamping=true
    bUseGravityOverride=true
    gravityScale=0.35
    simulationSpace=RootRelative
    alphaBoolBlend {blendOption=Linear}
    boundBone {boneName=dyn_fur_lowerarm_r_07}
    componentLinearAccScale=<1 1 1>
    componentLinearVelScale=<1 1 1>
    gravityOverride {z=-980}
    physicsBodyDefinitions [{
      sphereCollisionRadius=0
      boundBone {boneName=dyn_fur_lowerarm_r_07}
      boxExtents=<20 30 30>
      constraintSetup {angularLimitsMax=<15 15 15>, angularLimitsMin=<-15 -15 -15>, angularTarget=<1 0 0>}
      localJointOffset=<4 0 0>
     }]
   }
   n109 = Anim Dynamics - Bone: dyn_fur_lowerarm_r_05(ComponentPose=@n108, Alpha=1, ExternalForce="0,0,0") {
    angularDampingOverride=1
    angularSpringConstant=20000
    bAngularSpring=true
    bOverrideAngularDamping=true
    bUseGravityOverride=true
    gravityScale=0.35
    simulationSpace=RootRelative
    alphaBoolBlend {blendOption=Linear}
    boundBone {boneName=dyn_fur_lowerarm_r_05}
    componentLinearAccScale=<1 1 1>
    componentLinearVelScale=<1 1 1>
    gravityOverride {z=-980}
    physicsBodyDefinitions [{
      sphereCollisionRadius=0
      boundBone {boneName=dyn_fur_lowerarm_r_05}
      boxExtents=<20 30 30>
      constraintSetup {angularLimitsMax=<15 15 15>, angularLimitsMin=<-15 -15 -15>, angularTarget=<1 0 0>}
      localJointOffset=<4 0 0>
     }]
   }
   n110 = Anim Dynamics - Bone: dyn_fur_lowerarm_r_03(ComponentPose=@n109, Alpha=1, ExternalForce="0,0,0") {
    angularDampingOverride=1
    angularSpringConstant=20000
    bAngularSpring=true
    bOverrideAngularDamping=true
    bUseGravityOverride=true
    gravityScale=0.35
    simulationSpace=RootRelative
    alphaBoolBlend {blendOption=Linear}
    boundBone {boneName=dyn_fur_lowerarm_r_03}
    componentLinearAccScale=<1 1 1>
    componentLinearVelScale=<1 1 1>
    gravityOverride {z=-980}
    physicsBodyDefinitions [{
      sphereCollisionRadius=0
      boundBone {boneName=dyn_fur_lowerarm_r_03}
      boxExtents=<25 36 29>
      constraintSetup {angularLimitsMax=<15 15 15>, angularLimitsMin=<-15 -15 -15>, angularTarget=<1 0 0>}
      localJointOffset=<4.5 0 0>
     }]
   }
   n111 = Anim Dynamics - Bone: dyn_fur_lowerarm_r_04(ComponentPose=@n110, Alpha=1, ExternalForce="0,0,0") {
    angularDampingOverride=1
    angularSpringConstant=20000
    bAngularSpring=true
    bOverrideAngularDamping=true
    bUseGravityOverride=true
    gravityScale=0.35
    simulationSpace=RootRelative
    alphaBoolBlend {blendOption=Linear}
    boundBone {boneName=dyn_fur_lowerarm_r_04}
    componentLinearAccScale=<1 1 1>
    componentLinearVelScale=<1 1 1>
    gravityOverride {z=-980}
    physicsBodyDefinitions [{
      sphereCollisionRadius=0
      boundBone {boneName=dyn_fur_lowerarm_r_04}
      boxExtents=<20 30 30>
      constraintSetup {angularLimitsMax=<15 15 15>, angularLimitsMin=<-15 -15 -15>, angularTarget=<1 0 0>}
      localJointOffset=<4 0 0>
     }]
   }
   n112 = Anim Dynamics - Bone: dyn_fur_lowerarm_r_01(ComponentPose=@n111, Alpha=1, ExternalForce="0,0,0") {
    angularDampingOverride=1
    angularSpringConstant=20000
    bAngularSpring=true
    bOverrideAngularDamping=true
    bUseGravityOverride=true
    gravityScale=0.35
    simulationSpace=RootRelative
    alphaBoolBlend {blendOption=Linear}
    boundBone {boneName=dyn_fur_lowerarm_r_01}
    componentLinearAccScale=<1 1 1>
    componentLinearVelScale=<1 1 1>
    gravityOverride {z=-980}
    physicsBodyDefinitions [{
      sphereCollisionRadius=0
      boundBone {boneName=dyn_fur_lowerarm_r_01}
      boxExtents=<20 30 30>
      constraintSetup {angularLimitsMax=<15 15 15>, angularLimitsMin=<-15 -15 -15>, angularTarget=<1 0 0>}
      localJointOffset=<4 0 0>
     }]
   }
   n113 = Anim Dynamics - Bone: dyn_fur_upperarm_r_01(ComponentPose=@n112, Alpha=1, ExternalForce="0,0,0") {
    angularDampingOverride=1
    angularSpringConstant=20000
    bAngularSpring=true
    bOverrideAngularDamping=true
    bUseGravityOverride=true
    gravityScale=0.35
    simulationSpace=RootRelative
    alphaBoolBlend {blendOption=Linear}
    boundBone {boneName=dyn_fur_upperarm_r_01}
    componentLinearAccScale=<1 1 1>
    componentLinearVelScale=<1 1 1>
    gravityOverride {z=-980}
    physicsBodyDefinitions [{
      sphereCollisionRadius=0
      boundBone {boneName=dyn_fur_upperarm_r_01}
      boxExtents=<18 28 37>
      constraintSetup {angularLimitsMax=<15 15 15>, angularLimitsMin=<-15 -15 -15>, angularTarget=<1 0 0>}
      localJointOffset=<3.9 0 0>
     }]
   }
   n114 = Anim Dynamics - Bone: dyn_fur_upperarm_r_02(ComponentPose=@n113, Alpha=1, ExternalForce="0,0,0") {
    angularDampingOverride=1
    angularSpringConstant=20000
    bAngularSpring=true
    bOverrideAngularDamping=true
    bUseGravityOverride=true
    gravityScale=0.35
    simulationSpace=RootRelative
    alphaBoolBlend {blendOption=Linear}
    boundBone {boneName=dyn_fur_upperarm_r_02}
    componentLinearAccScale=<1 1 1>
    componentLinearVelScale=<1 1 1>
    gravityOverride {z=-980}
    physicsBodyDefinitions [{
      sphereCollisionRadius=0
      boundBone {boneName=dyn_fur_upperarm_r_02}
      boxExtents=<20 30 30>
      constraintSetup {angularLimitsMax=<15 15 15>, angularLimitsMin=<-15 -15 -15>, angularTarget=<1 0 0>}
      localJointOffset=<4 0 0>
     }]
   }
   n115 = Anim Dynamics - Bone: dyn_fur_lowerleg_r_09(ComponentPose=@n114, Alpha=1, ExternalForce="0,0,0") {
    angularDampingOverride=1
    angularSpringConstant=20000
    bAngularSpring=true
    bOverrideAngularDamping=true
    bUseGravityOverride=true
    gravityScale=0.35
    simulationSpace=RootRelative
    alphaBoolBlend {blendOption=Linear}
    boundBone {boneName=dyn_fur_lowerleg_r_09}
    componentLinearAccScale=<1 1 1>
    componentLinearVelScale=<1 1 1>
    gravityOverride {z=-980}
    physicsBodyDefinitions [{
      sphereCollisionRadius=0
      boundBone {boneName=dyn_fur_lowerleg_r_09}
      boxExtents=<18 28 37>
      constraintSetup {angularLimitsMax=<15 15 15>, angularLimitsMin=<-15 -15 -15>, angularTarget=<1 0 0>}
      localJointOffset=<3.9 0 0>
     }]
   }
   n116 = Anim Dynamics - Bone: dyn_fur_lowerleg_r_07(ComponentPose=@n115, Alpha=1, ExternalForce="0,0,0") {
    angularDampingOverride=1
    angularSpringConstant=20000
    bAngularSpring=true
    bOverrideAngularDamping=true
    bUseGravityOverride=true
    gravityScale=0.35
    simulationSpace=RootRelative
    alphaBoolBlend {blendOption=Linear}
    boundBone {boneName=dyn_fur_lowerleg_r_07}
    componentLinearAccScale=<1 1 1>
    componentLinearVelScale=<1 1 1>
    gravityOverride {z=-980}
    physicsBodyDefinitions [{
      sphereCollisionRadius=0
      boundBone {boneName=dyn_fur_lowerleg_r_07}
      boxExtents=<20 30 30>
      constraintSetup {angularLimitsMax=<15 15 15>, angularLimitsMin=<-15 -15 -15>, angularTarget=<1 0 0>}
      localJointOffset=<4 0 0>
     }]
   }
   n117 = Anim Dynamics - Bone: dyn_fur_lowerleg_r_08(ComponentPose=@n116, Alpha=1, ExternalForce="0,0,0") {
    angularDampingOverride=1
    angularSpringConstant=20000
    bAngularSpring=true
    bOverrideAngularDamping=true
    bUseGravityOverride=true
    gravityScale=0.35
    simulationSpace=RootRelative
    alphaBoolBlend {blendOption=Linear}
    boundBone {boneName=dyn_fur_lowerleg_r_08}
    componentLinearAccScale=<1 1 1>
    componentLinearVelScale=<1 1 1>
    gravityOverride {z=-980}
    physicsBodyDefinitions [{
      sphereCollisionRadius=0
      boundBone {boneName=dyn_fur_lowerleg_r_08}
      boxExtents=<20 30 30>
      constraintSetup {angularLimitsMax=<15 15 15>, angularLimitsMin=<-15 -15 -15>, angularTarget=<1 0 0>}
      localJointOffset=<4 0 0>
     }]
   }
   n118 = Anim Dynamics - Bone: dyn_fur_lowerleg_r_06(ComponentPose=@n117, Alpha=1, ExternalForce="0,0,0") {
    angularDampingOverride=1
    angularSpringConstant=20000
    bAngularSpring=true
    bOverrideAngularDamping=true
    bUseGravityOverride=true
    gravityScale=0.35
    simulationSpace=RootRelative
    alphaBoolBlend {blendOption=Linear}
    boundBone {boneName=dyn_fur_lowerleg_r_06}
    componentLinearAccScale=<1 1 1>
    componentLinearVelScale=<1 1 1>
    gravityOverride {z=-980}
    physicsBodyDefinitions [{
      sphereCollisionRadius=0
      boundBone {boneName=dyn_fur_lowerleg_r_06}
      boxExtents=<20 30 30>
      constraintSetup {angularLimitsMax=<15 15 15>, angularLimitsMin=<-15 -15 -15>, angularTarget=<1 0 0>}
      localJointOffset=<4 0 0>
     }]
   }
   n119 = Anim Dynamics - Bone: dyn_fur_lowerleg_r_04(ComponentPose=@n118, Alpha=1, ExternalForce="0,0,0") {
    angularDampingOverride=1
    angularSpringConstant=20000
    bAngularSpring=true
    bOverrideAngularDamping=true
    bUseGravityOverride=true
    gravityScale=0.35
    simulationSpace=RootRelative
    alphaBoolBlend {blendOption=Linear}
    boundBone {boneName=dyn_fur_lowerleg_r_04}
    componentLinearAccScale=<1 1 1>
    componentLinearVelScale=<1 1 1>
    gravityOverride {z=-980}
    physicsBodyDefinitions [{
      sphereCollisionRadius=0
      boundBone {boneName=dyn_fur_lowerleg_r_04}
      boxExtents=<25 36 29>
      constraintSetup {angularLimitsMax=<15 15 15>, angularLimitsMin=<-15 -15 -15>, angularTarget=<1 0 0>}
      localJointOffset=<4.5 0 0>
     }]
   }
   n120 = Anim Dynamics - Bone: dyn_fur_lowerleg_r_05(ComponentPose=@n119, Alpha=1, ExternalForce="0,0,0") {
    angularDampingOverride=1
    angularSpringConstant=20000
    bAngularSpring=true
    bOverrideAngularDamping=true
    bUseGravityOverride=true
    gravityScale=0.35
    simulationSpace=RootRelative
    alphaBoolBlend {blendOption=Linear}
    boundBone {boneName=dyn_fur_lowerleg_r_05}
    componentLinearAccScale=<1 1 1>
    componentLinearVelScale=<1 1 1>
    gravityOverride {z=-980}
    physicsBodyDefinitions [{
      sphereCollisionRadius=0
      boundBone {boneName=dyn_fur_lowerleg_r_05}
      boxExtents=<25 36 29>
      constraintSetup {angularLimitsMax=<15 15 15>, angularLimitsMin=<-15 -15 -15>, angularTarget=<1 0 0>}
      localJointOffset=<4.5 0 0>
     }]
   }
   n121 = Anim Dynamics - Bone: dyn_fur_lowerleg_r_03(ComponentPose=@n120, Alpha=1, ExternalForce="0,0,0") {
    angularDampingOverride=1
    angularSpringConstant=20000
    bAngularSpring=true
    bOverrideAngularDamping=true
    bUseGravityOverride=true
    gravityScale=0.35
    simulationSpace=RootRelative
    alphaBoolBlend {blendOption=Linear}
    boundBone {boneName=dyn_fur_lowerleg_r_03}
    componentLinearAccScale=<1 1 1>
    componentLinearVelScale=<1 1 1>
    gravityOverride {z=-980}
    physicsBodyDefinitions [{
      sphereCollisionRadius=0
      boundBone {boneName=dyn_fur_lowerleg_r_03}
      boxExtents=<20 30 30>
      constraintSetup {angularLimitsMax=<15 15 15>, angularLimitsMin=<-15 -15 -15>, angularTarget=<1 0 0>}
      localJointOffset=<4 0 0>
     }]
   }
   n122 = Anim Dynamics - Bone: dyn_fur_lowerleg_r_01(ComponentPose=@n121, Alpha=1, ExternalForce="0,0,0") {
    angularDampingOverride=1
    angularSpringConstant=20000
    bAngularSpring=true
    bOverrideAngularDamping=true
    bUseGravityOverride=true
    gravityScale=0.35
    simulationSpace=RootRelative
    alphaBoolBlend {blendOption=Linear}
    boundBone {boneName=dyn_fur_lowerleg_r_01}
    componentLinearAccScale=<1 1 1>
    componentLinearVelScale=<1 1 1>
    gravityOverride {z=-980}
    physicsBodyDefinitions [{
      sphereCollisionRadius=0
      boundBone {boneName=dyn_fur_lowerleg_r_01}
      boxExtents=<20 30 30>
      constraintSetup {angularLimitsMax=<15 15 15>, angularLimitsMin=<-15 -15 -15>, angularTarget=<1 0 0>}
      localJointOffset=<4 0 0>
     }]
   }
   n123 = Anim Dynamics - Bone: dyn_fur_lowerleg_r_02(ComponentPose=@n122, Alpha=1, ExternalForce="0,0,0") {
    angularDampingOverride=1
    angularSpringConstant=20000
    bAngularSpring=true
    bOverrideAngularDamping=true
    bUseGravityOverride=true
    gravityScale=0.35
    simulationSpace=RootRelative
    alphaBoolBlend {blendOption=Linear}
    boundBone {boneName=dyn_fur_lowerleg_r_02}
    componentLinearAccScale=<1 1 1>
    componentLinearVelScale=<1 1 1>
    gravityOverride {z=-980}
    physicsBodyDefinitions [{
      sphereCollisionRadius=0
      boundBone {boneName=dyn_fur_lowerleg_r_02}
      boxExtents=<18 28 37>
      constraintSetup {angularLimitsMax=<15 15 15>, angularLimitsMin=<-15 -15 -15>, angularTarget=<1 0 0>}
      localJointOffset=<3.9 0 0>
     }]
   }
   n124 = Anim Dynamics - Bone: dyn_fur_lowerarm_l_10(ComponentPose=@n123, Alpha=1, ExternalForce="0,0,0") {
    angularDampingOverride=1
    angularSpringConstant=20000
    bAngularSpring=true
    bOverrideAngularDamping=true
    bUseGravityOverride=true
    gravityScale=0.35
    simulationSpace=RootRelative
    alphaBoolBlend {blendOption=Linear}
    boundBone {boneName=dyn_fur_lowerarm_l_10}
    componentLinearAccScale=<1 1 1>
    componentLinearVelScale=<1 1 1>
    gravityOverride {z=-980}
    physicsBodyDefinitions [{
      sphereCollisionRadius=0
      boundBone {boneName=dyn_fur_lowerarm_l_10}
      boxExtents=<25 36 29>
      constraintSetup {angularLimitsMax=<15 15 15>, angularLimitsMin=<-15 -15 -15>, angularTarget=<1 0 0>}
      localJointOffset=<4.5 0 0>
     }]
   }
   n125 = Anim Dynamics - Bone: dyn_fur_lowerarm_l_08(ComponentPose=@n124, Alpha=1, ExternalForce="0,0,0") {
    angularDampingOverride=1
    angularSpringConstant=20000
    bAngularSpring=true
    bOverrideAngularDamping=true
    bUseGravityOverride=true
    gravityScale=0.35
    simulationSpace=RootRelative
    alphaBoolBlend {blendOption=Linear}
    boundBone {boneName=dyn_fur_lowerarm_l_08}
    componentLinearAccScale=<1 1 1>
    componentLinearVelScale=<1 1 1>
    gravityOverride {z=-980}
    physicsBodyDefinitions [{
      sphereCollisionRadius=0
      boundBone {boneName=dyn_fur_lowerarm_l_08}
      boxExtents=<20 30 30>
      constraintSetup {angularLimitsMax=<15 15 15>, angularLimitsMin=<-15 -15 -15>, angularTarget=<1 0 0>}
      localJointOffset=<4 0 0>
     }]
   }
   n126 = Anim Dynamics - Bone: dyn_fur_lowerarm_l_09(ComponentPose=@n125, Alpha=1, ExternalForce="0,0,0") {
    angularDampingOverride=1
    angularSpringConstant=20000
    bAngularSpring=true
    bOverrideAngularDamping=true
    bUseGravityOverride=true
    gravityScale=0.35
    simulationSpace=RootRelative
    alphaBoolBlend {blendOption=Linear}
    boundBone {boneName=dyn_fur_lowerarm_l_09}
    componentLinearAccScale=<1 1 1>
    componentLinearVelScale=<1 1 1>
    gravityOverride {z=-980}
    physicsBodyDefinitions [{
      sphereCollisionRadius=0
      boundBone {boneName=dyn_fur_lowerarm_l_09}
      boxExtents=<20 30 30>
      constraintSetup {angularLimitsMax=<15 15 15>, angularLimitsMin=<-15 -15 -15>, angularTarget=<1 0 0>}
      localJointOffset=<4 0 0>
     }]
   }
   n127 = Anim Dynamics - Bone: dyn_fur_lowerarm_l_07(ComponentPose=@n126, Alpha=1, ExternalForce="0,0,0") {
    angularDampingOverride=1
    angularSpringConstant=20000
    bAngularSpring=true
    bOverrideAngularDamping=true
    bUseGravityOverride=true
    gravityScale=0.35
    simulationSpace=RootRelative
    alphaBoolBlend {blendOption=Linear}
    boundBone {boneName=dyn_fur_lowerarm_l_07}
    componentLinearAccScale=<1 1 1>
    componentLinearVelScale=<1 1 1>
    gravityOverride {z=-980}
    physicsBodyDefinitions [{
      sphereCollisionRadius=0
      boundBone {boneName=dyn_fur_lowerarm_l_07}
      boxExtents=<20 30 30>
      constraintSetup {angularLimitsMax=<15 15 15>, angularLimitsMin=<-15 -15 -15>, angularTarget=<1 0 0>}
      localJointOffset=<4 0 0>
     }]
   }
   n128 = Anim Dynamics - Bone: dyn_fur_lowerarm_l_05(ComponentPose=@n127, Alpha=1, ExternalForce="0,0,0") {
    angularDampingOverride=1
    angularSpringConstant=20000
    bAngularSpring=true
    bOverrideAngularDamping=true
    bUseGravityOverride=true
    gravityScale=0.35
    simulationSpace=RootRelative
    alphaBoolBlend {blendOption=Linear}
    boundBone {boneName=dyn_fur_lowerarm_l_05}
    componentLinearAccScale=<1 1 1>
    componentLinearVelScale=<1 1 1>
    gravityOverride {z=-980}
    physicsBodyDefinitions [{
      sphereCollisionRadius=0
      boundBone {boneName=dyn_fur_lowerarm_l_05}
      boxExtents=<18 28 37>
      constraintSetup {angularLimitsMax=<15 15 15>, angularLimitsMin=<-15 -15 -15>, angularTarget=<1 0 0>}
      localJointOffset=<3.9 0 0>
     }]
   }
   n129 = Anim Dynamics - Bone: dyn_fur_lowerarm_l_06(ComponentPose=@n128, Alpha=1, ExternalForce="0,0,0") {
    angularDampingOverride=1
    angularSpringConstant=20000
    bAngularSpring=true
    bOverrideAngularDamping=true
    bUseGravityOverride=true
    gravityScale=0.35
    simulationSpace=RootRelative
    alphaBoolBlend {blendOption=Linear}
    boundBone {boneName=dyn_fur_lowerarm_l_06}
    componentLinearAccScale=<1 1 1>
    componentLinearVelScale=<1 1 1>
    gravityOverride {z=-980}
    physicsBodyDefinitions [{
      sphereCollisionRadius=0
      boundBone {boneName=dyn_fur_lowerarm_l_06}
      boxExtents=<20 30 30>
      constraintSetup {angularLimitsMax=<15 15 15>, angularLimitsMin=<-15 -15 -15>, angularTarget=<1 0 0>}
      localJointOffset=<4 0 0>
     }]
   }
   n130 = Anim Dynamics - Bone: dyn_fur_lowerarm_l_04(ComponentPose=@n129, Alpha=1, ExternalForce="0,0,0") {
    angularDampingOverride=1
    angularSpringConstant=20000
    bAngularSpring=true
    bOverrideAngularDamping=true
    bUseGravityOverride=true
    gravityScale=0.35
    simulationSpace=RootRelative
    alphaBoolBlend {blendOption=Linear}
    boundBone {boneName=dyn_fur_lowerarm_l_04}
    componentLinearAccScale=<1 1 1>
    componentLinearVelScale=<1 1 1>
    gravityOverride {z=-980}
    physicsBodyDefinitions [{
      sphereCollisionRadius=0
      boundBone {boneName=dyn_fur_lowerarm_l_04}
      boxExtents=<18 28 37>
      constraintSetup {angularLimitsMax=<15 15 15>, angularLimitsMin=<-15 -15 -15>, angularTarget=<1 0 0>}
      localJointOffset=<3.9 0 0>
     }]
   }
   n131 = Anim Dynamics - Bone: dyn_fur_lowerarm_l_02(ComponentPose=@n130, Alpha=1, ExternalForce="0,0,0") {
    angularDampingOverride=1
    angularSpringConstant=20000
    bAngularSpring=true
    bOverrideAngularDamping=true
    bUseGravityOverride=true
    gravityScale=0.35
    simulationSpace=RootRelative
    alphaBoolBlend {blendOption=Linear}
    boundBone {boneName=dyn_fur_lowerarm_l_02}
    componentLinearAccScale=<1 1 1>
    componentLinearVelScale=<1 1 1>
    gravityOverride {z=-980}
    physicsBodyDefinitions [{
      sphereCollisionRadius=0
      boundBone {boneName=dyn_fur_lowerarm_l_02}
      boxExtents=<25 36 29>
      constraintSetup {angularLimitsMax=<15 15 15>, angularLimitsMin=<-15 -15 -15>, angularTarget=<1 0 0>}
      localJointOffset=<4.5 0 0>
     }]
   }
   n132 = Anim Dynamics - Bone: dyn_fur_lowerarm_l_03(ComponentPose=@n131, Alpha=1, ExternalForce="0,0,0") {
    angularDampingOverride=1
    angularSpringConstant=20000
    bAngularSpring=true
    bOverrideAngularDamping=true
    bUseGravityOverride=true
    gravityScale=0.35
    simulationSpace=RootRelative
    alphaBoolBlend {blendOption=Linear}
    boundBone {boneName=dyn_fur_lowerarm_l_03}
    componentLinearAccScale=<1 1 1>
    componentLinearVelScale=<1 1 1>
    gravityOverride {z=-980}
    physicsBodyDefinitions [{
      sphereCollisionRadius=0
      boundBone {boneName=dyn_fur_lowerarm_l_03}
      boxExtents=<18 28 37>
      constraintSetup {angularLimitsMax=<15 15 15>, angularLimitsMin=<-15 -15 -15>, angularTarget=<1 0 0>}
      localJointOffset=<3.9 0 0>
     }]
   }
   n133 = Anim Dynamics - Bone: dyn_fur_lowerarm_l_01(ComponentPose=@n132, Alpha=1, ExternalForce="0,0,0") {
    angularDampingOverride=1
    angularSpringConstant=20000
    bAngularSpring=true
    bOverrideAngularDamping=true
    bUseGravityOverride=true
    gravityScale=0.35
    simulationSpace=RootRelative
    alphaBoolBlend {blendOption=Linear}
    boundBone {boneName=dyn_fur_lowerarm_l_01}
    componentLinearAccScale=<1 1 1>
    componentLinearVelScale=<1 1 1>
    gravityOverride {z=-980}
    physicsBodyDefinitions [{
      sphereCollisionRadius=0
      boundBone {boneName=dyn_fur_lowerarm_l_01}
      boxExtents=<20 30 30>
      constraintSetup {angularLimitsMax=<15 15 15>, angularLimitsMin=<-15 -15 -15>, angularTarget=<1 0 0>}
      localJointOffset=<4 0 0>
     }]
   }
   n134 = Anim Dynamics - Bone: dyn_fur_upperarm_l_01(ComponentPose=@n133, Alpha=1, ExternalForce="0,0,0") {
    angularDampingOverride=1
    angularSpringConstant=20000
    bAngularSpring=true
    bOverrideAngularDamping=true
    bUseGravityOverride=true
    gravityScale=0.35
    simulationSpace=RootRelative
    alphaBoolBlend {blendOption=Linear}
    boundBone {boneName=dyn_fur_upperarm_l_01}
    componentLinearAccScale=<1 1 1>
    componentLinearVelScale=<1 1 1>
    gravityOverride {z=-980}
    physicsBodyDefinitions [{
      sphereCollisionRadius=0
      boundBone {boneName=dyn_fur_upperarm_l_01}
      boxExtents=<20 30 30>
      constraintSetup {angularLimitsMax=<15 15 15>, angularLimitsMin=<-15 -15 -15>, angularTarget=<1 0 0>}
      localJointOffset=<4 0 0>
     }]
   }
   n135 = Anim Dynamics - Bone: dyn_fur_upperarm_l_02(ComponentPose=@n134, Alpha=1, ExternalForce="0,0,0") {
    angularDampingOverride=1
    angularSpringConstant=20000
    bAngularSpring=true
    bOverrideAngularDamping=true
    bUseGravityOverride=true
    gravityScale=0.35
    simulationSpace=RootRelative
    alphaBoolBlend {blendOption=Linear}
    boundBone {boneName=dyn_fur_upperarm_l_02}
    componentLinearAccScale=<1 1 1>
    componentLinearVelScale=<1 1 1>
    gravityOverride {z=-980}
    physicsBodyDefinitions [{
      sphereCollisionRadius=0
      boundBone {boneName=dyn_fur_upperarm_l_02}
      boxExtents=<20 30 30>
      constraintSetup {angularLimitsMax=<15 15 15>, angularLimitsMin=<-15 -15 -15>, angularTarget=<1 0 0>}
      localJointOffset=<4 0 0>
     }]
   }
   n136 = Anim Dynamics - Bone: dyn_fur_torso_r_02(ComponentPose=@n135, Alpha=1, ExternalForce="0,0,0") {
    angularDampingOverride=1
    angularSpringConstant=20000
    bAngularSpring=true
    bOverrideAngularDamping=true
    bUseGravityOverride=true
    gravityScale=0.35
    simulationSpace=RootRelative
    alphaBoolBlend {blendOption=Linear}
    boundBone {boneName=dyn_fur_torso_r_02}
    componentLinearAccScale=<1 1 1>
    componentLinearVelScale=<1 1 1>
    gravityOverride {z=-980}
    physicsBodyDefinitions [{
      sphereCollisionRadius=0
      boundBone {boneName=dyn_fur_torso_r_02}
      boxExtents=<20 30 30>
      constraintSetup {angularLimitsMax=<15 15 15>, angularLimitsMin=<-15 -15 -15>, angularTarget=<1 0 0>}
      localJointOffset=<4 0 0>
     }]
   }
   n137 = Anim Dynamics - Bone: dyn_fur_torso_l_02(ComponentPose=@n136, Alpha=1, ExternalForce="0,0,0") {
    angularDampingOverride=1
    angularSpringConstant=20000
    bAngularSpring=true
    bOverrideAngularDamping=true
    bUseGravityOverride=true
    gravityScale=0.35
    simulationSpace=RootRelative
    alphaBoolBlend {blendOption=Linear}
    boundBone {boneName=dyn_fur_torso_l_02}
    componentLinearAccScale=<1 1 1>
    componentLinearVelScale=<1 1 1>
    gravityOverride {z=-980}
    physicsBodyDefinitions [{
      sphereCollisionRadius=0
      boundBone {boneName=dyn_fur_torso_l_02}
      boxExtents=<20 30 30>
      constraintSetup {angularLimitsMax=<15 15 15>, angularLimitsMin=<-15 -15 -15>, angularTarget=<1 0 0>}
      localJointOffset=<4 0 0>
     }]
   }
   n138 = Anim Dynamics - Bone: dyn_fur_torso_r_01(ComponentPose=@n137, Alpha=1, ExternalForce="0,0,0") {
    angularDampingOverride=1
    angularSpringConstant=20000
    bAngularSpring=true
    bOverrideAngularDamping=true
    bUseGravityOverride=true
    gravityScale=0.35
    simulationSpace=RootRelative
    alphaBoolBlend {blendOption=Linear}
    boundBone {boneName=dyn_fur_torso_r_01}
    componentLinearAccScale=<1 1 1>
    componentLinearVelScale=<1 1 1>
    gravityOverride {z=-980}
    physicsBodyDefinitions [{
      sphereCollisionRadius=0
      boundBone {boneName=dyn_fur_torso_r_01}
      boxExtents=<25 36 29>
      constraintSetup {angularLimitsMax=<15 15 15>, angularLimitsMin=<-15 -15 -15>, angularTarget=<1 0 0>}
      localJointOffset=<4.5 0 0>
     }]
   }
   n139 = Anim Dynamics - Bone: dyn_fur_torso_l_01(ComponentPose=@n138, Alpha=1, ExternalForce="0,0,0") {
    angularDampingOverride=1
    angularSpringConstant=20000
    bAngularSpring=true
    bOverrideAngularDamping=true
    bUseGravityOverride=true
    gravityScale=0.35
    simulationSpace=RootRelative
    alphaBoolBlend {blendOption=Linear}
    boundBone {boneName=dyn_fur_torso_l_01}
    componentLinearAccScale=<1 1 1>
    componentLinearVelScale=<1 1 1>
    gravityOverride {z=-980}
    physicsBodyDefinitions [{
      sphereCollisionRadius=0
      boundBone {boneName=dyn_fur_torso_l_01}
      boxExtents=<20 30 30>
      constraintSetup {angularLimitsMax=<15 15 15>, angularLimitsMin=<-15 -15 -15>, angularTarget=<1 0 0>}
      localJointOffset=<4 0 0>
     }]
   }
   n140 = Anim Dynamics - Bone: dyn_fur_torso_fr_03(ComponentPose=@n139, Alpha=1, ExternalForce="0,0,0") {
    angularDampingOverride=1
    angularSpringConstant=20000
    bAngularSpring=true
    bOverrideAngularDamping=true
    bUseGravityOverride=true
    gravityScale=0.35
    simulationSpace=RootRelative
    alphaBoolBlend {blendOption=Linear}
    boundBone {boneName=dyn_fur_torso_fr_03}
    componentLinearAccScale=<1 1 1>
    componentLinearVelScale=<1 1 1>
    gravityOverride {z=-980}
    physicsBodyDefinitions [{
      sphereCollisionRadius=0
      boundBone {boneName=dyn_fur_torso_fr_03}
      boxExtents=<25 36 29>
      constraintSetup {angularLimitsMax=<15 15 15>, angularLimitsMin=<-15 -15 -15>, angularTarget=<1 0 0>}
      localJointOffset=<4.5 0 0>
     }]
   }
   n141 = Anim Dynamics - Bone: dyn_fur_torso_fr_04(ComponentPose=@n140, Alpha=1, ExternalForce="0,0,0") {
    angularDampingOverride=1
    angularSpringConstant=20000
    bAngularSpring=true
    bOverrideAngularDamping=true
    bUseGravityOverride=true
    gravityScale=0.35
    simulationSpace=RootRelative
    alphaBoolBlend {blendOption=Linear}
    boundBone {boneName=dyn_fur_torso_fr_04}
    componentLinearAccScale=<1 1 1>
    componentLinearVelScale=<1 1 1>
    gravityOverride {z=-980}
    physicsBodyDefinitions [{
      sphereCollisionRadius=0
      boundBone {boneName=dyn_fur_torso_fr_04}
      boxExtents=<20 30 30>
      constraintSetup {angularLimitsMax=<15 15 15>, angularLimitsMin=<-15 -15 -15>, angularTarget=<1 0 0>}
      localJointOffset=<4 0 0>
     }]
   }
   n142 = Anim Dynamics - Bone: dyn_fur_torso_fr_02(ComponentPose=@n141, Alpha=1, ExternalForce="0,0,0") {
    angularDampingOverride=1
    angularSpringConstant=20000
    bAngularSpring=true
    bOverrideAngularDamping=true
    bUseGravityOverride=true
    gravityScale=0.35
    simulationSpace=RootRelative
    alphaBoolBlend {blendOption=Linear}
    boundBone {boneName=dyn_fur_torso_fr_02}
    componentLinearAccScale=<1 1 1>
    componentLinearVelScale=<1 1 1>
    gravityOverride {z=-980}
    physicsBodyDefinitions [{
      sphereCollisionRadius=0
      boundBone {boneName=dyn_fur_torso_fr_02}
      boxExtents=<25 36 29>
      constraintSetup {angularLimitsMax=<15 15 15>, angularLimitsMin=<-15 -15 -15>, angularTarget=<1 0 0>}
      localJointOffset=<4.5 0 0>
     }]
   }
   n143 = Anim Dynamics - Bone: dyn_fur_torso_bk_02(ComponentPose=@n142, Alpha=1, ExternalForce="0,0,0") {
    angularDampingOverride=1
    angularSpringConstant=20000
    bAngularSpring=true
    bOverrideAngularDamping=true
    bUseGravityOverride=true
    gravityScale=0.35
    simulationSpace=RootRelative
    alphaBoolBlend {blendOption=Linear}
    boundBone {boneName=dyn_fur_torso_bk_02}
    componentLinearAccScale=<1 1 1>
    componentLinearVelScale=<1 1 1>
    gravityOverride {z=-980}
    physicsBodyDefinitions [{
      sphereCollisionRadius=0
      boundBone {boneName=dyn_fur_torso_bk_02}
      boxExtents=<20 30 30>
      constraintSetup {angularLimitsMax=<15 15 15>, angularLimitsMin=<-15 -15 -15>, angularTarget=<1 0 0>}
      localJointOffset=<4 0 0>
     }]
   }
   n144 = Anim Dynamics - Bone: dyn_fur_torso_fr_01(ComponentPose=@n143, Alpha=1, ExternalForce="0,0,0") {
    angularDampingOverride=1
    angularSpringConstant=20000
    bAngularSpring=true
    bOverrideAngularDamping=true
    bUseGravityOverride=true
    gravityScale=0.35
    simulationSpace=RootRelative
    alphaBoolBlend {blendOption=Linear}
    boundBone {boneName=dyn_fur_torso_fr_01}
    componentLinearAccScale=<1 1 1>
    componentLinearVelScale=<1 1 1>
    gravityOverride {z=-980}
    physicsBodyDefinitions [{
      sphereCollisionRadius=0
      boundBone {boneName=dyn_fur_torso_fr_01}
      boxExtents=<18 28 37>
      constraintSetup {angularLimitsMax=<15 15 15>, angularLimitsMin=<-15 -15 -15>, angularTarget=<1 0 0>}
      localJointOffset=<3.9 0 0>
     }]
   }
   n145 = Anim Dynamics - Bone: dyn_fur_torso_bk_01(ComponentPose=@n144, Alpha=1, ExternalForce="0,0,0") {
    angularDampingOverride=1
    angularSpringConstant=20000
    bAngularSpring=true
    bOverrideAngularDamping=true
    bUseGravityOverride=true
    gravityScale=0.35
    simulationSpace=RootRelative
    alphaBoolBlend {blendOption=Linear}
    boundBone {boneName=dyn_fur_torso_bk_01}
    componentLinearAccScale=<1 1 1>
    componentLinearVelScale=<1 1 1>
    gravityOverride {z=-980}
    physicsBodyDefinitions [{
      sphereCollisionRadius=0
      boundBone {boneName=dyn_fur_torso_bk_01}
      boxExtents=<18 28 37>
      constraintSetup {angularLimitsMax=<15 15 15>, angularLimitsMin=<-15 -15 -15>, angularTarget=<1 0 0>}
      localJointOffset=<3.9 0 0>
     }]
   }
   n146 = Anim Dynamics - Bone: dyn_fur_lowerleg_l_08(ComponentPose=@n145, Alpha=1, ExternalForce="0,0,0") {
    angularDampingOverride=1
    angularSpringConstant=20000
    bAngularSpring=true
    bOverrideAngularDamping=true
    bUseGravityOverride=true
    gravityScale=0.35
    simulationSpace=RootRelative
    alphaBoolBlend {blendOption=Linear}
    boundBone {boneName=dyn_fur_lowerleg_l_08}
    componentLinearAccScale=<1 1 1>
    componentLinearVelScale=<1 1 1>
    gravityOverride {z=-980}
    physicsBodyDefinitions [{
      sphereCollisionRadius=0
      boundBone {boneName=dyn_fur_lowerleg_l_08}
      boxExtents=<20 30 30>
      constraintSetup {angularLimitsMax=<15 15 15>, angularLimitsMin=<-15 -15 -15>, angularTarget=<1 0 0>}
      localJointOffset=<4 0 0>
     }]
   }
   n147 = Anim Dynamics - Bone: dyn_fur_lowerleg_l_09(ComponentPose=@n146, Alpha=1, ExternalForce="0,0,0") {
    angularDampingOverride=1
    angularSpringConstant=20000
    bAngularSpring=true
    bOverrideAngularDamping=true
    bUseGravityOverride=true
    gravityScale=0.35
    simulationSpace=RootRelative
    alphaBoolBlend {blendOption=Linear}
    boundBone {boneName=dyn_fur_lowerleg_l_09}
    componentLinearAccScale=<1 1 1>
    componentLinearVelScale=<1 1 1>
    gravityOverride {z=-980}
    physicsBodyDefinitions [{
      sphereCollisionRadius=0
      boundBone {boneName=dyn_fur_lowerleg_l_09}
      boxExtents=<25 36 29>
      constraintSetup {angularLimitsMax=<15 15 15>, angularLimitsMin=<-15 -15 -15>, angularTarget=<1 0 0>}
      localJointOffset=<4.5 0 0>
     }]
   }
   n148 = Anim Dynamics - Bone: dyn_fur_lowerleg_l_07(ComponentPose=@n147, Alpha=1, ExternalForce="0,0,0") {
    angularDampingOverride=1
    angularSpringConstant=20000
    bAngularSpring=true
    bOverrideAngularDamping=true
    bUseGravityOverride=true
    gravityScale=0.35
    simulationSpace=RootRelative
    alphaBoolBlend {blendOption=Linear}
    boundBone {boneName=dyn_fur_lowerleg_l_07}
    componentLinearAccScale=<1 1 1>
    componentLinearVelScale=<1 1 1>
    gravityOverride {z=-980}
    physicsBodyDefinitions [{
      sphereCollisionRadius=0
      boundBone {boneName=dyn_fur_lowerleg_l_07}
      boxExtents=<20 30 30>
      constraintSetup {angularLimitsMax=<15 15 15>, angularLimitsMin=<-15 -15 -15>, angularTarget=<1 0 0>}
      localJointOffset=<4 0 0>
     }]
   }
   n149 = Anim Dynamics - Bone: dyn_fur_lowerleg_l_05(ComponentPose=@n148, Alpha=1, ExternalForce="0,0,0") {
    angularDampingOverride=1
    angularSpringConstant=20000
    bAngularSpring=true
    bOverrideAngularDamping=true
    bUseGravityOverride=true
    gravityScale=0.35
    simulationSpace=RootRelative
    alphaBoolBlend {blendOption=Linear}
    boundBone {boneName=dyn_fur_lowerleg_l_05}
    componentLinearAccScale=<1 1 1>
    componentLinearVelScale=<1 1 1>
    gravityOverride {z=-980}
    physicsBodyDefinitions [{
      sphereCollisionRadius=0
      boundBone {boneName=dyn_fur_lowerleg_l_05}
      boxExtents=<20 30 30>
      constraintSetup {angularLimitsMax=<15 15 15>, angularLimitsMin=<-15 -15 -15>, angularTarget=<1 0 0>}
      localJointOffset=<4 0 0>
     }]
   }
   n150 = Anim Dynamics - Bone: dyn_fur_lowerleg_l_06(ComponentPose=@n149, Alpha=1, ExternalForce="0,0,0") {
    angularDampingOverride=1
    angularSpringConstant=20000
    bAngularSpring=true
    bOverrideAngularDamping=true
    bUseGravityOverride=true
    gravityScale=0.35
    simulationSpace=RootRelative
    alphaBoolBlend {blendOption=Linear}
    boundBone {boneName=dyn_fur_lowerleg_l_06}
    componentLinearAccScale=<1 1 1>
    componentLinearVelScale=<1 1 1>
    gravityOverride {z=-980}
    physicsBodyDefinitions [{
      sphereCollisionRadius=0
      boundBone {boneName=dyn_fur_lowerleg_l_06}
      boxExtents=<25 36 29>
      constraintSetup {angularLimitsMax=<15 15 15>, angularLimitsMin=<-15 -15 -15>, angularTarget=<1 0 0>}
      localJointOffset=<4.5 0 0>
     }]
   }
   n151 = Anim Dynamics - Bone: dyn_fur_lowerleg_l_04(ComponentPose=@n150, Alpha=1, ExternalForce="0,0,0") {
    angularDampingOverride=1
    angularSpringConstant=20000
    bAngularSpring=true
    bOverrideAngularDamping=true
    bUseGravityOverride=true
    gravityScale=0.35
    simulationSpace=RootRelative
    alphaBoolBlend {blendOption=Linear}
    boundBone {boneName=dyn_fur_lowerleg_l_04}
    componentLinearAccScale=<1 1 1>
    componentLinearVelScale=<1 1 1>
    gravityOverride {z=-980}
    physicsBodyDefinitions [{
      sphereCollisionRadius=0
      boundBone {boneName=dyn_fur_lowerleg_l_04}
      boxExtents=<20 30 30>
      constraintSetup {angularLimitsMax=<15 15 15>, angularLimitsMin=<-15 -15 -15>, angularTarget=<1 0 0>}
      localJointOffset=<4 0 0>
     }]
   }
   n152 = Anim Dynamics - Bone: dyn_fur_lowerleg_l_02(ComponentPose=@n151, Alpha=1, ExternalForce="0,0,0") {
    angularDampingOverride=1
    angularSpringConstant=20000
    bAngularSpring=true
    bOverrideAngularDamping=true
    bUseGravityOverride=true
    gravityScale=0.35
    simulationSpace=RootRelative
    alphaBoolBlend {blendOption=Linear}
    boundBone {boneName=dyn_fur_lowerleg_l_02}
    componentLinearAccScale=<1 1 1>
    componentLinearVelScale=<1 1 1>
    gravityOverride {z=-980}
    physicsBodyDefinitions [{
      sphereCollisionRadius=0
      boundBone {boneName=dyn_fur_lowerleg_l_02}
      boxExtents=<25 36 29>
      constraintSetup {angularLimitsMax=<15 15 15>, angularLimitsMin=<-15 -15 -15>, angularTarget=<1 0 0>}
      localJointOffset=<4.5 0 0>
     }]
   }
   n153 = Anim Dynamics - Bone: dyn_fur_lowerleg_l_03(ComponentPose=@n152, Alpha=1, ExternalForce="0,0,0") {
    angularDampingOverride=1
    angularSpringConstant=20000
    bAngularSpring=true
    bOverrideAngularDamping=true
    bUseGravityOverride=true
    gravityScale=0.35
    simulationSpace=RootRelative
    alphaBoolBlend {blendOption=Linear}
    boundBone {boneName=dyn_fur_lowerleg_l_03}
    componentLinearAccScale=<1 1 1>
    componentLinearVelScale=<1 1 1>
    gravityOverride {z=-980}
    physicsBodyDefinitions [{
      sphereCollisionRadius=0
      boundBone {boneName=dyn_fur_lowerleg_l_03}
      boxExtents=<18 28 37>
      constraintSetup {angularLimitsMax=<15 15 15>, angularLimitsMin=<-15 -15 -15>, angularTarget=<1 0 0>}
      localJointOffset=<3.9 0 0>
     }]
   }
   n154 = Anim Dynamics - Bone: dyn_fur_lowerleg_l_01(ComponentPose=@n153, Alpha=1, ExternalForce="0,0,0") {
    angularDampingOverride=1
    angularSpringConstant=20000
    bAngularSpring=true
    bOverrideAngularDamping=true
    bUseGravityOverride=true
    gravityScale=0.35
    simulationSpace=RootRelative
    alphaBoolBlend {blendOption=Linear}
    boundBone {boneName=dyn_fur_lowerleg_l_01}
    componentLinearAccScale=<1 1 1>
    componentLinearVelScale=<1 1 1>
    gravityOverride {z=-980}
    physicsBodyDefinitions [{
      sphereCollisionRadius=0
      boundBone {boneName=dyn_fur_lowerleg_l_01}
      boxExtents=<25 36 29>
      constraintSetup {angularLimitsMax=<15 15 15>, angularLimitsMin=<-15 -15 -15>, angularTarget=<1 0 0>}
      localJointOffset=<4.5 0 0>
     }]
   }
   n155 = Anim Dynamics - Bone: dyn_pouch_bk_r(ComponentPose=@n154, Alpha=1, ExternalForce="0,0,0") {
    angularDampingOverride=1
    angularSpringConstant=25000
    bAngularSpring=true
    bOverrideAngularDamping=true
    bUseGravityOverride=true
    gravityScale=0.35
    simulationSpace=RootRelative
    alphaBoolBlend {blendOption=Linear}
    boundBone {boneName=dyn_pouch_bk_r}
    componentLinearAccScale=<1 1 1>
    componentLinearVelScale=<1 1 1>
    gravityOverride {z=-980}
    physicsBodyDefinitions [{
      sphereCollisionRadius=0
      boundBone {boneName=dyn_pouch_bk_r}
      boxExtents=<10 20 20>
      constraintSetup {angularLimitsMax=<5 0 5>, angularLimitsMin=<-5 0 -5>, angularTarget=<1 0 0>}
      localJointOffset=<4 0 0>
     }]
   }
   n156 = RigidBody!(ComponentPose=@n155, Alpha, OverrideWorldGravity:Vector=Get GravityOverride(), ExternalForce) -> (Pose) {
    bClampLinearTranslationLimitToRefPose=true
    bOverrideWorldGravity=true
    overridePhysicsAsset=&/M_MED_Solstice_Dyn_scalps
    simulationSpace=BaseBoneSpace
    alphaBoolBlend {blendOption=Linear}
    baseBoneRef {boneName=pelvis}
    componentAppliedLinearAccClamp=<3000 3000 3000>
    componentLinearAccScale {x=0.4}
    componentLinearVelScale {x=0.4}
    overrideWorldGravity {z=-980}
    simSpaceSettings {
     maxAngularAcceleration=500
     maxAngularVelocity=3
     maxLinearAcceleration=700
     maxLinearVelocity=500
     velocityScaleZ=0
     worldAlpha=0.5
     externalLinearDragV=<1 1>
    }
   }
   n157 = Get Crouching_Alpha() -> (Crouching_Alpha:double)
   n158 = "Transform (Modify) Bone - Bone: dyn_scalp_fr_r_01_MSTR"(ComponentPose=@n156, Alpha=@n157, Translation="1.034,7.260,0.207", Rotation="-3.323,12.876,9.904", Scale="1,1,1") {
    rotationMode=BMM_Additive
    rotationSpace=BCS_BoneSpace
    translationMode=BMM_Additive
    translationSpace=BCS_BoneSpace
    alphaBoolBlend {blendOption=Linear}
    alphaScaleBiasClamp {bInterpResult=true, interpSpeedDecreasing=5, interpSpeedIncreasing=5}
    boneToModify {boneName=dyn_scalp_fr_r_01_MSTR}
    rotation=<rot -3.323 12.876 9.904>
    translation=<1.034 7.26 0.207>
   }
   n159 = "Transform (Modify) Bone - Bone: dyn_scalp_bk_r_01_MSTR"(ComponentPose=@n158, Alpha=@n157, Translation="1.077,7.091,-0.533", Rotation="-0.516,16.270,12.055", Scale="1,1,1") {
    rotationMode=BMM_Additive
    rotationSpace=BCS_BoneSpace
    translationMode=BMM_Additive
    translationSpace=BCS_BoneSpace
    alphaBoolBlend {blendOption=Linear}
    alphaScaleBiasClamp {bInterpResult=true, interpSpeedDecreasing=5, interpSpeedIncreasing=5}
    boneToModify {boneName=dyn_scalp_bk_r_01_MSTR}
    rotation=<rot -0.516 16.2701 12.0551>
    translation=<1.077 7.091 -0.533>
   }
   Output Pose(Result=Component To Local(ComponentPose=@n159)) {layerGroup=None}
  ]}
 EventGraph {nodes [
   n1 = Event BlueprintUpdateAnimation()
   > n2 = Sequence!(execute) -> (then_0, then_1, then_2:Exec, then_3:Exec, then_4:Exec, then_5:Exec, then_6:Exec)
   > Set DeltaTimeLocal(DeltaTimeLocal:double=@n1.DeltaTimeX) -> (Output_Get:double)
   > n3 = Cloth Controller Function DEPRECATED!(
    execute
    CharacterPartAnimInstance:Custom Character Part Anim Instance Object Reference=Self-Reference()
    ClothingAssetName:String=M_MED_Solstice_Clothing_0
    Cloth Paramaters:Cloth Params Struct Structure=Get Cloth Paramaters_0() -> (Cloth Paramaters_0:Cloth Params Struct Structure)
    Cloth Core Settings:Cloth Core Setting Struct Structure=Get Cloth Core Settings_0() -> (Cloth Core Settings_0:Cloth Core Setting Struct Structure)
    ClothUpAxis:Vector="1,0,0"
    OnGround_Moving_SpeedLimit:double=2
    OnGround_Moving_AngularVel_Max:double=0.05
    OnGround_Moving_FictitiousVel_Max:double=0.6
    LinVelScale_FwdBwd_Clamp:double=0
    LinVelScale_UpDwn_Clamp:double=0.7
    DeltaTime:double=0
    GameTimeInSeconds:double=0
    WindGustsParams:Wind Gust Params Struct Structure
    WindGustLocalSettings:Wind Gust Local Setting Struct Structure
   ) -> (
    then
    CurrentStateOut:String
    Cloth Settings Out:Cloth Setting Struct Structure
    Cloth Core Settings Out:Cloth Core Setting Struct Structure
    WindGustSettings Out:Wind Gust Setting Struct Structure
    WindGustLocalSettings Out:Wind Gust Local Setting Struct Structure
    GustStrength - WindVelocityOut:Vector
    LinearVelVector-DELETE LATER:Vector
   )
   > Set Gust Local Settings(Gust Local Settings:Wind Gust Local Setting Struct Structure=@n3.WindGustLocalSettings Out) -> (Output_Get:Wind Gust Local Setting Struct Structure)
   n4 = Gravity Override State Function DEPRECATED(
    execute=@n2.then_1
    CharacterPartAnimInstance=Self-Reference()
    GameTimeInSeconds=GetGameTimeInSeconds()
    GravityOverrideParams=Get Gravity Override Params()
   )
   > Set GravityOverride(GravityOverride=@n4)
   n5 = Branch(execute=@n2.then_2, Condition=Get bIsCrouching())
   > Set Crouching_Alpha(Crouching_Alpha:double=1) -> (Output_Get:double)
   Set Crouching_Alpha(execute=@n5.else, Crouching_Alpha=0)
   n6 = Get Dynamic Material Instance(execute=@n2.then_5, SkeletalMeshComponent=GetPartSkeletalMeshComponent(), ElementIndex=2)
   > n7 = Branch(Condition=Get bClothEnabled())
   > SetScalarParameterValue(self=@n6, ParameterName=Hide Element 01, Value=0)
   > SetScalarParameterValue(self=@n6, ParameterName=Hide Element 02, Value=0)
   > n8 = Get Dynamic Material Instance(SkeletalMeshComponent=GetPartSkeletalMeshComponent(), ElementIndex=2)
   > n9 = Branch(Condition=Get is_wearing_westsausage() -> (is_wearing_westsausage:bool))
   > SetScalarParameterValue(self=@n8, ParameterName=Hide Element 02, Value=1)
   SetScalarParameterValue(execute=@n9.else, self=@n8, ParameterName=Hide Element 02, Value=0)
   SetScalarParameterValue(execute=@n7.else, self=@n6, ParameterName=Hide Element 01, Value=1)
   > SetScalarParameterValue(self=@n6, ParameterName=Hide Element 02, Value=1)
   n10 = Branch(
    execute=[@n16.then, @n2.then_6]
    Condition=Get bIsInFirstPerson!(self:Fort Player Anim Instance Object Reference=GetOwnerFortPlayerAnimInstance!(self:Custom Character Part Anim Instance Object Reference) -> (ReturnValue:Fort Player Anim Instance Object Reference)) -> (bIsInFirstPerson:bool)
   )
   n11 = GetPartSkeletalMeshComponent()
   > n12 = Get Dynamic Material Instance(SkeletalMeshComponent=@n11, ElementIndex=2)
   > SetScalarParameterValue(self=@n12, ParameterName=OpacityMaskMultiplier, Value=0)
   n13 = Get Dynamic Material Instance(execute=[@n17.then, @n10.else], SkeletalMeshComponent=@n11, ElementIndex=2)
   > SetScalarParameterValue(self=@n13, ParameterName=OpacityMaskMultiplier, Value=1)
   n14 = Cast To FortPlayerPawn(execute=@n2.then_4, Object=GetOwningActor())
   > Set As Fort Player Pawn(As Fort Player Pawn:Fort Player Pawn Object Reference=@n14) -> (Output_Get:Fort Player Pawn Object Reference)
   > n15 = Get!(execute:Exec) -> (then:Exec, else:Exec, As Fort Player Pawn:Fort Player Pawn Object Reference)
   > Set is_wearing_westsausage(is_wearing_westsausage:bool=HasMatchingGameplayTag!(self:Gameplay Tag Asset Interface Interface=@n15, TagToCheck:Gameplay Tag Structure=(TagName="Granted.Athena.HandsAreReplaced")) -> (ReturnValue:bool)) -> (Output_Get:bool)
   n16 = Event OnBeginTargeting() {is=title K2Node_Event}
   n17 = Event OnEndTargeting() {is=title K2Node_Event}
   TryGetPawnOwner()
   GetGameTimeInSeconds()
   Branch(Condition="Equal (Integer)"(A=Get CurrentLOD()))
   Get Wind Gust Params() -> (Wind Gust Params:Wind Gust Params Struct Structure)
   Get DeltaTimeLocal() -> (DeltaTimeLocal:double)
   Get Gust Local Settings() -> (Gust Local Settings:Wind Gust Local Setting Struct Structure)
  ]}
]
variables [
 As Fort Player Pawn {type=FortPlayerPawn}
 Cloth Core Settings_0 {
  type=ClothCoreSettingStruct
  default {
   collisionThickness=0
   dampingCoefficient=0.2
   frictionCoefficient=0.3
   localDampingCoefficient=0.1
   numberIterations=2
   numberSubsteps=1
   selfCollisionThickness=0
   animDriveDamping=<0 1>
   animDriveStiffness=<0 1>
   areaStiffness=<0.6 0.6>
   bendingStiffness=<1 1>
   edgeStiffness=<0.8 0.8>
   tetherStiffness=<0.8 0.8>
  }
 }
 Cloth Paramaters_0 {
  type=ClothParamsStruct
  default {
   dBNO {
    angularVelocityScale=0.6
    fictitiousAngularScale=0.1
    jointName=pelvis
    dragCoefficient=<0.2 1>
    gravityOverride=<-980 400 0>
    liftCoefficient=<0 0.5>
    linearVelocityScale=<0.3 0.3 0.1>
   }
   dBNO_Carried {
    angularVelocityScale=0.6
    fictitiousAngularScale=0.1
    jointName=pelvis
    dragCoefficient=<0.2 1>
    gravityOverride=<-980 -200 0>
    liftCoefficient=<0 0.5>
    linearVelocityScale=<0.3 0.3 0.1>
   }
   dBNO_Impulse {
    angularVelocityScale=0
    fictitiousAngularScale=0
    jointName=pelvis
    dragCoefficient=<0 0.2>
    gravityOverride=<-3000 -2000 0>
    liftCoefficient=<0 0.1>
    linearVelocityScale=<0.3 0.1 0.3>
   }
   dBNO_RollBwd {
    angularVelocityScale=0
    fictitiousAngularScale=0
    jointName=pelvis
    dragCoefficient=<0 0.2>
    gravityOverride=<-2000 0 0>
    liftCoefficient=<0 0.1>
    linearVelocityScale=<0.3 0.1 0.3>
   }
   dBNO_RollFwd {
    angularVelocityScale=0
    fictitiousAngularScale=0
    jointName=pelvis
    dragCoefficient=<0 0.2>
    gravityOverride=<-980 0 0>
    liftCoefficient=<0 0.1>
    linearVelocityScale=<0.3 0.1 0.3>
   }
   dBNO_Tumbling {
    angularVelocityScale=0
    fictitiousAngularScale=0
    jointName=pelvis
    dragCoefficient=<0 0.2>
    gravityOverride=<-2500 -300 0>
    liftCoefficient=<0 0.1>
    linearVelocityScale=<0.3 0.1 0.3>
   }
   emoteOrMelee {
    angularVelocityScale=0.2
    fictitiousAngularScale=0.1
    jointName=pelvis
    dragCoefficient=<0 0.3>
    gravityOverride=<-980 -500 0>
    liftCoefficient=<0 0.2>
    linearVelocityScale=<0.05 0.05 0.05>
   }
   front_End {
    angularVelocityScale=0.15
    fictitiousAngularScale=0.4
    jointName=root
    dragCoefficient=<0.2 0.2>
    gravityOverride=<0 -200 -980>
    liftCoefficient=<0.1 0.1>
    linearVelocityScale=<0.1 0.1 0.1>
   }
   onGround_Crouch_Moving {
    angularVelocityScale=0.3
    fictitiousAngularScale=0.1
    jointName=root
    dragCoefficient=<1 1>
    gravityOverride=<0 0 -980>
    liftCoefficient=<0.2 0.2>
    linearVelocityScale=<0.2 0.1 0.1>
   }
   onGround_Falling {
    angularVelocityScale=0.6
    fictitiousAngularScale=0
    jointName=root
    dragCoefficient=<0 0.2>
    gravityOverride=<0 -500 -980>
    liftCoefficient=<0 0.1>
    linearVelocityScale=<2 1 0.3>
   }
   onGround_Moving {
    angularVelocityScale=0.3
    fictitiousAngularScale=0.1
    jointName=root
    dragCoefficient=<0.5 0.3>
    gravityOverride=<0 -100 -2000>
    liftCoefficient=<0.05 0.2>
    linearVelocityScale=<0.2 0.1 1>
   }
   onGround_Standing {
    angularVelocityScale=0.3
    fictitiousAngularScale=0.1
    jointName=root
    dragCoefficient=<2 2>
    gravityOverride=<0 -100 -2000>
    liftCoefficient=<0.1 0.5>
    linearVelocityScale=<0.2 0.1 0.1>
   }
   rail_Grinding {
    angularVelocityScale=0.2
    fictitiousAngularScale=0
    jointName=root
    dragCoefficient=<0.1 0.2>
    gravityOverride=<50 -500 -400>
    liftCoefficient=<0 0.5>
    linearVelocityScale=<0.1 0.1 0.1>
   }
   riding_Moving {
    angularVelocityScale=0.05
    fictitiousAngularScale=1
    jointName=root
    dragCoefficient=<0.5 0.5>
    gravityOverride=<0 0 -980>
    liftCoefficient=<0.5 0.5>
    linearVelocityScale=<0.4 0.6 0.3>
   }
   riding_Moving_Targeting {
    angularVelocityScale=0.05
    fictitiousAngularScale=0.7
    jointName=root
    dragCoefficient=<0.5 0.5>
    gravityOverride=<0 0 -980>
    liftCoefficient=<0.5 0.5>
    linearVelocityScale=<0.2 0.6 0.2>
   }
   riding_Standing {
    angularVelocityScale=0.05
    fictitiousAngularScale=1
    jointName=root
    dragCoefficient=<0.5 0.5>
    gravityOverride=<0 -450 -980>
    liftCoefficient=<0.5 0.5>
    linearVelocityScale=<0.6 0.6 0.3>
   }
   skydive_DivingDown {
    angularVelocityScale=0.1
    fictitiousAngularScale=0
    jointName=pelvis
    dragCoefficient=<0 1.5>
    gravityOverride=<-450 -100 0>
    liftCoefficient=<0 0.3>
    linearVelocityScale=<0.1 0.1 0.1>
   }
   skydive_FloatingDown {
    angularVelocityScale=0.1
    fictitiousAngularScale=0
    jointName=root
    dragCoefficient=<0 2.5>
    gravityOverride=<0 0 980>
    liftCoefficient=<0 0.4>
    linearVelocityScale=<0.2 0.2 0.1>
   }
   skydive_GoingUp {
    angularVelocityScale=0.1
    fictitiousAngularScale=0
    jointName=root
    dragCoefficient=<0 0>
    gravityOverride=<0 -3000 3000>
    liftCoefficient=<0 0>
    linearVelocityScale=<0.1 0.1 0.05>
   }
   skydive_Parachuting {
    angularVelocityScale=0.1
    fictitiousAngularScale=0
    jointName=root
    dragCoefficient=<0 2.5>
    gravityOverride=<0 0 -980>
    liftCoefficient=<0 0.5>
    linearVelocityScale=<0.2 0.2 0.2>
   }
   skydive_Surfing {
    angularVelocityScale=0.6
    fictitiousAngularScale=0.1
    jointName=root
    dragCoefficient=<1.5 1.5>
    gravityOverride=<0 -600 -500>
    liftCoefficient=<1 1>
    linearVelocityScale=<0.1 0.1 0.3>
   }
   swimming {
    angularVelocityScale=0.6
    fictitiousAngularScale=0
    jointName=root
    dragCoefficient=<0 2>
    gravityOverride=<0 -600 400>
    liftCoefficient=<0 0.4>
    linearVelocityScale=<0.6 0.2 0.2>
   }
   tactical_Sprint {
    angularVelocityScale=0.6
    fictitiousAngularScale=0.1
    jointName=root
    dragCoefficient=<2 2>
    gravityOverride=<0 0 -980>
    liftCoefficient=<0.5 0.5>
    linearVelocityScale=<0.6 0.6 0.5>
   }
   ziplining {
    angularVelocityScale=0.6
    fictitiousAngularScale=0
    jointName=root
    dragCoefficient=<0 1.5>
    gravityOverride=<0 0 -980>
    liftCoefficient=<0 0.3>
    linearVelocityScale=<0.4 0.4 0.5>
   }
  }
 }
 Crouching_Alpha {type=number}
 DeltaTimeLocal {type=number}
 Gravity Override Params {type=GravityOverrideParamsStruct}
 GravityOverride {type=Vector}
 Gust Local Settings {type=WindGustLocalSettingStruct, default {elapsedTime=0, finalGustStrengthLocal=0, gustDuration=0, gustStrength=0, timeBetweenGusts=0}}
 HeadMesh {type=SkeletalMeshComponent}
 is_wearing_westsausage {type=boolean}
 upperarm_rotZ_l {type=number}
 upperarm_rotZ_r {type=number}
 Wind Gust Params {
  type=WindGustParamsStruct
  default {
   dBNO {
    maxDownInterpSpeed=5
    maxGustDuration=2
    maxGustStrength=7
    maxTimeBetweenGusts=7
    maxUpInterpSpeed=5
    minDownInterpSpeed=2
    minGustDuration=0.5
    minGustStrength=3
    minTimeBetweenGusts=0
    minUpInterpSpeed=2
    windTurbulenceScalarA=0.1
    windTurbulenceScalarB=1
    windAmplitude=<0 0 0>
    windFrequency=<0 0 0>
    windOffset=<0 0 0>
   }
   dBNO_Carried {
    maxDownInterpSpeed=5
    maxGustDuration=2
    maxGustStrength=7
    maxTimeBetweenGusts=7
    maxUpInterpSpeed=5
    minDownInterpSpeed=2
    minGustDuration=0.5
    minGustStrength=3
    minTimeBetweenGusts=0
    minUpInterpSpeed=2
    windTurbulenceScalarA=0.1
    windTurbulenceScalarB=1
    windAmplitude=<0 0 0>
    windFrequency=<0 0 0>
    windOffset=<0 0 0>
   }
   dBNO_Impulse {
    maxDownInterpSpeed=5
    maxGustDuration=2
    maxGustStrength=7
    maxTimeBetweenGusts=7
    maxUpInterpSpeed=5
    minDownInterpSpeed=2
    minGustDuration=0.5
    minGustStrength=3
    minTimeBetweenGusts=0
    minUpInterpSpeed=2
    windTurbulenceScalarA=0.1
    windTurbulenceScalarB=1
    windAmplitude=<0 0 0>
    windFrequency=<0 0 0>
    windOffset=<0 0 0>
   }
   dBNO_RollBwd {
    maxDownInterpSpeed=5
    maxGustDuration=2
    maxGustStrength=7
    maxTimeBetweenGusts=7
    maxUpInterpSpeed=5
    minDownInterpSpeed=2
    minGustDuration=0.5
    minGustStrength=3
    minTimeBetweenGusts=0
    minUpInterpSpeed=2
    windTurbulenceScalarA=0.1
    windTurbulenceScalarB=1
    windAmplitude=<0 0 0>
    windFrequency=<0 0 0>
    windOffset=<0 0 0>
   }
   dBNO_RollFwd {
    maxDownInterpSpeed=5
    maxGustDuration=2
    maxGustStrength=7
    maxTimeBetweenGusts=7
    maxUpInterpSpeed=5
    minDownInterpSpeed=2
    minGustDuration=0.5
    minGustStrength=3
    minTimeBetweenGusts=0
    minUpInterpSpeed=2
    windTurbulenceScalarA=0.1
    windTurbulenceScalarB=1
    windAmplitude=<0 0 0>
    windFrequency=<0 0 0>
    windOffset=<0 0 0>
   }
   dBNO_Tumbling {
    maxDownInterpSpeed=5
    maxGustDuration=2
    maxGustStrength=7
    maxTimeBetweenGusts=7
    maxUpInterpSpeed=5
    minDownInterpSpeed=2
    minGustDuration=0.5
    minGustStrength=3
    minTimeBetweenGusts=0
    minUpInterpSpeed=2
    windTurbulenceScalarA=0.1
    windTurbulenceScalarB=1
    windAmplitude=<0 0 0>
    windFrequency=<0 0 0>
    windOffset=<0 0 0>
   }
   emoteOrMelee {
    maxDownInterpSpeed=5
    maxGustDuration=2
    maxGustStrength=7
    maxTimeBetweenGusts=7
    maxUpInterpSpeed=5
    minDownInterpSpeed=2
    minGustDuration=0.5
    minGustStrength=3
    minTimeBetweenGusts=0
    minUpInterpSpeed=2
    windTurbulenceScalarA=0.1
    windTurbulenceScalarB=1
    windAmplitude=<30 25 50>
    windFrequency=<2 3 5>
    windOffset=<-300 -150 0>
   }
   front_End {
    maxDownInterpSpeed=5
    maxGustDuration=2
    maxGustStrength=10
    maxTimeBetweenGusts=7
    maxUpInterpSpeed=5
    minDownInterpSpeed=2
    minGustDuration=0.5
    minGustStrength=5
    minTimeBetweenGusts=0
    minUpInterpSpeed=2
    windTurbulenceScalarA=0.1
    windTurbulenceScalarB=1
    windAmplitude=<10 30 50>
    windFrequency=<15 6 7>
    windOffset=<0 -200 -250>
   }
   onGround_Crouch_Moving {
    maxDownInterpSpeed=5
    maxGustDuration=2
    maxGustStrength=7
    maxTimeBetweenGusts=7
    maxUpInterpSpeed=5
    minDownInterpSpeed=2
    minGustDuration=0.5
    minGustStrength=3
    minTimeBetweenGusts=0
    minUpInterpSpeed=2
    windTurbulenceScalarA=0.1
    windTurbulenceScalarB=1
    windAmplitude=<100 25 75>
    windFrequency=<1.5 0.5 7.5>
    windOffset=<0 -100 -175>
   }
   onGround_Falling {
    maxDownInterpSpeed=5
    maxGustDuration=2
    maxGustStrength=7
    maxTimeBetweenGusts=7
    maxUpInterpSpeed=5
    minDownInterpSpeed=2
    minGustDuration=0.5
    minGustStrength=3
    minTimeBetweenGusts=0
    minUpInterpSpeed=2
    windTurbulenceScalarA=0.1
    windTurbulenceScalarB=1
    windAmplitude=<0 0 0>
    windFrequency=<0 0 0>
    windOffset=<0 0 0>
   }
   onGround_Moving {
    maxDownInterpSpeed=5
    maxGustDuration=8
    maxGustStrength=7
    maxTimeBetweenGusts=8
    maxUpInterpSpeed=5
    minDownInterpSpeed=2
    minGustDuration=0.2
    minGustStrength=3
    minTimeBetweenGusts=1
    minUpInterpSpeed=2
    windTurbulenceScalarA=0.1
    windTurbulenceScalarB=0.8
    windAmplitude=<30 65 42>
    windFrequency=<5 6 8.6>
    windOffset=<30 -250 -300>
   }
   onGround_Standing {
    maxDownInterpSpeed=5
    maxGustDuration=8
    maxGustStrength=7
    maxTimeBetweenGusts=8
    maxUpInterpSpeed=5
    minDownInterpSpeed=2
    minGustDuration=0.2
    minGustStrength=2
    minTimeBetweenGusts=1
    minUpInterpSpeed=2
    windTurbulenceScalarA=0.1
    windTurbulenceScalarB=0.8
    windAmplitude=<50 48 56>
    windFrequency=<1.5 2.3 3.7>
    windOffset=<150 -200 -300>
   }
   rail_Grinding {
    maxDownInterpSpeed=5
    maxGustDuration=8
    maxGustStrength=7
    maxTimeBetweenGusts=8
    maxUpInterpSpeed=5
    minDownInterpSpeed=2
    minGustDuration=0.2
    minGustStrength=3
    minTimeBetweenGusts=1
    minUpInterpSpeed=2
    windTurbulenceScalarA=0.1
    windTurbulenceScalarB=0.8
    windAmplitude=<20 65 42>
    windFrequency=<5 6 8.6>
    windOffset=<50 -500 -400>
   }
   riding_Moving {
    maxDownInterpSpeed=5
    maxGustDuration=2
    maxGustStrength=7
    maxTimeBetweenGusts=7
    maxUpInterpSpeed=5
    minDownInterpSpeed=2
    minGustDuration=0.5
    minGustStrength=3
    minTimeBetweenGusts=0
    minUpInterpSpeed=2
    windTurbulenceScalarA=0.1
    windTurbulenceScalarB=1
    windAmplitude=<0 0 0>
    windFrequency=<0 0 0>
    windOffset=<0 0 0>
   }
   riding_Moving_Targeting {
    maxDownInterpSpeed=5
    maxGustDuration=2
    maxGustStrength=7
    maxTimeBetweenGusts=7
    maxUpInterpSpeed=5
    minDownInterpSpeed=2
    minGustDuration=0.5
    minGustStrength=3
    minTimeBetweenGusts=0
    minUpInterpSpeed=2
    windTurbulenceScalarA=0.1
    windTurbulenceScalarB=1
    windAmplitude=<0 0 0>
    windFrequency=<0 0 0>
    windOffset=<0 0 0>
   }
   riding_Standing {
    maxDownInterpSpeed=5
    maxGustDuration=2
    maxGustStrength=7
    maxTimeBetweenGusts=7
    maxUpInterpSpeed=5
    minDownInterpSpeed=2
    minGustDuration=0.5
    minGustStrength=3
    minTimeBetweenGusts=0
    minUpInterpSpeed=2
    windTurbulenceScalarA=0.1
    windTurbulenceScalarB=1
    windAmplitude=<0 0 0>
    windFrequency=<0 0 0>
    windOffset=<0 0 0>
   }
   skydive_DivingDown {
    maxDownInterpSpeed=5
    maxGustDuration=8
    maxGustStrength=7
    maxTimeBetweenGusts=8
    maxUpInterpSpeed=5
    minDownInterpSpeed=2
    minGustDuration=0.2
    minGustStrength=3
    minTimeBetweenGusts=1
    minUpInterpSpeed=2
    windTurbulenceScalarA=0.1
    windTurbulenceScalarB=4
    windAmplitude=<0 25 35>
    windFrequency=<0 4 10>
    windOffset=<-450 -100 100>
   }
   skydive_FloatingDown {
    maxDownInterpSpeed=5
    maxGustDuration=8
    maxGustStrength=1
    maxTimeBetweenGusts=8
    maxUpInterpSpeed=5
    minDownInterpSpeed=2
    minGustDuration=0.2
    minGustStrength=0
    minTimeBetweenGusts=1
    minUpInterpSpeed=2
    windTurbulenceScalarA=0.5
    windTurbulenceScalarB=1
    windAmplitude=<50 20 35>
    windFrequency=<8 2 3>
    windOffset=<0 -300 200>
   }
   skydive_GoingUp {
    maxDownInterpSpeed=5
    maxGustDuration=8
    maxGustStrength=7
    maxTimeBetweenGusts=8
    maxUpInterpSpeed=5
    minDownInterpSpeed=2
    minGustDuration=0.2
    minGustStrength=3
    minTimeBetweenGusts=1
    minUpInterpSpeed=2
    windTurbulenceScalarA=0.1
    windTurbulenceScalarB=0.8
    windAmplitude=<0 0 0>
    windFrequency=<0 0 0>
    windOffset=<0 -500 300>
   }
   skydive_Parachuting {
    maxDownInterpSpeed=5
    maxGustDuration=8
    maxGustStrength=1
    maxTimeBetweenGusts=8
    maxUpInterpSpeed=5
    minDownInterpSpeed=2
    minGustDuration=5
    minGustStrength=0
    minTimeBetweenGusts=5
    minUpInterpSpeed=2
    windTurbulenceScalarA=0.2
    windTurbulenceScalarB=1
    windAmplitude=<25 50 75>
    windFrequency=<8 6 2>
    windOffset=<-150 -400 -150>
   }
   skydive_Surfing {
    maxDownInterpSpeed=5
    maxGustDuration=2
    maxGustStrength=7
    maxTimeBetweenGusts=7
    maxUpInterpSpeed=5
    minDownInterpSpeed=2
    minGustDuration=0.5
    minGustStrength=3
    minTimeBetweenGusts=0
    minUpInterpSpeed=2
    windTurbulenceScalarA=0.1
    windTurbulenceScalarB=1
    windAmplitude=<0 0 0>
    windFrequency=<0 0 0>
    windOffset=<0 0 0>
   }
   swimming {
    maxDownInterpSpeed=5
    maxGustDuration=2
    maxGustStrength=7
    maxTimeBetweenGusts=7
    maxUpInterpSpeed=5
    minDownInterpSpeed=2
    minGustDuration=0.5
    minGustStrength=3
    minTimeBetweenGusts=0
    minUpInterpSpeed=2
    windTurbulenceScalarA=0.1
    windTurbulenceScalarB=1
    windAmplitude=<10 25 20>
    windFrequency=<1.5 0.5 7.5>
    windOffset=<0 -150 50>
   }
   tactical_Sprint {
    maxDownInterpSpeed=5
    maxGustDuration=2
    maxGustStrength=7
    maxTimeBetweenGusts=7
    maxUpInterpSpeed=5
    minDownInterpSpeed=2
    minGustDuration=0.5
    minGustStrength=3
    minTimeBetweenGusts=0
    minUpInterpSpeed=2
    windTurbulenceScalarA=0.1
    windTurbulenceScalarB=1
    windAmplitude=<0 0 0>
    windFrequency=<0 0 0>
    windOffset=<0 0 0>
   }
   ziplining {
    maxDownInterpSpeed=5
    maxGustDuration=2
    maxGustStrength=7
    maxTimeBetweenGusts=7
    maxUpInterpSpeed=5
    minDownInterpSpeed=2
    minGustDuration=0.5
    minGustStrength=3
    minTimeBetweenGusts=0
    minUpInterpSpeed=2
    windTurbulenceScalarA=0.1
    windTurbulenceScalarB=1
    windAmplitude=<0 0 0>
    windFrequency=<0 0 0>
    windOffset=<0 0 0>
   }
  }
 }
]
