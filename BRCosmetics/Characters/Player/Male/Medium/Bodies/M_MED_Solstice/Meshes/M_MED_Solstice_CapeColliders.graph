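# Physics asset of collision volumes for the cape on the M_MED_Solstice male player body, as its name indicates.
# Capsules cover the pelvis, spine, clavicles, neck, back-right pouch and four right-side dynamic scalp bones, with a sphere on the head.
# Arm, leg and foot bodies have no shapes.
# There are no constraints, so the asset supplies collider shapes rather than a jointed ragdoll.
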PhysicsAsset /BRCosmetics/Characters/Player/Male/Medium/Bodies/M_MED_Solstice/Meshes/M_MED_Solstice_CapeColliders
bodies {
 calf_l []
 calf_r []
 clavicle_l [{shapeType=Capsule, length=11.2587, radius=11.7237, center=<9.5596 -0.8065 -1.4909>, rotation=<rot 75.9351 -0.0001 175.5622>}]
 clavicle_r [{shapeType=Capsule, length=11.2587, radius=11.7237, center=<-9.5596 0.8065 1.4909>, rotation=<rot 75.9351 -0.0001 -4.4378>}]
 dyn_pouch_bk_r [{
   shapeType=Capsule
   length=5.8429
   radius=8.9341
   shapeName=dyn_pouch_bk_r_capsule
   center=<3.6761 0 4.4333>
   rotation=<rot 85.527 179.9988 179.999>
  }]
 dyn_scalp_bk_r_04 [{shapeType=Capsule, length=8.6913, radius=3.8177, shapeName=dyn_scalp_fr_r_04_capsule, center=<5 0 0>, rotation=<rot 90 0 0>}]
 dyn_scalp_bk_r_05 [{shapeType=Capsule, length=8.6913, radius=3.8177, shapeName=dyn_scalp_fr_r_04_capsule, center=<5 0 0>, rotation=<rot 90 0 0>}]
 dyn_scalp_fr_r_04 [{shapeType=Capsule, length=8.6913, radius=3.8177, shapeName=dyn_scalp_fr_r_04_capsule, center=<5 0 0>, rotation=<rot 90 0 0>}]
 dyn_scalp_fr_r_05 [{shapeType=Capsule, length=8.6913, radius=3.8177, shapeName=dyn_scalp_fr_r_04_capsule, center=<5 0 0>, rotation=<rot 90 0 0>}]
 foot_l []
 foot_r []
 head [{shapeType=Sphere, radius=17.3125, center=<11.028 -0.174 0>}]
 lowerarm_l []
 lowerarm_r []
 neck_02 [{shapeType=Capsule, length=13.4375, radius=7.3175, center=<-0.2642 -0.3408 0>, rotation=<rot 90 -171.1193 -175.3978>}]
 pelvis [{shapeType=Capsule, length=13.3417, radius=14.923, center=<-2.6358 -1.7685 0>}]
 spine_01 [{shapeType=Capsule, length=13.7204, radius=14.291, center=<-2.6215 -2.1725 -0.0002>}]
 spine_02 [
  {shapeType=Capsule, length=11.188, radius=14.8295, center=<-2.6215 -1.483 -0.0002>}
  {shapeType=Capsule, length=28.5347, radius=12.4416, center=<2.5024 -1.6928 3.9823>, rotation=<rot 90 -0.0396 -0.0396>}
  {shapeType=Capsule, length=28.5347, radius=12.4416, center=<2.4704 -1.6929 -3.9827>, rotation=<rot 90 0.0099 0.0099>}
 ]
 spine_03 [{shapeType=Capsule, length=12.4685, radius=16.6193, center=<-1.953 -2.0732 0.0004>}]
 spine_04 [{shapeType=Capsule, length=16.0386, radius=18.3808, center=<-0.8298 -1.2748 0.0002>}]
 spine_05 [
  {shapeType=Capsule, length=17.8437, radius=17.195, center=<-1.7514 -0.8401 -0.0003>}
  {shapeType=Capsule, length=12.7812, radius=17.0625, center=<0.7003 -0.1688 0>}
 ]
 thigh_l []
 thigh_r []
 upperarm_l []
 upperarm_r []
}
constraints []
