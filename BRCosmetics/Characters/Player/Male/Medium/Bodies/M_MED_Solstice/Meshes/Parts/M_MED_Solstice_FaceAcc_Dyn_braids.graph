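# Physics asset for the dynamic braids on the face accessory of the male medium Solstice body.
# Six braids are capsule chains of four or five segments, each rooted in a tiny MSTR body linked to neck_02 with free angular motion.
# Segment joints have limited swing and mostly tight twist, with root limits varying per braid.
# Capsules on spine, neck, clavicles and upper arms form the surrounding torso volume, joined by angularly free constraints.
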PhysicsAsset /BRCosmetics/Characters/Player/Male/Medium/Bodies/M_MED_Solstice/Meshes/Parts/M_MED_Solstice_FaceAcc_Dyn_braids
bodies {
 clavicle_l [
  {
   shapeType=Capsule
   length=11.1643
   radius=12.4085
   shapeName=dyn_braid_fr_l_MSTR_capsule
   center=<11.1322 -1.2474 -3.6751>
   rotation=<rot 73.9441 2.9633 -179.9999>
  }
  {
   shapeType=Capsule
   length=22.2761
   radius=22.1386
   shapeName=clavicle_r_capsule
   center=<36.1749 -0.1407 12.9165>
   rotation=<rot 14.3031 2.2209 -164.007>
  }
 ]
 clavicle_r [
  {
   shapeType=Capsule
   length=11.1643
   radius=12.4085
   shapeName=dyn_braid_fr_l_MSTR_capsule
   center=<-11.1322 1.2474 3.6751>
   rotation=<rot 73.9442 2.9633 -179.9999>
  }
  {
   shapeType=Capsule
   length=22.2761
   radius=22.1386
   shapeName=clavicle_r_capsule
   center=<-35.6139 0.1691 -12.8788>
   rotation=<rot 14.3031 2.2209 -164.0068>
  }
 ]
 dyn_braid_bk_in_l_01 [{shapeType=Capsule, length=7.3241, radius=1.9831, shapeName=dyn_braid_fr_l_01_capsule, center=<13.1575 0.0002 0.0001>, rotation=<rot 90 0 0>}]
 dyn_braid_bk_in_l_02 [{shapeType=Capsule, length=7.4593, radius=1.9831, shapeName=dyn_braid_fr_l_01_capsule, center=<3.8736 0 0>, rotation=<rot 90 0 0>}]
 dyn_braid_bk_in_l_03 [{shapeType=Capsule, length=9.5335, radius=1.9831, shapeName=dyn_braid_fr_l_01_capsule, center=<4.5047 0 0>, rotation=<rot 90 0 0>}]
 dyn_braid_bk_in_l_04 [{shapeType=Capsule, length=7.3241, radius=1.9831, shapeName=dyn_braid_fr_l_01_capsule, center=<3.4397 0 0>, rotation=<rot 90 0 0>}]
 dyn_braid_bk_in_l_MSTR [{shapeType=Capsule, length=0.505, radius=0.505, shapeName=dyn_braid_fr_l_MSTR_capsule}]
 dyn_braid_bk_l_01 [{shapeType=Capsule, length=11.0813, radius=1.9831, shapeName=dyn_braid_fr_l_01_capsule, center=<11.564 0 0>, rotation=<rot 90 0 0>}]
 dyn_braid_bk_l_02 [{shapeType=Capsule, length=11.8479, radius=1.9831, shapeName=dyn_braid_fr_l_01_capsule, center=<6.0568 -0.0001 0>, rotation=<rot 90 0 0>}]
 dyn_braid_bk_l_03 [{shapeType=Capsule, length=10.8317, radius=1.9831, shapeName=dyn_braid_fr_l_01_capsule, center=<5.4655 0 0>, rotation=<rot 90 0 0>}]
 dyn_braid_bk_l_04 [{shapeType=Capsule, length=10.594, radius=1.9831, shapeName=dyn_braid_fr_l_01_capsule, center=<6.0147 -0.0002 0>, rotation=<rot 90 0 0>}]
 dyn_braid_bk_l_05 [{shapeType=Capsule, length=7.3241, radius=1.9831, shapeName=dyn_braid_fr_l_01_capsule, center=<3.4397 0 0>, rotation=<rot 90 0 0>}]
 dyn_braid_bk_l_MSTR [{shapeType=Capsule, length=0.505, radius=0.505, shapeName=dyn_braid_fr_l_MSTR_capsule}]
 dyn_braid_bk_r_01 [{shapeType=Capsule, length=13.3302, radius=1.9831, shapeName=dyn_braid_fr_l_01_capsule, center=<14.765 0.0005 0.0002>, rotation=<rot 90 0 0>}]
 dyn_braid_bk_r_02 [{shapeType=Capsule, length=8.7055, radius=1.9831, shapeName=dyn_braid_fr_l_01_capsule, center=<6.326 0.0001 0>, rotation=<rot 90 0 0>}]
 dyn_braid_bk_r_03 [{shapeType=Capsule, length=10.2546, radius=1.9831, shapeName=dyn_braid_fr_l_01_capsule, center=<6.6856 -0.0001 0>, rotation=<rot 90 0 0>}]
 dyn_braid_bk_r_04 [{shapeType=Capsule, length=7.3241, radius=1.9831, shapeName=dyn_braid_fr_l_01_capsule, center=<4.9389 -0.0001 0>, rotation=<rot 90 0 0>}]
 dyn_braid_bk_r_MSTR [{shapeType=Capsule, length=0.505, radius=0.505, shapeName=dyn_braid_fr_l_MSTR_capsule}]
 dyn_braid_fr_in_r_01 [{shapeType=Capsule, length=6.8902, radius=2, shapeName=dyn_braid_fr_l_01_capsule, center=<4.8011 0.0001 0.0001>, rotation=<rot 90 0 0>}]
 dyn_braid_fr_in_r_02 [{shapeType=Capsule, length=6.757, radius=2, shapeName=dyn_braid_fr_l_01_capsule, center=<5.8852 0.0001 0>, rotation=<rot 90 0 0>}]
 dyn_braid_fr_in_r_03 [{shapeType=Capsule, length=7.9803, radius=2, shapeName=dyn_braid_fr_l_01_capsule, center=<4.3697 0 0>, rotation=<rot 90 0 0>}]
 dyn_braid_fr_in_r_04 [{shapeType=Capsule, length=7.8805, radius=2, shapeName=dyn_braid_fr_l_01_capsule, center=<3.4397 0 0>, rotation=<rot 90 0 0>}]
 dyn_braid_fr_in_r_MSTR [{shapeType=Capsule, length=0.505, radius=0.505, shapeName=dyn_braid_fr_l_MSTR_capsule}]
 dyn_braid_fr_l_01 [{shapeType=Capsule, length=7.3241, radius=1.9831, shapeName=dyn_braid_fr_l_01_capsule, center=<7.9731 -0.0001 0>, rotation=<rot 90 0 0>}]
 dyn_braid_fr_l_02 [{shapeType=Capsule, length=12.4095, radius=1.9831, shapeName=dyn_braid_fr_l_01_capsule, center=<6.5528 -0.0002 0>, rotation=<rot 90 0 0>}]
 dyn_braid_fr_l_03 [{shapeType=Capsule, length=9.0535, radius=1.9831, shapeName=dyn_braid_fr_l_01_capsule, center=<4.8444 0 0>, rotation=<rot 90 0 0>}]
 dyn_braid_fr_l_04 [{shapeType=Capsule, length=7.3241, radius=1.9831, shapeName=dyn_braid_fr_l_01_capsule, center=<3.4397 0 0>, rotation=<rot 90 0 0>}]
 dyn_braid_fr_l_MSTR [{shapeType=Capsule, length=0.505, radius=0.505, shapeName=dyn_braid_fr_l_MSTR_capsule}]
 dyn_braid_fr_r_01 [{shapeType=Capsule, length=12.7372, radius=1.9831, shapeName=dyn_braid_fr_l_01_capsule, center=<9.4662 0.0003 -0.0001>, rotation=<rot 90 0 0>}]
 dyn_braid_fr_r_02 [{shapeType=Capsule, length=8.512, radius=1.9831, shapeName=dyn_braid_fr_l_01_capsule, center=<6.0802 -0.0001 0>, rotation=<rot 90 0 0>}]
 dyn_braid_fr_r_03 [{shapeType=Capsule, length=8.8763, radius=1.9831, shapeName=dyn_braid_fr_l_01_capsule, center=<6.1595 0 0>, rotation=<rot 90 0 0>}]
 dyn_braid_fr_r_04 [{shapeType=Capsule, length=7.3241, radius=1.9831, shapeName=dyn_braid_fr_l_01_capsule, center=<4.9476 0.0001 0>, rotation=<rot 90 0 0>}]
 dyn_braid_fr_r_MSTR [{shapeType=Capsule, length=0.505, radius=0.505, shapeName=dyn_braid_fr_l_MSTR_capsule}]
 neck_02 [{shapeType=Capsule, length=13.8646, radius=8.7995, shapeName=dyn_braid_fr_l_MSTR_capsule, center=<4.7833 0.0001 0>, rotation=<rot 90 0 0>}]
 spine_03 [{shapeType=Capsule, length=11.1442, radius=18.6358, shapeName=dyn_braid_fr_l_MSTR_capsule, center=<0.6975 -0.1486 0.3111>}]
 spine_04 [{shapeType=Capsule, length=10.4355, radius=19.3444, shapeName=dyn_braid_fr_l_MSTR_capsule, center=<0.6514 0.0892 0>}]
 spine_05 [
  {shapeType=Capsule, length=20.8873, radius=14.0341, shapeName=dyn_braid_fr_l_MSTR_capsule, center=<-0.5712 -0.4774 0>}
  {shapeType=Capsule, length=23.5963, radius=11.3252, shapeName=spine_05_capsule, center=<-5.8629 6.0913 0>}
 ]
 upperarm_l [{shapeType=Capsule, length=17.1612, radius=7.8174, shapeName=dyn_braid_fr_l_MSTR_capsule, center=<13.4017 0.0001 0>, rotation=<rot 90 0 0>}]
 upperarm_r [{
   shapeType=Capsule
   length=36.5826
   radius=7.8174
   shapeName=dyn_braid_fr_l_MSTR_capsule
   center=<-22.6959 -0.0002 0>
   rotation=<rot 90 0 180>
  }]
}
constraints [
 {bone1Name=dyn_braid_fr_in_r_01, bone2Name=dyn_braid_fr_in_r_MSTR, swing1LimitDegrees=60, swing2LimitDegrees=60, twistLimitDegrees=25}
 {bone1Name=dyn_braid_fr_in_r_02, bone2Name=dyn_braid_fr_in_r_01, swing1LimitDegrees=20, swing2LimitDegrees=45, twistLimitDegrees=5}
 {bone1Name=dyn_braid_fr_in_r_03, bone2Name=dyn_braid_fr_in_r_02, swing1LimitDegrees=45, swing2LimitDegrees=45, twistLimitDegrees=5}
 {bone1Name=dyn_braid_fr_in_r_04, bone2Name=dyn_braid_fr_in_r_03, swing1LimitDegrees=45, swing2LimitDegrees=45, twistLimitDegrees=5}
 {bone1Name=dyn_braid_fr_r_01, bone2Name=dyn_braid_fr_r_MSTR, swing1LimitDegrees=60, swing2LimitDegrees=60, twistLimitDegrees=25}
 {bone1Name=dyn_braid_fr_r_02, bone2Name=dyn_braid_fr_r_01, swing1LimitDegrees=20, swing2LimitDegrees=45, twistLimitDegrees=5}
 {bone1Name=dyn_braid_fr_r_03, bone2Name=dyn_braid_fr_r_02, swing1LimitDegrees=45, swing2LimitDegrees=45, twistLimitDegrees=5}
 {bone1Name=dyn_braid_fr_r_04, bone2Name=dyn_braid_fr_r_03, swing1LimitDegrees=45, swing2LimitDegrees=45, twistLimitDegrees=5}
 {bone1Name=dyn_braid_bk_l_01, bone2Name=dyn_braid_bk_l_MSTR, swing1LimitDegrees=10, swing2LimitDegrees=20, twistLimitDegrees=5}
 {bone1Name=dyn_braid_bk_l_02, bone2Name=dyn_braid_bk_l_01, swing1LimitDegrees=20, swing2LimitDegrees=45, twistLimitDegrees=5}
 {bone1Name=dyn_braid_bk_l_03, bone2Name=dyn_braid_bk_l_02, swing1LimitDegrees=45, swing2LimitDegrees=45, twistLimitDegrees=5}
 {bone1Name=dyn_braid_bk_l_04, bone2Name=dyn_braid_bk_l_03, swing1LimitDegrees=45, swing2LimitDegrees=45, twistLimitDegrees=5}
 {bone1Name=dyn_braid_bk_in_l_01, bone2Name=dyn_braid_bk_in_l_MSTR, swing1LimitDegrees=10, swing2LimitDegrees=20, twistLimitDegrees=5}
 {bone1Name=dyn_braid_bk_in_l_02, bone2Name=dyn_braid_bk_in_l_01, swing1LimitDegrees=20, swing2LimitDegrees=45, twistLimitDegrees=5}
 {bone1Name=dyn_braid_bk_in_l_03, bone2Name=dyn_braid_bk_in_l_02, swing1LimitDegrees=45, swing2LimitDegrees=45, twistLimitDegrees=5}
 {bone1Name=dyn_braid_bk_in_l_04, bone2Name=dyn_braid_bk_in_l_03, swing1LimitDegrees=45, swing2LimitDegrees=45, twistLimitDegrees=5}
 {bone1Name=dyn_braid_bk_r_01, bone2Name=dyn_braid_bk_r_MSTR, swing1LimitDegrees=3, swing2LimitDegrees=20, twistLimitDegrees=5}
 {bone1Name=dyn_braid_bk_r_02, bone2Name=dyn_braid_bk_r_01, swing1LimitDegrees=20, swing2LimitDegrees=45, twistLimitDegrees=5}
 {bone1Name=dyn_braid_bk_r_03, bone2Name=dyn_braid_bk_r_02, swing1LimitDegrees=45, swing2LimitDegrees=45, twistLimitDegrees=5}
 {bone1Name=dyn_braid_bk_r_04, bone2Name=dyn_braid_bk_r_03, swing1LimitDegrees=45, swing2LimitDegrees=45, twistLimitDegrees=5}
 {
  bone1Name=dyn_braid_fr_in_r_MSTR
  bone2Name=neck_02
  swing1LimitDegrees=45
  swing1Motion=Free
  swing2LimitDegrees=45
  swing2Motion=Free
  twistLimitDegrees=45
  twistMotion=Free
 }
 {
  bone1Name=dyn_braid_fr_r_MSTR
  bone2Name=neck_02
  swing1LimitDegrees=45
  swing1Motion=Free
  swing2LimitDegrees=45
  swing2Motion=Free
  twistLimitDegrees=45
  twistMotion=Free
 }
 {
  bone1Name=dyn_braid_bk_l_MSTR
  bone2Name=neck_02
  swing1LimitDegrees=45
  swing1Motion=Free
  swing2LimitDegrees=45
  swing2Motion=Free
  twistLimitDegrees=45
  twistMotion=Free
 }
 {
  bone1Name=dyn_braid_bk_in_l_MSTR
  bone2Name=neck_02
  swing1LimitDegrees=45
  swing1Motion=Free
  swing2LimitDegrees=45
  swing2Motion=Free
  twistLimitDegrees=45
  twistMotion=Free
 }
 {
  bone1Name=dyn_braid_bk_r_MSTR
  bone2Name=neck_02
  swing1LimitDegrees=45
  swing1Motion=Free
  swing2LimitDegrees=45
  swing2Motion=Free
  twistLimitDegrees=45
  twistMotion=Free
 }
 {
  bone1Name=spine_05
  bone2Name=spine_04
  swing1LimitDegrees=45
  swing1Motion=Free
  swing2LimitDegrees=45
  swing2Motion=Free
  twistLimitDegrees=45
  twistMotion=Free
 }
 {
  bone1Name=upperarm_l
  bone2Name=clavicle_l
  swing1LimitDegrees=45
  swing1Motion=Free
  swing2LimitDegrees=45
  swing2Motion=Free
  twistLimitDegrees=45
  twistMotion=Free
 }
 {
  bone1Name=upperarm_r
  bone2Name=clavicle_r
  swing1LimitDegrees=45
  swing1Motion=Free
  swing2LimitDegrees=45
  swing2Motion=Free
  twistLimitDegrees=45
  twistMotion=Free
 }
 {
  bone1Name=clavicle_r
  bone2Name=spine_05
  swing1LimitDegrees=45
  swing1Motion=Free
  swing2LimitDegrees=45
  swing2Motion=Free
  twistLimitDegrees=45
  twistMotion=Free
 }
 {
  bone1Name=clavicle_l
  bone2Name=spine_05
  swing1LimitDegrees=45
  swing1Motion=Free
  swing2LimitDegrees=45
  swing2Motion=Free
  twistLimitDegrees=45
  twistMotion=Free
 }
 {
  bone1Name=spine_04
  bone2Name=spine_03
  swing1LimitDegrees=45
  swing1Motion=Free
  swing2LimitDegrees=45
  swing2Motion=Free
  twistLimitDegrees=45
  twistMotion=Free
 }
 {
  bone1Name=dyn_braid_fr_l_MSTR
  bone2Name=neck_02
  swing1LimitDegrees=45
  swing1Motion=Free
  swing2LimitDegrees=45
  swing2Motion=Free
  twistLimitDegrees=45
  twistMotion=Free
 }
 {bone1Name=dyn_braid_fr_l_01, bone2Name=dyn_braid_fr_l_MSTR, swing1LimitDegrees=10, swing2LimitDegrees=30, twistLimitDegrees=5}
 {bone1Name=dyn_braid_fr_l_02, bone2Name=dyn_braid_fr_l_01, swing1LimitDegrees=20, swing2LimitDegrees=45, twistLimitDegrees=5}
 {bone1Name=dyn_braid_fr_l_03, bone2Name=dyn_braid_fr_l_02, swing1LimitDegrees=45, swing2LimitDegrees=45, twistLimitDegrees=5}
 {bone1Name=dyn_braid_fr_l_04, bone2Name=dyn_braid_fr_l_03, swing1LimitDegrees=45, swing2LimitDegrees=45, twistLimitDegrees=5}
 {bone1Name=dyn_braid_bk_l_05, bone2Name=dyn_braid_bk_l_04, swing1LimitDegrees=45, swing2LimitDegrees=45, twistLimitDegrees=5}
 {
  bone1Name=neck_02
  bone2Name=spine_05
  swing1LimitDegrees=45
  swing1Motion=Free
  swing2LimitDegrees=45
  swing2Motion=Free
  twistLimitDegrees=45
  twistMotion=Free
 }
]
