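# Anim Blueprint for the M_MED_Solstice face accessory part, adding physics-driven motion to the character's braids.
# It copies the attached parent's pose, drives braid master bones with weighted multi-bone constraints, and caches the result.
# Two neck-relative RigidBody simulations read that cache, with a backpack-specific physics asset selected when back bling is detected on customization load.
# Each update computes their gravity override from base or DBNO-carried parameters via a deprecated state function.
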AnimBlueprint /BRCosmetics/Characters/Player/Male/Medium/Bodies/M_MED_Solstice/Meshes/Parts/M_MED_Solstice_FaceAcc_AnimBP
target_skeleton=&/M_MED_Solstice_FaceAcc_Skeleton
graphs [
 AnimGraph {nodes [
   n1 = Get GravityOverride()
   n2 = RigidBody!(ComponentPose=Local To Component(LocalPose=Use cached pose 'startCache'()), Alpha=0.8, OverrideWorldGravity:Vector=@n1, ExternalForce) -> (Pose) {
    bClampLinearTranslationLimitToRefPose=true
    bOverrideWorldGravity=true
    overridePhysicsAsset=&/M_MED_Solstice_FaceAcc_Dyn_braids
    simulationSpace=BaseBoneSpace
    alphaBoolBlend {blendOption=Linear}
    baseBoneRef {boneName=neck_02}
    componentAppliedLinearAccClamp=<3000 3000 3000>
    componentLinearAccScale {x=0.5}
    componentLinearVelScale {x=0.5}
    overrideWorldGravity {z=-980}
    simSpaceSettings {
     maxAngularAcceleration=500
     maxAngularVelocity=3
     maxLinearAcceleration=700
     maxLinearVelocity=500
     velocityScaleZ=0
     worldAlpha=0.5
     externalLinearDragV=<1 1>
    }
   }
   n3 = RigidBody(ComponentPose=Local To Component(LocalPose=Use cached pose 'startCache'()), Alpha=0.8, OverrideWorldGravity=@n1) {
    bClampLinearTranslationLimitToRefPose=true
    bOverrideWorldGravity=true
    overridePhysicsAsset=&/M_MED_Solstice_FaceAcc_Dyn_braids_backpack
    simulationSpace=BaseBoneSpace
    alphaBoolBlend {blendOption=Linear}
    baseBoneRef {boneName=neck_02}
    componentAppliedLinearAccClamp=<3000 3000 3000>
    componentLinearAccScale {x=0.5}
    componentLinearVelScale {x=0.5}
    overrideWorldGravity {z=-980}
    simSpaceSettings {
     maxAngularAcceleration=500
     maxAngularVelocity=3
     maxLinearAcceleration=700
     maxLinearVelocity=500
     velocityScaleZ=0
     worldAlpha=0.5
     externalLinearDragV=<1 1>
    }
   }
   Output Pose(Result=Blend Poses by bool(BlendPose_0=Component To Local(ComponentPose=@n3), BlendPose_1=Component To Local(ComponentPose=@n2), bActiveValue=Get isWearingBackbling()) {blendType=Linear}) {layerGroup=None}
   n4 = "Constraint - dyn_braid_bk_in_l_MSTR (PPPP)"!(
    ComponentPose=Local To Component(LocalPose=Copy Pose From Mesh() {bUseAttachedParent=true})
    Alpha=1
    ConstraintWeights_0=0.437
    ConstraintWeights_1:float=0.3207
    ConstraintWeights_2:float=0.1412
    ConstraintWeights_3:float=0.1011
   ) -> (Pose) {
    alphaBoolBlend {blendOption=Linear}
    boneToModify {boneName=dyn_braid_bk_in_l_MSTR}
    constraintSetup [{targetBone {boneName=head}}, {targetBone {boneName=neck_02}}, {targetBone {boneName=spine_05}}, {targetBone {boneName=neck_01}}]
    constraintWeights [1,1,1,1]
   }
   n5 = "Constraint - dyn_braid_bk_r_MSTR (PPPPP)"!(
    ComponentPose=@n4
    Alpha=1
    ConstraintWeights_0=0.4885
    ConstraintWeights_1:float=0.2278
    ConstraintWeights_2:float=0.1568
    ConstraintWeights_3:float=0.1019
    ConstraintWeights_4:float=0.025
   ) -> (Pose) {
    alphaBoolBlend {blendOption=Linear}
    boneToModify {boneName=dyn_braid_bk_r_MSTR}
    constraintSetup [
     {targetBone {boneName=head}}
     {targetBone {boneName=spine_05}}
     {targetBone {boneName=neck_02}}
     {targetBone {boneName=neck_01}}
     {targetBone {boneName=clavicle_r}}
    ]
    constraintWeights [1,1,1,1,1]
   }
   n6 = "Constraint - dyn_braid_fr_l_MSTR (PPPPPP)"!(
    ComponentPose=@n5
    Alpha=1
    ConstraintWeights_0=0.3353
    ConstraintWeights_1:float=0.2071
    ConstraintWeights_2:float=0.2016
    ConstraintWeights_3:float=0.0966
    ConstraintWeights_4:float=0.0837
    ConstraintWeights_5:float=0.0757
   ) -> (Pose) {
    alphaBoolBlend {blendOption=Linear}
    boneToModify {boneName=dyn_braid_fr_l_MSTR}
    constraintSetup [
     {targetBone {boneName=spine_05}}
     {targetBone {boneName=C_jaw}}
     {targetBone {boneName=head}}
     {targetBone {boneName=neck_02}}
     {targetBone {boneName=neck_01}}
     {targetBone {boneName=clavicle_l}}
    ]
    constraintWeights [1,1,1,1,1,1]
   }
   n7 = "Constraint - dyn_braid_fr_in_r_MSTR (PPPPPP)"!(
    ComponentPose=@n6
    Alpha=1
    ConstraintWeights_0=0.323
    ConstraintWeights_1:float=0.2879
    ConstraintWeights_2:float=0.2447
    ConstraintWeights_3:float=0.0609
    ConstraintWeights_4:float=0.0523
    ConstraintWeights_5:float=0.0313
   ) -> (Pose) {
    alphaBoolBlend {blendOption=Linear}
    boneToModify {boneName=dyn_braid_fr_in_r_MSTR}
    constraintSetup [
     {targetBone {boneName=C_jaw}}
     {targetBone {boneName=spine_05}}
     {targetBone {boneName=head}}
     {targetBone {boneName=neck_02}}
     {targetBone {boneName=neck_01}}
     {targetBone {boneName=clavicle_r}}
    ]
    constraintWeights [1,1,1,1,1,1]
   }
   n8 = "Constraint - dyn_braid_fr_r_MSTR (PPPPP)"!(
    ComponentPose=@n7
    Alpha=1
    ConstraintWeights_0=0.3093
    ConstraintWeights_1:float=0.2423
    ConstraintWeights_2:float=0.2221
    ConstraintWeights_3:float=0.138
    ConstraintWeights_4:float=0.0883
   ) -> (Pose) {
    alphaBoolBlend {blendOption=Linear}
    boneToModify {boneName=dyn_braid_fr_r_MSTR}
    constraintSetup [
     {targetBone {boneName=head}}
     {targetBone {boneName=spine_05}}
     {targetBone {boneName=neck_02}}
     {targetBone {boneName=neck_01}}
     {targetBone {boneName=clavicle_r}}
    ]
    constraintWeights [1,1,1,1,1]
   }
   n9 = "Constraint - dyn_braid_bk_l_MSTR (PPPPPP)"!(
    ComponentPose=@n8
    Alpha=1
    ConstraintWeights_0=0.3424
    ConstraintWeights_1:float=0.2566
    ConstraintWeights_2:float=0.1578
    ConstraintWeights_3:float=0.1307
    ConstraintWeights_4:float=0.0977
    ConstraintWeights_5:float=0.0147
   ) -> (Pose) {
    alphaBoolBlend {blendOption=Linear}
    boneToModify {boneName=dyn_braid_bk_l_MSTR}
    constraintSetup [
     {targetBone {boneName=spine_05}}
     {targetBone {boneName=neck_02}}
     {targetBone {boneName=neck_01}}
     {targetBone {boneName=clavicle_l}}
     {targetBone {boneName=head}}
     {targetBone {boneName=spine_03}}
    ]
    constraintWeights [1,1,1,1,1,1]
   }
   Save cached pose 'startCache'(Pose=Component To Local(ComponentPose=@n9)) {cacheName=startCache}
  ]}
 EventGraph {nodes [
   Event BlueprintUpdateAnimation()
   > n1 = Branch(Condition=Get bIsDBNOCarried())
   > n2 = Set Gravity Override Params(Gravity Override Params:Gravity Override Params Struct Structure=Get Gravity Override Params_DBNO() -> (Gravity Override Params_DBNO:Gravity Override Params Struct Structure)) -> (Output_Get:Gravity Override Params Struct Structure)
   n3 = Gravity Override State Function DEPRECATED(
    execute=[@n4.then, @n2.then]
    CharacterPartAnimInstance=Self-Reference()
    GameTimeInSeconds=GetGameTimeInSeconds()
    GravityOverrideParams=Get Gravity Override Params()
    DebugJointName=pelvis
   )
   > Set GravityOverride(GravityOverride=@n3)
   n4 = Set Gravity Override Params(
    execute=@n1.else
    Gravity Override Params=Get Gravity Override Params_Base() -> (Gravity Override Params_Base:Gravity Override Params Struct Structure)
   )
   Event OnCharacterCustomizationLoaded()
   > Do Once()
   n5 = CheckForBackbling()
   > Set isWearingBackbling(isWearingBackbling=OR Boolean!(A=@n5, B=@n5.bIsBackblingACape, C:bool=@n5.bIsBackblingLong) -> (ReturnValue))
  ]}
]
variables [
 Gravity Override Params {type=GravityOverrideParamsStruct}
 Gravity Override Params_Base {type=GravityOverrideParamsStruct}
 Gravity Override Params_DBNO {
  type=GravityOverrideParamsStruct
  default {onGround_Moving {gravityOverride=<0 -1500 -980>}, onGround_Standing {gravityOverride=<-500 500 1000>}}
 }
 GravityOverride {type=Vector}
 isWearingBackbling {type=boolean}
]
